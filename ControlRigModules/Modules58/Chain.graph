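# Modular Control Rig module that rigs a bone chain between Start Bone and End Bone connectors.
# Construction duplicates the chain as rig bones, adds controls per bone or spaced along a spline, and optionally creates dynamics animation channels, tracked through module metadata.
# Forward solve drives rig bones by constraints or spline IK, applies optional chain dynamics from those channels, then attaches the skeleton.
# Backwards solve repositions controls from their linked items.
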ControlRigBlueprint /ControlRigModules/Modules58/Chain
graphs [
 CollapseNode_ContainedGraph {nodes [
   RigVMFunctionEntryNode()
   DISPATCH_RigDispatch_GetModuleMetadata = RigVMDispatchNode(Default=(Type=Bone,Name="None"), Name=Dynamics Channel, NameSpace=Self) {template="DISPATCH_RigDispatch_GetModuleMetadata(in Name,in NameSpace,in Default,out Value,out Found)"}
   n1 = RigVMDispatchNode(Default=(Type=Bone,Name="None"), Name=Channel Parent, NameSpace=Self) {template="DISPATCH_RigDispatch_GetModuleMetadata(in Name,in NameSpace,in Default,out Value,out Found)"}
   GetAnimationChannel = /Script/ControlRig.RigUnit_GetBoolAnimationChannel(bInitial=False, Channel=@DISPATCH_RigDispatch_GetModuleMetadata.Value.Name, Control=@n1.Value.Name) {is=struct RigVMUnitNode, template="GetAnimationChannel::Execute(out Value,in Control,in Channel,in bInitial)"}
   RigVMFunctionReturnNode(
    Value=@GetAnimationChannel.Value
    Value_1=/Script/ControlRig.RigUnit_GetFloatAnimationChannel(
     bInitial=False
     Channel=RigVMDispatchNode(Default=(Type=Bone,Name="None"), Name=Strength Channel, NameSpace=Self) {template="DISPATCH_RigDispatch_GetModuleMetadata(in Name,in NameSpace,in Default,out Value,out Found)"}.Value.Name
     Control=@n1.Value.Name
    ) {is=struct RigVMUnitNode, template="GetAnimationChannel::Execute(out Value,in Control,in Channel,in bInitial)"}.Value
    Value_2=&RigUnit_GetFloatAnimationChannel(
     bInitial=False
     Channel=RigVMDispatchNode(Default=(Type=Bone,Name="None"), Name=Damping Channel, NameSpace=Self) {template="DISPATCH_RigDispatch_GetModuleMetadata(in Name,in NameSpace,in Default,out Value,out Found)"}.Value.Name
     Control=@n1.Value.Name
    ) {template="GetAnimationChannel::Execute(out Value,in Control,in Channel,in bInitial)"}.Value
    Value_3=&RigUnit_GetFloatAnimationChannel(
     bInitial=False
     Channel=RigVMDispatchNode(Default=(Type=Bone,Name="None"), Name=Blend Start Channel, NameSpace=Self) {template="DISPATCH_RigDispatch_GetModuleMetadata(in Name,in NameSpace,in Default,out Value,out Found)"}.Value.Name
     Control=@n1.Value.Name
    ) {template="GetAnimationChannel::Execute(out Value,in Control,in Channel,in bInitial)"}.Value
    Value_4=&RigUnit_GetFloatAnimationChannel(
     bInitial=False
     Channel=RigVMDispatchNode(Default=(Type=Bone,Name="None"), Name=Blend End Channel, NameSpace=Self) {template="DISPATCH_RigDispatch_GetModuleMetadata(in Name,in NameSpace,in Default,out Value,out Found)"}.Value.Name
     Control=@n1.Value.Name
    ) {template="GetAnimationChannel::Execute(out Value,in Control,in Channel,in bInitial)"}.Value
    Value_5=&RigUnit_GetFloatAnimationChannel(
     bInitial=False
     Channel=RigVMDispatchNode(Default=(Type=Bone,Name="None"), Name=Blend Twist Channel, NameSpace=Self) {template="DISPATCH_RigDispatch_GetModuleMetadata(in Name,in NameSpace,in Default,out Value,out Found)"}.Value.Name
     Control=@n1.Value.Name
    ) {template="GetAnimationChannel::Execute(out Value,in Control,in Channel,in bInitial)"}.Value
   )
  ]}
 CollapseNode_ContainedGraph {nodes [
   n1 = RigVMFunctionEntryNode()
   RigVMFunction_StringSplit = /Script/RigVM.RigVMFunction_StringSplit(Separator=_, Value=RigVMDispatchNode(Value=@n1.Value) {template="DISPATCH_RigDispatch_ToString(in Value,out Result)"}.Result) {is=struct RigVMUnitNode}
   > RigVMFunctionReturnNode(Result=RigVMDispatchNode(String=RigVMDispatchNode(Index=0, Array=@RigVMFunction_StringSplit.Result) {template="DISPATCH_RigVMDispatch_ArrayGetAtIndex(in Array,in Index,out Element)"}.Element) {template="DISPATCH_RigDispatch_FromString(in String,out Result)"}.Result)
  ]}
 Concat_1_1_ContainedGraph {nodes [
   n1 = RigVMFunctionEntryNode()
   n2 = /Script/RigVM.RigVMFunction_NameConcat(A=@n1.A, B=@n1.B) {is=struct RigVMUnitNode, template="Concat::Execute(in A,in B,out Result)"}
   RigVMFunctionReturnNode(Result=&RigVMFunction_NameConcat(A=&RigVMFunction_NameConcat(A=@n2.Result, B=@n1.C) {template="Concat::Execute(in A,in B,out Result)"}.Result, B=@n1.D) {template="Concat::Execute(in A,in B,out Result)"}.Result)
  ]}
 Concat_1_1_ContainedGraph {nodes [
   n1 = RigVMFunctionEntryNode()
   n2 = &RigVMFunction_NameConcat(A=@n1.A, B=@n1.B) {template="Concat::Execute(in A,in B,out Result)"}
   RigVMFunctionReturnNode(Result=&RigVMFunction_NameConcat(A=&RigVMFunction_NameConcat(A=@n2.Result, B=@n1.C) {template="Concat::Execute(in A,in B,out Result)"}.Result, B=@n1.D) {template="Concat::Execute(in A,in B,out Result)"}.Result)
  ]}
 Concat_4_1_ContainedGraph {nodes [
   n1 = RigVMFunctionEntryNode()
   n2 = &RigVMFunction_NameConcat(A=@n1.A, B=@n1.B) {template="Concat::Execute(in A,in B,out Result)"}
   RigVMFunctionReturnNode(Result=&RigVMFunction_NameConcat(A=&RigVMFunction_NameConcat(A=@n2.Result, B=@n1.C) {template="Concat::Execute(in A,in B,out Result)"}.Result, B=@n1.D) {template="Concat::Execute(in A,in B,out Result)"}.Result)
  ]}
 ConstructionGraph {nodes [
   RigUnit_PrepareForExecution = &RigUnit_PrepareForExecution()
   n1 = RigVMVariableNode() {variable=Module Settings}
   CONSTR Chain = RigVMFunctionReferenceNode(
    Axis=(X=-1,Y=0,Z=0)
    Parent Controls to each other=true
    ExecuteContext=@RigUnit_PrepareForExecution.ExecutePin
    Dynamics=RigVMVariableNode() {variable=Dynamics}.Value
    Create Control Per Bone=RigVMVariableNode() {variable=Create Control per Bone}.Value
    Debug=RigVMVariableNode() {variable=Debug}.Value
    Dynamics Strength=RigVMVariableNode() {variable=Default Strength}.Value
    Suffix=@n1.Value.ControlSuffix_38_#1
    Prefix=@n1.Value.Prefix_20_#2
    Dynamics Damping=RigVMVariableNode() {variable=Default Damping}.Value
    Number of Controls=RigVMVariableNode() {variable=Number of Controls}.Value
    Enable Orient Bones=RigVMVariableNode() {variable=Re-Orient Controls}.Value
    Dynamics Blend Start=RigVMVariableNode() {variable=Default Blend Start}.Value
    Dynamics Blend End=RigVMVariableNode() {variable=Default Blend End}.Value
    Dynamics Blend Twist=RigVMVariableNode() {variable=Default Blend Twist}.Value
   )
   > RigVMVariableNode(Value=/Script/ControlRig.RigUnit_FindItemsWithMetadataTag(NameSpace=Self, Tag=Controls) {is=struct RigVMUnitNode}.Items) {variable=Controls}
   > RigVMVariableNode(Value=RigVMDispatchNode(Default=(), Name=Rig Bones, NameSpace=Self) {template="DISPATCH_RigDispatch_GetModuleMetadata(in Name,in NameSpace,in Default,out Value,out Found)"}.Value) {variable=Rig Bones}
   > RigVMVariableNode(Value=RigVMDispatchNode(Default=(), Name=Bones, NameSpace=Self) {template="DISPATCH_RigDispatch_GetModuleMetadata(in Name,in NameSpace,in Default,out Value,out Found)"}.Value) {variable=Bones}
  ]}
 ContainedGraph {nodes [
   n1 = RigVMFunctionEntryNode()
   RigUnit_SequenceAggregate_1 = &RigVMFunction_Sequence(ExecuteContext=@n1.ExecuteContext)
   RigUnit_SequenceAggregate = RigVMAggregateNode(ExecuteContext=@RigUnit_SequenceAggregate_1.A)
   BranchNode_4 = &RigVMFunction_ControlFlowBranch(Condition=@n1.X Plane, ExecuteContext=@RigUnit_SequenceAggregate.A)
   n2 = RigVMVariableNode() {variable=vector pos}
   RigVMVariableNode(
    Value=(X=0,Y=0,Z=0)
    ExecuteContext=@BranchNode_4.True
    Value.X=&RigUnit_GetTransform(bInitial=true, Item=@n1.Item).Transform.Translation.X
    Value.Y=@n2.Value.Y
    Value.Z=@n2.Value.Z
   ) {variable=vector pos}
   BranchNode_1_1 = &RigVMFunction_ControlFlowBranch(Condition=@n1.Y Plane, ExecuteContext=@RigUnit_SequenceAggregate.B)
   n3 = RigVMVariableNode() {variable=vector pos}
   RigVMVariableNode(
    Value=(X=0,Y=0,Z=0)
    ExecuteContext=@BranchNode_1_1.True
    Value.Y=&RigUnit_GetTransform(bInitial=true, Item=@n1.Item).Transform.Translation.Y
    Value.X=@n3.Value.X
    Value.Z=@n3.Value.Z
   ) {variable=vector pos}
   BranchNode_2_1 = &RigVMFunction_ControlFlowBranch(Condition=@n1.Z Plane, ExecuteContext=@RigUnit_SequenceAggregate.C)
   n4 = RigVMVariableNode() {variable=vector pos}
   RigVMVariableNode(
    Value=(X=0,Y=0,Z=0)
    ExecuteContext=@BranchNode_2_1.True
    Value.Z=&RigUnit_GetTransform(bInitial=true, Item=@n1.Item).Transform.Translation.Z
    Value.X=@n4.Value.X
    Value.Y=@n4.Value.Y
   ) {variable=vector pos}
   Cast = &RigVMFunction_MathBoolToFloat(Value=@n1.X Plane)
   Dot__ = /Script/RigVM.RigVMFunction_MathVectorDot(
    B=(X=1,Y=1,Z=0)
    B.X=@Cast.Result
    B.Y=&RigVMFunction_MathBoolToFloat(Value=@n1.Y Plane).Result
    B.Z=&RigVMFunction_MathBoolToFloat(Value=@n1.Z Plane).Result
    A=RigVMVariableNode() {variable=vector pos}.Value
   ) {is=struct RigVMUnitNode, template="Dot,|::Execute(in A,in B,out Result)"}
   Less = /Script/RigVM.RigVMFunction_MathDoubleLess(B=0, A=@Dot__.Result) {is=struct RigVMUnitNode, template="Less::Execute(in A,in B,out Result)"}
   RigVMVariableNode(Value=@Less.Result, ExecuteContext=@RigUnit_SequenceAggregate.D) {variable=is on right}
   RigVMFunctionReturnNode(ExecuteContext=@RigUnit_SequenceAggregate_1.B, Right Side=RigVMVariableNode() {variable=is on right}.Value)
  ]}
 ContainedGraph {nodes [
   n1 = RigVMFunctionEntryNode()
   RigUnit_SequenceAggregate = &RigVMFunction_Sequence(ExecuteContext=@n1.ExecuteContext)
   RigUnit_SequenceAggregate_1 = &RigVMFunction_Sequence(ExecuteContext=@RigUnit_SequenceAggregate.B)
   RigUnit_SequenceAggregate_2 = &RigVMFunction_Sequence(ExecuteContext=@RigUnit_SequenceAggregate_1.B)
   RigVMFunctionReturnNode(A=@RigUnit_SequenceAggregate.A, B=@RigUnit_SequenceAggregate_1.A, C=@RigUnit_SequenceAggregate_2.A, D=@RigUnit_SequenceAggregate_2.B)
  ]}
 New Function_1_ContainedGraph {nodes [
   n1 = RigVMFunctionEntryNode()
   RigVMFunction_Sequence = &RigVMFunction_Sequence(ExecuteContext=@n1.ExecuteContext)
   n2 = /Script/ControlRig.RigUnit_GetCandidates() {is=struct RigVMUnitNode}
   n3 = &RigVMFunction_ControlFlowBranch(
    Condition=RigVMDispatchNode(B=@n1.Connector, A=@n2.Connector) {template="DISPATCH_RigVMDispatch_CoreEquals(in A,in B,out Result)"}.Result
    ExecuteContext=@RigVMFunction_Sequence.A
   )
   RigVMDispatchNode(Array=@n2.Candidates, ExecuteContext=@n3.True) {template="DISPATCH_RigVMDispatch_ArrayIterator(in Array,out Element,out Index,out Count,out Ratio)"}
   n4 = /Script/ControlRig.RigUnit_CollectionChildrenArray(
    TypeToSearch=Bone
    bDefaultChildren=True
    bIncludeParent=true
    bRecursive=true
    Parent=/Script/ControlRig.RigUnit_ResolveConnector(SkipSocket=true, Connector=@n1.Primary Connector) {is=struct RigVMUnitNode}.Result
   ) {is=struct RigVMUnitNode}
   > Clamp Array = RigVMFunctionReferenceNode(Array=@n4.Items, Max=@n1.Clamp)
   > /Script/ControlRig.RigUnit_SetDefaultMatch(Default=RigVMDispatchNode(Index=-1, Array=@n4.Items) {template="DISPATCH_RigVMDispatch_ArrayGetAtIndex(in Array,in Index,out Element)"}.Element) {is=struct RigVMUnitNode, template="Set Default Match::Execute(in Default)"}
   RigVMFunctionReturnNode(ExecuteContext=@RigVMFunction_Sequence.B)
  ]}
 New Function_ContainedGraph {nodes [
   n1 = RigVMFunctionEntryNode()
   n2 = &RigUnit_FindItemsWithMetadataTag(NameSpace=Self, Tag=Controls)
   > Scale Module Controls = RigVMFunctionReferenceNode(
    Global Scale=RigVMDispatchNode(Default=1, Name=Control Size, NameSpace=Root) {template="DISPATCH_RigDispatch_GetModuleMetadata(in Name,in NameSpace,in Default,out Value,out Found)"}.Value
    Local Scale=@n1.Module Settings.ControlSize_12_#3
    Controls=@n2.Items
   )
   > Color Module Controls = RigVMFunctionReferenceNode(
    CenterColor=@n1.Module Settings.CenterColor_9_#4
    LeftColor=@n1.Module Settings.LeftColor_3_#5
    Override=@n1.Module Settings.OverrideColor_48_#6
    RightColor=@n1.Module Settings.RightColor_5_#7
    Controls=@n2.Items
   )
   > RigVMFunctionReturnNode()
  ]}
 New Function_ContainedGraph {nodes [
   n1 = RigVMFunctionEntryNode()
   RigVMFunction_Sequence = &RigVMFunction_Sequence(ExecuteContext=@n1.ExecuteContext)
   RigVMVariableNode(Value=@n1.Axis, ExecuteContext=@RigVMFunction_Sequence.A) {variable=Bone Axis}
   > RigVMFunction_Sequence_1 = RigVMAggregateNode()
   n2 = &RigUnit_ResolveConnector(Connector=(Type=Connector,Name="Start Bone"), SkipSocket=False)
   RigVMDispatchNode(Name=Start Bone, NameSpace=Self, Value=@n2.Result, ExecuteContext=@RigVMFunction_Sequence_1.A) {template="DISPATCH_RigDispatch_SetModuleMetadata(in Name,in NameSpace,in Value,out Success)"}
   n3 = &RigUnit_ResolveConnector(Connector=(Type=Connector,Name="End Bone"), SkipSocket=False)
   > DISPATCH_RigDispatch_SetModuleMetadata = RigVMDispatchNode(Name=End Bone, NameSpace=Self, Value=@n3.Result) {template="DISPATCH_RigDispatch_SetModuleMetadata(in Name,in NameSpace,in Value,out Success)"}
   > RigVMDispatchNode(
    Name=Bones
    NameSpace=Self
    Value=/Script/ControlRig.RigUnit_HierarchyGetChainItemArray(bIncludeEnd=True, bIncludeStart=True, bReverse=false, End=@n3.Result, Start=@n2.Result) {is=struct RigVMUnitNode}.Chain
   ) {template="DISPATCH_RigDispatch_SetModuleMetadata(in Name,in NameSpace,in Value,out Success)"}
   Get_Module_Metadata = RigVMDispatchNode(Default=(), Name=Bones, NameSpace=Self) {template="DISPATCH_RigDispatch_GetModuleMetadata(in Name,in NameSpace,in Default,out Value,out Found)"}
   n4 = RigVMVariableNode() {variable=Module Settings}
   > Create Bone Chain = RigVMFunctionReferenceNode(
    Debug=true
    Debug Scale=15
    Debug Thickness=0.5
    Kind=Location
    Optional Suffix=rig
    Transforms=/Script/ControlRig.RigUnit_GetTransformItemArray(Space=GlobalSpace, bInitial=False, Items=@Get_Module_Metadata.Value) {is=struct RigVMUnitNode}.Transforms
    Parent=/Script/ControlRig.RigUnit_HierarchyGetParent(
     bDefaultParent=True
     Child=RigVMDispatchNode(Index=0, Array=@Get_Module_Metadata.Value) {template="DISPATCH_RigVMDispatch_ArrayGetAtIndex(in Array,in Index,out Element)"}.Element
    ) {is=struct RigVMUnitNode}.Parent
    Enable Orient Bones=RigVMVariableNode() {variable=Enable Orient Bones}.Value
    Name=@n4.Value.Prefix_20_#2
    Primary Axis=@n4.Value.PrimaryBoneAxis_32_#8
    Secondary Axis=@n4.Value.SecondaryAxis_51_#9
   )
   > RigVMDispatchNode(Name=Rig Bones, NameSpace=Self, Value=@"Create Bone Chain".Bones) {template="DISPATCH_RigDispatch_SetModuleMetadata(in Name,in NameSpace,in Value,out Success)"}
   n5 = &RigVMFunction_ControlFlowBranch(ExecuteContext=@RigVMFunction_Sequence_1.B, Condition=RigVMVariableNode() {variable=Create Control Per Bone}.Value)
   RigVMFunction_Sequence_2 = RigVMAggregateNode(ExecuteContext=@n5.False)
   DISPATCH_RigDispatch_GetModuleMetadata = RigVMDispatchNode(Default=(), Name=Bones, NameSpace=Self) {template="DISPATCH_RigDispatch_GetModuleMetadata(in Name,in NameSpace,in Default,out Value,out Found)"}
   SplineFromItems = RigVMFunctionReferenceNode(Samples Per Segment=16, Spline Mode=Hermite, Items=@DISPATCH_RigDispatch_GetModuleMetadata.Value, ExecuteContext=@RigVMFunction_Sequence_2.A)
   > n6 = RigVMVariableNode(Value=@SplineFromItems.Spline) {variable=Creation Spline}
   n7 = &RigVMFunction_ControlFlowBranch(ExecuteContext=@n6.ExecuteContext, Condition=RigVMVariableNode() {variable=Debug}.Value)
   /Script/ControlRigSpline.RigUnit_DrawControlRigSpline(Color=(R=1,G=0,B=0,A=1), Detail=16, Thickness=1, ExecutePin=@n7.True, Spline=@SplineFromItems.Spline) {is=struct RigVMUnitNode}
   RigVMFunction_ForLoopCount = /Script/RigVM.RigVMFunction_ForLoopCount(ExecutePin=@RigVMFunction_Sequence_2.B, Count=RigVMVariableNode() {variable=Number of Controls}.Value) {is=struct RigVMUnitNode}
   Auto Get Prefix = RigVMFunctionReferenceNode(Value=RigVMDispatchNode(
     Index=0
     Array=RigVMDispatchNode(Default=(), Name=Bones, NameSpace=Self) {template="DISPATCH_RigDispatch_GetModuleMetadata(in Name,in NameSpace,in Default,out Value,out Found)"}.Value
    ) {template="DISPATCH_RigVMDispatch_ArrayGetAtIndex(in Array,in Index,out Element)"}.Element.Name)
   n8 = RigVMVariableNode() {variable=Prefix}
   Add = /Script/RigVM.RigVMFunction_MathIntAdd(B=1, A=@RigVMFunction_ForLoopCount.Index) {is=struct RigVMUnitNode, template="Add::Execute(in A,in B,out Result)"}
   Int to Name = /Script/RigVM.RigVMFunction_MathIntToName(PaddedSize=0, Number=@Add.Result) {is=struct RigVMUnitNode, template="Int to Name::Execute(in Number,in PaddedSize,out Result)"}
   Concat_1 = RigVMAggregateNode(
    B=_
    A=RigVMDispatchNode(
     True=@"Auto Get Prefix".Result
     Condition=RigVMDispatchNode(B=None, A=@n8.Value) {template="DISPATCH_RigVMDispatch_CoreEquals(in A,in B,out Result)"}.Result
     False=RigVMVariableNode() {variable=Prefix}.Value
    ) {template="DISPATCH_RigVMDispatch_If(in Condition,in True,in False,out Result)"}.Result
    D=@"Int to Name".Result
   )
   > Control Stack at Position_1 = RigVMFunctionReferenceNode(
    Add Bottom Null=False
    Add Null=false
    Add Secondary Control=False
    Null Suffix=_null
    Secondary Settings="(DisplayName=\"None\",Shape=(bVisible=True,Name=\"Default\",Color=(R=1,G=0,B=0,A=1),Transform=(Rotation=(X=0,Y=0,Z=0,W=1),Translation=(X=0,Y=0,Z=0),Scale3D=(X=1,Y=1,Z=1))),Proxy=(bIsProxy=False,ShapeVisibility=BasedOnSelection),InitialSpace=LocalSpace,bUsePref ...(truncated)"
    Tag Controls=true
    Zero Controls=true
    Name=@Concat_1.Result
    Parent=&RigUnit_ResolveConnector(Connector=(Type=Connector,Name="Parent Control"), SkipSocket=true).Result
    Global Position=/Script/ControlRigSpline.RigUnit_TransformFromControlRigSpline2(
     PrimaryAxis=(X=1,Y=0,Z=0)
     SecondaryAxis=(X=0,Y=0,Z=0)
     U=@RigVMFunction_ForLoopCount.Ratio
     Spline=RigVMVariableNode() {variable=Creation Spline}.Value
    ) {is=struct RigVMUnitNode}.Transform
    Control Suffix=RigVMVariableNode() {variable=Suffix}.Value
    Settings=RigVMVariableNode() {variable=Control Settings}.Value
   )
   > RigVMDispatchNode(Name=FK Parent, NameSpace=Self, Value=@"Control Stack at Position_1".Control) {template="DISPATCH_RigDispatch_SetModuleMetadata(in Name,in NameSpace,in Value,out Success)"}
   RigVMFunction_ControlFlowBranch = &RigVMFunction_ControlFlowBranch(ExecuteContext=@RigVMFunction_Sequence_2.C, Condition=RigVMVariableNode() {variable=Parent Controls to each other}.Value)
   n9 = &RigUnit_FindItemsWithMetadataTag(NameSpace=Self, Tag=Controls)
   DISPATCH_RigVMDispatch_ArrayIterator = RigVMDispatchNode(Array=@n9.Items, ExecuteContext=@RigVMFunction_ControlFlowBranch.Completed) {template="DISPATCH_RigVMDispatch_ArrayIterator(in Array,out Element,out Index,out Count,out Ratio)"}
   Subtract = /Script/RigVM.RigVMFunction_MathIntSub(B=1, A=@DISPATCH_RigVMDispatch_ArrayIterator.Count) {is=struct RigVMUnitNode, template="Subtract::Execute(in A,in B,out Result)"}
   DISPATCH_RigVMDispatch_CoreEquals = RigVMDispatchNode(A=@DISPATCH_RigVMDispatch_ArrayIterator.Index, B=@Subtract.Result) {template="DISPATCH_RigVMDispatch_CoreEquals(in A,in B,out Result)"}
   RigVMFunction_MathBoolNot = &RigVMFunction_MathBoolNot(Value=@DISPATCH_RigVMDispatch_CoreEquals.Result)
   n10 = &RigVMFunction_ControlFlowBranch(ExecuteContext=@DISPATCH_RigVMDispatch_ArrayIterator.ExecuteContext, Condition=@RigVMFunction_MathBoolNot.Result)
   DISPATCH_RigVMDispatch_ArrayGetAtIndex = RigVMDispatchNode(
    Index=&RigVMFunction_MathIntAdd(B=1, A=@DISPATCH_RigVMDispatch_ArrayIterator.Index) {template="Add::Execute(in A,in B,out Result)"}.Result
    Array=@n9.Items
   ) {template="DISPATCH_RigVMDispatch_ArrayGetAtIndex(in Array,in Index,out Element)"}
   /Script/ControlRig.RigUnit_SetDefaultParent(Child=@DISPATCH_RigVMDispatch_ArrayGetAtIndex.Element, Parent=@DISPATCH_RigVMDispatch_ArrayIterator.Element, ExecutePin=@n10.True) {is=struct RigVMUnitNode}
   n11 = RigVMDispatchNode(
    Array=RigVMDispatchNode(Default=(), Name=Rig Bones, NameSpace=Self) {template="DISPATCH_RigDispatch_GetModuleMetadata(in Name,in NameSpace,in Default,out Value,out Found)"}.Value
    ExecuteContext=@n5.True
   ) {template="DISPATCH_RigVMDispatch_ArrayIterator(in Array,out Element,out Index,out Count,out Ratio)"}
   n12 = RigVMDispatchNode(B=0, A=@n11.Index) {template="DISPATCH_RigVMDispatch_CoreEquals(in A,in B,out Result)"}
   n13 = &RigVMFunction_ControlFlowBranch(ExecuteContext=@n11.ExecuteContext, Condition=@n12.Result)
   n14 = RigVMDispatchNode(
    Array=RigVMDispatchNode(Default=(), Name=Bones, NameSpace=Self) {template="DISPATCH_RigDispatch_GetModuleMetadata(in Name,in NameSpace,in Default,out Value,out Found)"}.Value
    Index=@n11.Index
   ) {template="DISPATCH_RigVMDispatch_ArrayGetAtIndex(in Array,in Index,out Element)"}
   Concat = &RigVMFunction_NameConcat(B=_ctrl, A=&RigUnit_HierarchyGetParent(bDefaultParent=True, Child=@n14.Element).Parent.Name) {template="Concat::Execute(in A,in B,out Result)"}
   Item = /Script/ControlRig.RigUnit_Item(Item=(Type=Control,Name="None"), Item.Name=@Concat.Result) {is=struct RigVMUnitNode}
   RigVMFunction_MathBoolOr = &RigVMFunction_MathBoolOr(A=RigVMVariableNode() {variable=Add Null Above}.Value, B=RigVMVariableNode() {variable=Add Null}.Value)
   Control Stack at Item = RigVMFunctionReferenceNode(
    Add Bottom Null=False
    Add Secondary Control=False
    Null Suffix=_null
    Secondary Settings="(DisplayName=\"None\",Shape=(bVisible=True,Name=\"Default\",Color=(R=1,G=0,B=0,A=1),Transform=(Rotation=(X=0,Y=0,Z=0,W=1),Translation=(X=0,Y=0,Z=0),Scale3D=(X=1,Y=1,Z=1))),Proxy=(bIsProxy=False,ShapeVisibility=BasedOnSelection),InitialSpace=LocalSpace,bUsePref ...(truncated)"
    Tag Controls=true
    Zero Controls=true
    Name=@n14.Element.Name
    Item=@n11.Element
    Parent=RigVMDispatchNode(
     Condition=@n12.Result
     False=@Item.Item
     True=&RigUnit_ResolveConnector(Connector=(Type=Connector,Name="Parent Control"), SkipSocket=true).Result
    ) {template="DISPATCH_RigVMDispatch_If(in Condition,in True,in False,out Result)"}.Result
    ExecuteContext=@n13.Completed
    Add Null=@RigVMFunction_MathBoolOr.Result
    Control Suffix=RigVMVariableNode() {variable=Suffix}.Value
    Settings=RigVMVariableNode() {variable=Control Settings}.Value
   )
   > RigVMDispatchNode(Name=FK Parent, NameSpace=Self, Value=@"Control Stack at Item".Control) {template="DISPATCH_RigDispatch_SetModuleMetadata(in Name,in NameSpace,in Value,out Success)"}
   RigVMVariableNode(Value=false, ExecuteContext=@n13.False) {variable=Add Null Above}
   RigVMVariableNode(Value=true, ExecuteContext=@n13.True) {variable=Add Null Above}
   n15 = &RigVMFunction_ControlFlowBranch(ExecuteContext=@RigVMFunction_Sequence_1.C, Condition=RigVMVariableNode() {variable=Dynamics}.Value)
   n16 = &RigVMFunction_Sequence(ExecuteContext=@n15.True)
   n17 = RigVMDispatchNode(Default=(Type=Bone,Name="None"), Name=Settings Control, NameSpace=Root) {template="DISPATCH_RigDispatch_GetModuleMetadata(in Name,in NameSpace,in Default,out Value,out Found)"}
   DISPATCH_RigVMDispatch_If = RigVMDispatchNode(
    Condition=@n17.Found
    True=@n17.Value
    False=RigVMDispatchNode(Index=0, Array=&RigUnit_FindItemsWithMetadataTag(NameSpace=Self, Tag=Controls).Items) {template="DISPATCH_RigVMDispatch_ArrayGetAtIndex(in Array,in Index,out Element)"}.Element
   ) {template="DISPATCH_RigVMDispatch_If(in Condition,in True,in False,out Result)"}
   RigVMDispatchNode(Name=Channel Parent, NameSpace=Self, Value=@DISPATCH_RigVMDispatch_If.Result, ExecuteContext=@n16.B) {template="DISPATCH_RigDispatch_SetModuleMetadata(in Name,in NameSpace,in Value,out Success)"}
   Format Name = RigVMFunctionReferenceNode(name=" Dynamics", side=None, suffix=None, prefix=RigVMVariableNode() {variable=Prefix}.Value)
   > SpawnAnimationChannel = /Script/ControlRig.RigUnit_HierarchyAddAnimationChannelBool(
    InitialValue=true
    LimitsEnabled=()
    MaximumValue=True
    MinimumValue=False
    Parent=@DISPATCH_RigVMDispatch_If.Result
    Name=@"Format Name".Result
   ) {
    is=struct RigVMUnitNode
    template="SpawnAnimationChannel::Execute(in InitialValue,in MinimumValue,in MaximumValue,in LimitsEnabled,in Parent,in Name,out Item)"
   }
   > RigVMDispatchNode(Name=Dynamics Channel, NameSpace=Self, Value=(Type=Control,Name="None"), Value.Name=@"Format Name".Result) {template="DISPATCH_RigDispatch_SetModuleMetadata(in Name,in NameSpace,in Value,out Success)"}
   Format Name_1 = RigVMFunctionReferenceNode(name=" Strength", side=None, suffix=None, prefix=RigVMVariableNode() {variable=Prefix}.Value)
   > /Script/ControlRig.RigUnit_HierarchyAddAnimationChannelFloat(
    LimitsEnabled=(Enabled=(bMinimum=True,bMaximum=True))
    MaximumValue=999
    MinimumValue=0
    Parent=@DISPATCH_RigVMDispatch_If.Result
    Name=@"Format Name_1".Result
    InitialValue=RigVMVariableNode() {variable=Dynamics Strength}.Value
   ) {
    is=struct RigVMUnitNode
    template="SpawnAnimationChannel::Execute(in InitialValue,in MinimumValue,in MaximumValue,in LimitsEnabled,in Parent,in Name,out Item)"
   }
   > RigVMDispatchNode(Name=Strength Channel, NameSpace=Self, Value=(Type=Control,Name="None"), Value.Name=@"Format Name_1".Result) {template="DISPATCH_RigDispatch_SetModuleMetadata(in Name,in NameSpace,in Value,out Success)"}
   Format Name_1_1 = RigVMFunctionReferenceNode(name=" Damping", side=None, suffix=None, prefix=RigVMVariableNode() {variable=Prefix}.Value)
   > &RigUnit_HierarchyAddAnimationChannelFloat(
    LimitsEnabled=(Enabled=(bMinimum=True,bMaximum=True))
    MaximumValue=999
    MinimumValue=0
    Parent=@DISPATCH_RigVMDispatch_If.Result
    Name=@"Format Name_1_1".Result
    InitialValue=RigVMVariableNode() {variable=Dynamics Damping}.Value
   ) {template="SpawnAnimationChannel::Execute(in InitialValue,in MinimumValue,in MaximumValue,in LimitsEnabled,in Parent,in Name,out Item)"}
   > RigVMDispatchNode(Name=Damping Channel, NameSpace=Self, Value=(Type=Control,Name="None"), Value.Name=@"Format Name_1_1".Result) {template="DISPATCH_RigDispatch_SetModuleMetadata(in Name,in NameSpace,in Value,out Success)"}
   Format Name_1_2 = RigVMFunctionReferenceNode(name=" Blend Start", side=None, suffix=None, prefix=RigVMVariableNode() {variable=Prefix}.Value)
   > &RigUnit_HierarchyAddAnimationChannelFloat(
    LimitsEnabled=(Enabled=(bMinimum=True,bMaximum=True))
    MaximumValue=999
    MinimumValue=0
    Parent=@DISPATCH_RigVMDispatch_If.Result
    Name=@"Format Name_1_2".Result
    InitialValue=RigVMVariableNode() {variable=Dynamics Blend Start}.Value
   ) {template="SpawnAnimationChannel::Execute(in InitialValue,in MinimumValue,in MaximumValue,in LimitsEnabled,in Parent,in Name,out Item)"}
   > RigVMDispatchNode(Name=Blend Start Channel, NameSpace=Self, Value=(Type=Bone,Name="None"), Value.Name=@"Format Name_1_2".Result) {template="DISPATCH_RigDispatch_SetModuleMetadata(in Name,in NameSpace,in Value,out Success)"}
   Format Name_1_2_1 = RigVMFunctionReferenceNode(name=" Blend End", side=None, suffix=None, prefix=RigVMVariableNode() {variable=Prefix}.Value)
   > &RigUnit_HierarchyAddAnimationChannelFloat(
    LimitsEnabled=(Enabled=(bMinimum=True,bMaximum=True))
    MaximumValue=999
    MinimumValue=0
    Parent=@DISPATCH_RigVMDispatch_If.Result
    Name=@"Format Name_1_2_1".Result
    InitialValue=RigVMVariableNode() {variable=Dynamics Blend End}.Value
   ) {template="SpawnAnimationChannel::Execute(in InitialValue,in MinimumValue,in MaximumValue,in LimitsEnabled,in Parent,in Name,out Item)"}
   > RigVMDispatchNode(Name=Blend End Channel, NameSpace=Self, Value=(Type=Bone,Name="None"), Value.Name=@"Format Name_1_2_1".Result) {template="DISPATCH_RigDispatch_SetModuleMetadata(in Name,in NameSpace,in Value,out Success)"}
   Format Name_1_2_1_1 = RigVMFunctionReferenceNode(name=" Blend Twist", side=None, suffix=None, prefix=RigVMVariableNode() {variable=Prefix}.Value)
   > &RigUnit_HierarchyAddAnimationChannelFloat(
    LimitsEnabled=(Enabled=(bMinimum=True,bMaximum=True))
    MaximumValue=999
    MinimumValue=0
    Parent=@DISPATCH_RigVMDispatch_If.Result
    Name=@"Format Name_1_2_1_1".Result
    InitialValue=RigVMVariableNode() {variable=Dynamics Blend Twist}.Value
   ) {template="SpawnAnimationChannel::Execute(in InitialValue,in MinimumValue,in MaximumValue,in LimitsEnabled,in Parent,in Name,out Item)"}
   > RigVMDispatchNode(Name=Blend Twist Channel, NameSpace=Self, Value=(Type=Bone,Name="None"), Value.Name=@"Format Name_1_2_1_1".Result) {template="DISPATCH_RigDispatch_SetModuleMetadata(in Name,in NameSpace,in Value,out Success)"}
   Finalize Controls = RigVMCollapseNode(ExecuteContext=@RigVMFunction_Sequence_1.F, Module Settings=RigVMVariableNode() {variable=Module Settings}.Value)
   RigVMFunctionReturnNode(ExecuteContext=@RigVMFunction_Sequence.B)
   Int to Name_1 = &RigVMFunction_MathIntToName(PaddedSize=0, Number=&RigVMFunction_MathIntAdd(B=1, A=@n11.Index) {template="Add::Execute(in A,in B,out Result)"}.Result) {template="Int to Name::Execute(in Number,in PaddedSize,out Result)"}
   Concat_1_1 = RigVMAggregateNode(B=_, C=_, D=@"Int to Name_1".Result, A=RigVMVariableNode() {variable=Prefix}.Value)
  ]}
 New Function_ContainedGraph {nodes [
   n1 = RigVMFunctionEntryNode()
   n2 = &RigUnit_FindItemsWithMetadataTag(NameSpace=Self, Tag=Controls)
   > Scale Module Controls = RigVMFunctionReferenceNode(
    Global Scale=RigVMDispatchNode(Default=1, Name=Control Size, NameSpace=Root) {template="DISPATCH_RigDispatch_GetModuleMetadata(in Name,in NameSpace,in Default,out Value,out Found)"}.Value
    Local Scale=@n1.Module Settings.ControlSize_12_#3
    Controls=@n2.Items
   )
   > Color Module Controls = RigVMFunctionReferenceNode(
    CenterColor=@n1.Module Settings.CenterColor_9_#4
    LeftColor=@n1.Module Settings.LeftColor_3_#5
    Override=@n1.Module Settings.OverrideColor_48_#6
    RightColor=@n1.Module Settings.RightColor_5_#7
    Controls=@n2.Items
   )
   > RigVMFunctionReturnNode()
  ]}
 New Function_ContainedGraph {nodes [
   n1 = RigVMFunctionEntryNode()
   RigVMFunction_Sequence = &RigVMFunction_Sequence(ExecuteContext=@n1.ExecuteContext)
   RigVMFunction_Sequence_1 = RigVMAggregateNode(ExecuteContext=@RigVMFunction_Sequence.A)
   RigVMFunction_ControlFlowBranch = &RigVMFunction_ControlFlowBranch(ExecuteContext=@RigVMFunction_Sequence_1.A, Condition=RigVMVariableNode() {variable=Use SplineIK}.Value)
   n2 = RigVMDispatchNode(ExecuteContext=@RigVMFunction_ControlFlowBranch.False, Array=RigVMVariableNode() {variable=Controls}.Value) {template="DISPATCH_RigVMDispatch_ArrayIterator(in Array,out Element,out Index,out Count,out Ratio)"}
   DISPATCH_RigDispatch_GetMetadata = RigVMDispatchNode(Default=(Type=Bone,Name="None"), Name=Item, NameSpace=Self, Item=@n2.Element) {template="DISPATCH_RigDispatch_GetMetadata(in Item,in Name,in NameSpace,in Default,out Value,out Found)"}
   > &RigUnit_ParentConstraint(
    AdvancedSettings=(InterpolationType=Average,RotationOrderForFilter=XZY)
    Parents=((Item=(Type=Bone,Name="None"),Weight=1))
    Weight=1
    Child=@DISPATCH_RigDispatch_GetMetadata.Value
    Parents.0.Item=@n2.Element
   )
   > &RigUnit_ParentConstraint(
    AdvancedSettings=(InterpolationType=Average,RotationOrderForFilter=XZY)
    Parents=((Item=(Type=Bone,Name="None"),Weight=1))
    Weight=1
    Child=RigVMDispatchNode(Default=(Type=Bone,Name="None"), Name=Dyn Bone, NameSpace=Self, Item=@DISPATCH_RigDispatch_GetMetadata.Value) {template="DISPATCH_RigDispatch_GetMetadata(in Item,in Name,in NameSpace,in Default,out Value,out Found)"}.Value
    Parents.0.Item=@n2.Element
   )
   n3 = &RigVMFunction_ControlFlowBranch(ExecuteContext=@n2.Completed, Condition=RigVMVariableNode() {variable=Create Control per Bone}.Value)
   DISPATCH_RigVMDispatch_ArrayIterator = RigVMDispatchNode(ExecuteContext=@n3.False, Array=RigVMVariableNode() {variable=Controls}.Value) {template="DISPATCH_RigVMDispatch_ArrayIterator(in Array,out Element,out Index,out Count,out Ratio)"}
   > FindClosestBone_2 = RigVMFunctionReferenceNode(Initial=true, Item=@DISPATCH_RigVMDispatch_ArrayIterator.Element, Bones=RigVMVariableNode() {variable=Bones}.Value)
   > Parent = RigVMFunctionReferenceNode(New Parent=@DISPATCH_RigVMDispatch_ArrayIterator.Element, Child=@FindClosestBone_2.Value)
   SplineIK = RigVMFunctionReferenceNode(
    Samples Per Segment=16
    Primary Axis=@n1.Primary Axis
    Secondary Spline Direction=@n1.Secondary Spline Direction
    Spline Mode=@n1.Spline Mode
    Stretch=@n1.Stretch
    Up Axis=@n1.Up Axis
    debug=@n1.debug
    ExecuteContext=@RigVMFunction_ControlFlowBranch.True
    Controls=RigVMVariableNode() {variable=Controls}.Value
    Bones=RigVMVariableNode() {variable=Rig Bones}.Value
   )
   n4 = &RigVMFunction_ControlFlowBranch(ExecuteContext=@RigVMFunction_Sequence_1.B, Condition=RigVMVariableNode() {variable=Dynamics}.Value)
   Local Chain Dynamics = RigVMFunctionReferenceNode(
    Blend End=@n1.Blend End
    Blend Start=@n1.Blend Start
    Blend Twist=@n1.Blend Twist
    Damping=@n1.Damping
    Strength=@n1.Strength
    ExecuteContext=@n4.True
    Draw Debug=RigVMVariableNode() {variable=debug}.Value
    Bones=RigVMVariableNode() {variable=Rig Bones}.Value
   )
   n5 = &RigVMFunction_ControlFlowBranch(ExecuteContext=@RigVMFunction_Sequence_1.C, Condition=RigVMVariableNode() {variable=Attach Skeleton}.Value)
   n6 = RigVMDispatchNode(ExecuteContext=@n5.True, Array=RigVMVariableNode() {variable=Rig Bones}.Value) {template="DISPATCH_RigVMDispatch_ArrayIterator(in Array,out Element,out Index,out Count,out Ratio)"}
   > Parent_2_1_1 = RigVMFunctionReferenceNode(
    Child=RigVMDispatchNode(Index=@n6.Index, Array=RigVMVariableNode() {variable=Bones}.Value) {template="DISPATCH_RigVMDispatch_ArrayGetAtIndex(in Array,in Index,out Element)"}.Element
    New Parent=@n6.Element
   )
   RigVMFunctionReturnNode(ExecuteContext=@RigVMFunction_Sequence.B)
  ]}
 New Function_ContainedGraph {nodes [
   RigVMFunctionEntryNode()
   > RigVMFunction_Sequence = RigVMAggregateNode()
   DISPATCH_RigVMDispatch_ArrayIterator = RigVMDispatchNode(ExecuteContext=@RigVMFunction_Sequence.A, Array=RigVMVariableNode() {variable=Controls}.Value) {template="DISPATCH_RigVMDispatch_ArrayIterator(in Array,out Element,out Index,out Count,out Ratio)"}
   DISPATCH_RigDispatch_GetMetadata = RigVMDispatchNode(Default=(Type=Bone,Name="None"), Name=Item, NameSpace=Self, Item=@DISPATCH_RigVMDispatch_ArrayIterator.Element) {template="DISPATCH_RigDispatch_GetMetadata(in Item,in Name,in NameSpace,in Default,out Value,out Found)"}
   RigVMFunction_ControlFlowBranch = &RigVMFunction_ControlFlowBranch(Condition=@DISPATCH_RigDispatch_GetMetadata.Found, ExecuteContext=@DISPATCH_RigVMDispatch_ArrayIterator.ExecuteContext)
   &RigUnit_SetTransform(
    Weight=1
    bInitial=False
    Item=@DISPATCH_RigVMDispatch_ArrayIterator.Element
    Value=/Script/ControlRig.RigUnit_ProjectTransformToNewParent(
     bChildInitial=True
     bNewParentInitial=False
     bOldParentInitial=True
     NewParent=@DISPATCH_RigDispatch_GetMetadata.Value
     OldParent=@DISPATCH_RigDispatch_GetMetadata.Value
     Child=@DISPATCH_RigVMDispatch_ArrayIterator.Element
    ) {is=struct RigVMUnitNode}.Transform
    ExecutePin=@RigVMFunction_ControlFlowBranch.True
   )
   RigVMFunctionReturnNode(ExecuteContext=@RigVMFunction_Sequence.C)
  ]}
 New Function_ContainedGraph {nodes [
   n1 = RigVMFunctionEntryNode()
   Concat_4 = RigVMAggregateNode(B=@n1.name, A=@n1.prefix, C=@n1.side, D=@n1.suffix)
   > RigVMFunctionReturnNode(Result=@Concat_4.Result)
  ]}
 New Function_ContainedGraph {nodes [
   RigVMFunctionEntryNode()
   > RigVMFunctionReturnNode()
  ]}
 RigVMFunctionLibrary {nodes [
   Auto Get Prefix = RigVMCollapseNode(Value=None)
   CONSTR Chain = RigVMCollapseNode(
    Axis=(X=1,Y=0,Z=0)
    Create Control Per Bone=False
    Debug=False
    Dynamics Blend End=1
    Dynamics Blend Start=1
    Dynamics Blend Twist=1
    Dynamics Damping=0.5
    Dynamics Strength=4
    Enable Orient Bones=true
    ExecuteContext=()
    Number of Controls=4
    Parent Controls to each other=False
   )
   Connect to Leaf Bones = RigVMCollapseNode(Clamp=0, Connector=(Type=Connector,Name="End"), ExecuteContext=(), Primary Connector=(Type=Connector,Name="Finger"))
   FWD Chain = RigVMCollapseNode(
    Attach Skeleton=true
    Blend End=1
    Blend Start=1
    Blend Twist=1
    Damping=0.5
    Dynamics=False
    ExecuteContext=()
    Primary Axis=(X=1,Y=0,Z=0)
    Secondary Spline Direction=(X=0,Y=5,Z=0)
    Spline Mode=BSpline
    Strength=4
    Stretch=true
    Up Axis=(X=0,Y=1,Z=0)
    Use SplineIK=False
   )
   Finalize Controls = RigVMCollapseNode(
    ExecuteContext=()
    Module Settings="(OverrideColor_48_#6=False,LeftColor_3_#5=(R=0,G=0.0231,B=0.4115,A=1),RightColor_5_#7=(R=1,G=0,B=0,A=1),CenterColor_9_#4=(R=1,G=1,B=0,A= ...(truncated)"
   )
   Format Name = RigVMCollapseNode(name=arm, prefix=None, side=_fk_ik_switch, suffix=_fk_ik_switch)
   Get Channel Fro = RigVMCollapseNode(ExecuteContext=())
   INV Chain = RigVMCollapseNode(ExecuteContext=())
   Is Right Side = RigVMCollapseNode(ExecuteContext=(), Item=(Type=Bone))
  ]}
 RigVMFunction_Sequence_1_1_ContainedGraph {nodes [
   n1 = RigVMFunctionEntryNode()
   n2 = &RigVMFunction_Sequence(ExecuteContext=@n1.ExecuteContext)
   n3 = &RigVMFunction_Sequence(ExecuteContext=@n2.B)
   n4 = &RigVMFunction_Sequence(ExecuteContext=@n3.B)
   n5 = &RigVMFunction_Sequence(ExecuteContext=@n4.B)
   n6 = &RigVMFunction_Sequence(ExecuteContext=@n5.B)
   RigVMFunctionReturnNode(A=@n2.A, B=@n3.A, C=@n4.A, D=@n5.A, E=@n6.A, F=@n6.B)
  ]}
 RigVMFunction_Sequence_1_1_ContainedGraph {nodes [
   n1 = RigVMFunctionEntryNode()
   n2 = &RigVMFunction_Sequence(ExecuteContext=@n1.ExecuteContext)
   n3 = &RigVMFunction_Sequence(ExecuteContext=@n2.B)
   n4 = &RigVMFunction_Sequence(ExecuteContext=@n3.B)
   RigVMFunctionReturnNode(A=@n2.A, B=@n3.A, C=@n4.A, D=@n4.B)
  ]}
 RigVMFunction_Sequence_1_ContainedGraph {nodes [
   n1 = RigVMFunctionEntryNode()
   RigVMFunction_Sequence = &RigVMFunction_Sequence(ExecuteContext=@n1.ExecuteContext)
   n2 = &RigVMFunction_Sequence(ExecuteContext=@RigVMFunction_Sequence.B)
   n3 = &RigVMFunction_Sequence(ExecuteContext=@n2.B)
   RigVMFunctionReturnNode(A=@RigVMFunction_Sequence.A, B=@n2.A, C=@n3.A, D=@n3.B)
  ]}
 RigVMFunction_Sequence_1_ContainedGraph {nodes [
   n1 = RigVMFunctionEntryNode()
   RigVMFunction_Sequence = &RigVMFunction_Sequence(ExecuteContext=@n1.ExecuteContext)
   n2 = &RigVMFunction_Sequence(ExecuteContext=@RigVMFunction_Sequence.B)
   RigVMFunctionReturnNode(A=@RigVMFunction_Sequence.A, B=@n2.A, C=@n2.B)
  ]}
 RigVMFunction_Sequence_2_1_ContainedGraph {nodes [
   n1 = RigVMFunctionEntryNode()
   n2 = &RigVMFunction_Sequence(ExecuteContext=@n1.ExecuteContext)
   n3 = &RigVMFunction_Sequence(ExecuteContext=@n2.B)
   RigVMFunctionReturnNode(A=@n2.A, B=@n3.A, C=@n3.B)
  ]}
 RigVMModel {nodes [
   RigUnit_BeginExecution = &RigUnit_BeginExecution()
   Get Dynamics Channels = RigVMCollapseNode()
   FWD Chain = RigVMFunctionReferenceNode(
    Attach Skeleton=true
    Dynamics=@"Get Dynamics Channels".Value
    Strength=@"Get Dynamics Channels".Value_1
    Damping=@"Get Dynamics Channels".Value_2
    Blend Start=@"Get Dynamics Channels".Value_3
    Blend End=@"Get Dynamics Channels".Value_4
    Blend Twist=@"Get Dynamics Channels".Value_5
    ExecuteContext=@RigUnit_BeginExecution.ExecutePin
    Primary Axis=RigVMVariableNode() {variable=Module Settings}.Value.PrimaryBoneAxis_32_#8
    Use SplineIK=RigVMVariableNode() {variable=Solve with SplineIK}.Value
    Secondary Spline Direction=RigVMVariableNode() {variable=Secondary Spline Direction}.Value
    debug=RigVMVariableNode() {variable=Debug}.Value
    Spline Mode=RigVMVariableNode() {variable=Spline Mode}.Value
    Stretch=RigVMVariableNode() {variable=Stretch}.Value
    Up Axis=RigVMVariableNode() {variable=Up Axis}.Value
   )
  ]}
 RigVMModel Backwards Solve Graph {nodes [
   n1 = /Script/ControlRig.RigUnit_InverseExecution() {is=struct RigVMUnitNode}
   INV Chain = RigVMFunctionReferenceNode(ExecuteContext=@n1.ExecutePin)
  ]}
 RigVMModel Connector Event Graph {nodes [
   n1 = /Script/ControlRig.RigUnit_ConnectorExecution() {is=struct RigVMUnitNode}
   RigVMFunction_Sequence = RigVMAggregateNode(ExecuteContext=@n1.ExecuteContext)
   n2 = &RigUnit_GetCandidates()
   n3 = &RigVMFunction_ControlFlowBranch(
    Condition=RigVMDispatchNode(B=(Type=Connector,Name="Start Bone"), A=@n2.Connector) {template="DISPATCH_RigVMDispatch_CoreEquals(in A,in B,out Result)"}.Result
    ExecuteContext=@RigVMFunction_Sequence.A
   )
   n4 = RigVMDispatchNode(ExecuteContext=@n3.True, Array=@n2.Candidates) {template="DISPATCH_RigVMDispatch_ArrayIterator(in Array,out Element,out Index,out Count,out Ratio)"}
   n5 = &RigUnit_ResolveConnector(Connector=(Type=Connector,Name="Chain Primary"), SkipSocket=true)
   n6 = &RigUnit_CollectionChildrenArray(TypeToSearch=Bone, bDefaultChildren=True, bIncludeParent=False, bRecursive=False, Parent=@n5.Result)
   n7 = &RigVMFunction_ControlFlowBranch(
    ExecuteContext=@n4.ExecuteContext
    Condition=RigVMDispatchNode(B=0, A=RigVMDispatchNode(Array=@n6.Items) {template="DISPATCH_RigVMDispatch_ArrayGetNum(in Array,out Num)"}.Num) {template="DISPATCH_RigVMDispatch_CoreEquals(in A,in B,out Result)"}.Result
   )
   &RigUnit_SetDefaultMatch(Default=@n5.Result, ExecutePin=@n7.False) {template="Set Default Match::Execute(in Default)"}
   &RigUnit_SetDefaultMatch(Default=&RigUnit_HierarchyGetParent(bDefaultParent=True, Child=@n5.Result).Parent, ExecutePin=@n7.True) {template="Set Default Match::Execute(in Default)"}
   > DISPATCH_RigDispatch_SetModuleMetadata = RigVMDispatchNode(Name=Connected at End, NameSpace=Self, Value=true) {template="DISPATCH_RigDispatch_SetModuleMetadata(in Name,in NameSpace,in Value,out Success)"}
   n8 = &RigUnit_GetCandidates()
   n9 = &RigVMFunction_ControlFlowBranch(
    Condition=RigVMDispatchNode(B=(Type=Connector,Name="End Bone"), A=@n8.Connector) {template="DISPATCH_RigVMDispatch_CoreEquals(in A,in B,out Result)"}.Result
    ExecuteContext=@RigVMFunction_Sequence.B
   )
   n10 = RigVMDispatchNode(ExecuteContext=@n9.True, Array=@n8.Candidates) {template="DISPATCH_RigVMDispatch_ArrayIterator(in Array,out Element,out Index,out Count,out Ratio)"}
   DISPATCH_RigDispatch_GetModuleMetadata = RigVMDispatchNode(Default=False, Name=Connected at End, NameSpace=Self) {template="DISPATCH_RigDispatch_GetModuleMetadata(in Name,in NameSpace,in Default,out Value,out Found)"}
   n11 = &RigVMFunction_ControlFlowBranch(Condition=@DISPATCH_RigDispatch_GetModuleMetadata.Found, ExecuteContext=@n10.ExecuteContext)
   n12 = &RigVMFunction_ControlFlowBranch(Condition=@DISPATCH_RigDispatch_GetModuleMetadata.Value, ExecuteContext=@n11.True)
   RigVMFunction_ControlFlowBranch = &RigVMFunction_ControlFlowBranch(
    Condition=RigVMDispatchNode(A=@n10.Element, B=&RigUnit_ResolveConnector(Connector=(Type=Connector,Name="Chain Primary"), SkipSocket=true).Result) {template="DISPATCH_RigVMDispatch_CoreEquals(in A,in B,out Result)"}.Result
    ExecuteContext=@n12.True
   )
   &RigUnit_SetDefaultMatch(Default=@n10.Element, ExecutePin=@RigVMFunction_ControlFlowBranch.True) {template="Set Default Match::Execute(in Default)"}
   n13 = RigVMDispatchNode(Default=False, Name=Connected at End, NameSpace=Self) {template="DISPATCH_RigDispatch_GetModuleMetadata(in Name,in NameSpace,in Default,out Value,out Found)"}
   n14 = &RigVMFunction_ControlFlowBranch(Condition=@n13.Found, ExecuteContext=@n10.Completed)
   n15 = &RigVMFunction_ControlFlowBranch(Condition=@n13.Value, ExecuteContext=@n14.False)
   Connect to Leaf Bones = RigVMFunctionReferenceNode(Clamp=4, Connector=(Type=Connector,Name="End Bone"), Primary Connector=(Type=Connector,Name="Chain Primary"), ExecuteContext=@n15.False)
   Connect to Module Metadata_1 = RigVMFunctionReferenceNode(
    Connector=(Type=Connector,Name="Parent Control")
    Module Metadata Name=Parent Control
    Module Metadata NameSpace=Parent
    ExecuteContext=@RigVMFunction_Sequence.D
   )
   > Connect to Module Metadata_1_1 = RigVMFunctionReferenceNode(Connector=(Type=Connector,Name="Parent Control"), Module Metadata Name=FK Parent, Module Metadata NameSpace=Parent)
   > Connect to Module Metadata_1_2 = RigVMFunctionReferenceNode(Connector=(Type=Connector,Name="Settings Control"), Module Metadata Name=Settings Control, Module Metadata NameSpace=Root)
  ]}
]
hierarchy [
 Chain Primary {type=Connector, loc=<0 0 0>, rot=<0 0 0 1>}
 Root {type=Socket, loc=<0 0 0>, rot=<0 0 0 1>}
 Start Bone {type=Connector, loc=<0 0 0>, rot=<0 0 0 1>}
 End Bone {type=Connector, loc=<0 0 0>, rot=<0 0 0 1>}
 Parent Control {type=Connector, loc=<0 0 0>, rot=<0 0 0 1>}
 Settings Control {type=Connector, loc=<0 0 0>, rot=<0 0 0 1>}
]
variables [
 Add Null {type=bool, default=False}
 Bones {type="TArray<FRigElementKey>"}
 Control Settings {
  type=FRigUnit_HierarchyAddControlTransform_Settings
  default="(InitialSpace=LocalSpace,bUsePreferredRotationOrder=False,PreferredRotationOrder=YZX,Limits=(LimitTranslationX=(bMinimum=False,bMaximum=False),LimitTranslationY=(bMinimum=False,bMaximum=False),LimitTranslationZ=(bMinimum=False,bMaximum=False),LimitPitch=(b ...(truncated)"
 }
 Controls {type="TArray<FRigElementKey>"}
 Create Control per Bone {type=bool, default=True}
 Debug {type=bool, default=False}
 Default Blend End {type=double, default=1}
 Default Blend Start {type=double, default=1}
 Default Blend Twist {type=double, default=1}
 Default Damping {type=double, default=0.5}
 Default Strength {type=double, default=4}
 Dynamics {type=bool, default=False}
 Module Settings {
  type=FUserDefinedStruct_#10
  default="(OverrideColor_48_#6=False,LeftColor_3_#5=(R=0,G=0.0231,B=0.4115,A=1),RightColor_5_#7=(R=1,G=0,B=0,A=1),CenterColor_9_#4=(R=1,G=1,B=0,A= ...(truncated)"
 }
 Name {type=FName, default=None}
 Number of Controls {type=int32, default=5}
 Prefix {type=FName, default=FOo}
 Primary Axis {type=FVector, default=(X=1,Y=0,Z=0)}
 Re-Orient Controls {type=bool, default=True}
 Rig Bones {type="TArray<FRigElementKey>"}
 Secondary Spline Direction {type=FVector, default=(X=0,Y=5,Z=0)}
 Solve with SplineIK {type=bool, default=False}
 Spline Mode {type=ESplineType, default=BSpline}
 Stretch {type=bool, default=True}
 Suffix {type=FName, default=_ctrl}
 Up Axis {type=FVector, default=(X=0,Y=1,Z=0)}
]
