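# Control Rig module that constrains the item on its Constraint connector to the item on its Parent connector.
# Forward solve resolves both connectors, then toggles gate parent, position, rotation, scale and aim constraints with per-axis filters and shared maintain-offset, plus debug drawing.
# Active weights only the parent constraint; aim takes its world-up space from the Aim Up connector.
# A connector event defaults Parent to the Constraint item when it is among the candidates.
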ControlRigBlueprint /ControlRigModules/Modules58/Constraint
graphs [
 ConstructionGraph {nodes [RigUnit_PrepareForExecution = &RigUnit_PrepareForExecution()]}
 New Function_ContainedGraph {nodes [
   RigVMFunctionEntryNode()
   > RigVMFunction_Sequence = RigVMAggregateNode()
   RigVMVariableNode(
    Value=/Script/ControlRig.RigUnit_ResolveConnector(Connector=(Type=Connector,Name="Constraint"), SkipSocket=true) {is=struct RigVMUnitNode}.Result
    ExecuteContext=@RigVMFunction_Sequence.A
   ) {variable=child item}
   > RigVMVariableNode(Value=&RigUnit_ResolveConnector(Connector=(Type=Connector,Name="Parent"), SkipSocket=true).Result) {variable=parent item}
   RigVMFunction_ControlFlowBranch = &RigVMFunction_ControlFlowBranch(ExecuteContext=@RigVMFunction_Sequence.B, Condition=RigVMVariableNode() {variable=Parent}.Value)
   RigVMFunction_MathBoolMake = /Script/RigVM.RigVMFunction_MathBoolMake(Value=RigVMVariableNode() {variable=Active}.Value) {is=struct RigVMUnitNode}
   RigVMFunction_MathBoolToFloat = &RigVMFunction_MathBoolToFloat(Value=@RigVMFunction_MathBoolMake.Value)
   &RigUnit_ParentConstraint(
    AdvancedSettings=(InterpolationType=Average,RotationOrderForFilter=XZY)
    Parents=((Item=(Type=Bone,Name="None"),Weight=1))
    ExecutePin=@RigVMFunction_ControlFlowBranch.True
    Weight=@RigVMFunction_MathBoolToFloat.Result
    Child=RigVMVariableNode() {variable=child item}.Value
    Parents.0.Item=RigVMVariableNode() {variable=parent item}.Value
    Filter=RigVMVariableNode() {variable=ParentConstraint_Filter}.Value
    bMaintainOffset=RigVMVariableNode() {variable=Maintain Offset}.Value
   )
   n1 = &RigVMFunction_ControlFlowBranch(ExecuteContext=@RigVMFunction_Sequence.C, Condition=RigVMVariableNode() {variable=Position}.Value)
   /Script/ControlRig.RigUnit_PositionConstraintLocalSpaceOffset(
    Parents=((Item=(Type=Bone,Name="None"),Weight=1))
    Weight=1
    ExecutePin=@n1.True
    Child=RigVMVariableNode() {variable=child item}.Value
    Parents.0.Item=RigVMVariableNode() {variable=parent item}.Value
    Filter=RigVMVariableNode() {variable=PositionConstraint_Filter}.Value
    bMaintainOffset=RigVMVariableNode() {variable=Maintain Offset}.Value
   ) {is=struct RigVMUnitNode}
   n2 = &RigVMFunction_ControlFlowBranch(ExecuteContext=@RigVMFunction_Sequence.D, Condition=RigVMVariableNode() {variable=Rotation}.Value)
   &RigUnit_RotationConstraintLocalSpaceOffset(
    AdvancedSettings=(InterpolationType=Average,RotationOrderForFilter=XZY)
    Parents=((Item=(Type=Bone,Name="None"),Weight=1))
    Weight=1
    ExecutePin=@n2.True
    Child=RigVMVariableNode() {variable=child item}.Value
    Parents.0.Item=RigVMVariableNode() {variable=parent item}.Value
    Filter=RigVMVariableNode() {variable=RotationConstraint_Filter}.Value
    bMaintainOffset=RigVMVariableNode() {variable=Maintain Offset}.Value
   )
   n3 = &RigVMFunction_ControlFlowBranch(ExecuteContext=@RigVMFunction_Sequence.E, Condition=RigVMVariableNode() {variable=Scale}.Value)
   /Script/ControlRig.RigUnit_ScaleConstraintLocalSpaceOffset(
    Parents=((Item=(Type=Bone,Name="None"),Weight=1))
    Weight=1
    ExecutePin=@n3.True
    Child=RigVMVariableNode() {variable=child item}.Value
    Parents.0.Item=RigVMVariableNode() {variable=parent item}.Value
    Filter=RigVMVariableNode() {variable=ScaleConstraint_Filter}.Value
    bMaintainOffset=RigVMVariableNode() {variable=Maintain Offset}.Value
   ) {is=struct RigVMUnitNode}
   n4 = &RigVMFunction_ControlFlowBranch(ExecuteContext=@RigVMFunction_Sequence.F, Condition=RigVMVariableNode() {variable=Aim}.Value)
   &RigUnit_AimConstraintLocalSpaceOffset(
    Parents=((Item=(Type=Bone,Name="None"),Weight=1))
    Weight=1
    WorldUp=(Target=(X=0,Y=0,Z=1),Kind=Direction,Space=(Type=None,Name="None"))
    WorldUp.Space=&RigUnit_ResolveConnector(Connector=(Type=Connector,Name="Aim Up"), SkipSocket=true).Result
    ExecutePin=@n4.True
    Child=RigVMVariableNode() {variable=child item}.Value
    Parents.0.Item=RigVMVariableNode() {variable=parent item}.Value
    Filter=RigVMVariableNode() {variable=AimConstraint_Filter}.Value
    bMaintainOffset=RigVMVariableNode() {variable=Maintain Offset}.Value
    AimAxis=RigVMVariableNode() {variable=AimAxis}.Value
    UpAxis=RigVMVariableNode() {variable=AimUpAxis}.Value
    WorldUp.Kind=RigVMVariableNode() {variable=AimUpKind}.Value
   )
   n5 = &RigVMFunction_ControlFlowBranch(ExecuteContext=@RigVMFunction_Sequence.G, Condition=RigVMVariableNode() {variable=Debug}.Value)
   DrawItems = RigVMFunctionReferenceNode(
    Active=true
    Items=((Type=None,Name="None"),(Type=None,Name="None"))
    Scale=10
    Thickness=1
    ExecuteContext=@n5.True
    Items.0=RigVMVariableNode() {variable=child item}.Value
    Items.1=RigVMVariableNode() {variable=parent item}.Value
   )
   RigVMFunctionReturnNode(ExecuteContext=@RigVMFunction_Sequence.H)
  ]}
 RigVMFunctionLibrary {nodes [FWD Constraint = RigVMCollapseNode(ExecuteContext=())]}
 RigVMFunction_Sequence_1_ContainedGraph {nodes [
   n1 = RigVMFunctionEntryNode()
   RigVMFunction_Sequence = &RigVMFunction_Sequence(ExecuteContext=@n1.ExecuteContext)
   n2 = &RigVMFunction_Sequence(ExecuteContext=@RigVMFunction_Sequence.B)
   n3 = &RigVMFunction_Sequence(ExecuteContext=@n2.B)
   n4 = &RigVMFunction_Sequence(ExecuteContext=@n3.B)
   n5 = &RigVMFunction_Sequence(ExecuteContext=@n4.B)
   n6 = &RigVMFunction_Sequence(ExecuteContext=@n5.B)
   n7 = &RigVMFunction_Sequence(ExecuteContext=@n6.B)
   RigVMFunctionReturnNode(A=@RigVMFunction_Sequence.A, B=@n2.A, C=@n3.A, D=@n4.A, E=@n5.A, F=@n6.A, G=@n7.A, H=@n7.B)
  ]}
 RigVMModel {nodes [
   RigUnit_BeginExecution = &RigUnit_BeginExecution()
   FWD Constraint_1 = RigVMFunctionReferenceNode(ExecuteContext=@RigUnit_BeginExecution.ExecutePin)
  ]}
 RigVMModel Connector Event Graph {nodes [
   n1 = /Script/ControlRig.RigUnit_GetCandidates() {is=struct RigVMUnitNode}
   Item = /Script/ControlRig.RigUnit_Item(Item=(Type=Connector,Name="Parent")) {is=struct RigVMUnitNode}
   n2 = &RigVMFunction_ControlFlowBranch(
    ExecuteContext=()
    Condition=RigVMDispatchNode(B=@n1.Connector, A=@Item.Item) {template="DISPATCH_RigVMDispatch_CoreEquals(in A,in B,out Result)"}.Result
   )
   n3 = RigVMDispatchNode(Array=@n1.Candidates, ExecuteContext=@n2.True) {template="DISPATCH_RigVMDispatch_ArrayIterator(in Array,out Element,out Index,out Count,out Ratio)"}
   n4 = &RigUnit_ResolveConnector(Connector=(Type=Connector,Name="Constraint"), SkipSocket=true)
   RigVMFunction_ControlFlowBranch = &RigVMFunction_ControlFlowBranch(
    ExecuteContext=@n3.ExecuteContext
    Condition=RigVMDispatchNode(B=@n3.Element, A=@n4.Result) {template="DISPATCH_RigVMDispatch_CoreEquals(in A,in B,out Result)"}.Result
   )
   /Script/ControlRig.RigUnit_SetDefaultMatch(Default=@n4.Result, ExecutePin=@RigVMFunction_ControlFlowBranch.True) {is=struct RigVMUnitNode, template="Set Default Match::Execute(in Default)"}
   /Script/ControlRig.RigUnit_ConnectorExecution() {is=struct RigVMUnitNode}
  ]}
]
hierarchy [
 Constraint {type=Connector, loc=<0 0 0>, rot=<0 0 0 1>}
 Root {type=Socket, loc=<0 0 0>, rot=<0 0 0 1>}
 Parent {type=Connector, loc=<0 0 0>, rot=<0 0 0 1>}
 Aim Up {type=Connector, loc=<0 0 0>, rot=<0 0 0 1>}
]
variables [
 Active {type=bool, default=True}
 Aim {type=bool, default=False}
 AimAxis {type=FVector, default=(X=1,Y=0,Z=0)}
 AimConstraint_Filter {type=FFilterOptionPerAxis, default=(bX=True,bY=True,bZ=True)}
 AimUpAxis {type=FVector, default=(X=0,Y=0,Z=1)}
 AimUpKind {type=EControlRigVectorKind, default=Direction}
 Debug {type=bool, default=False}
 Maintain Offset {type=bool, default=True}
 Parent {type=bool, default=True}
 ParentConstraint_Filter {
  type=FTransformFilter
  default=(TranslationFilter=(bX=True,bY=True,bZ=True),RotationFilter=(bX=True,bY=True,bZ=True),ScaleFilter=(bX=True,bY=True,bZ=True))
 }
 Position {type=bool, default=False}
 PositionConstraint_Filter {type=FFilterOptionPerAxis, default=(bX=True,bY=True,bZ=True)}
 Rotation {type=bool, default=False}
 RotationConstraint_Filter {type=FFilterOptionPerAxis, default=(bX=True,bY=True,bZ=True)}
 Scale {type=bool, default=False}
 ScaleConstraint_Filter {type=FFilterOptionPerAxis, default=(bX=True,bY=True,bZ=True)}
]
