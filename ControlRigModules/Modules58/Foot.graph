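# Modular Control Rig module that builds a foot rig with ball FK and reverse-foot roll controls from connector-resolved bones.
# Construction reads effector, foot control and FK parent metadata from the parent module, then creates controls and re-parents the effector under a foot IK null.
# Helper graphs spread roll-control rotation along back-chain buffers with clamping and side detection.
# IK/FK logic toggles visibility, snaps inactive controls to bones and requests auto-keys.
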ControlRigBlueprint /ControlRigModules/Modules58/Foot
graphs [
 Concat_1_ContainedGraph {nodes [
   n1 = RigVMFunctionEntryNode()
   Concat = /Script/RigVM.RigVMFunction_NameConcat(A=@n1.A, B=@n1.B) {is=struct RigVMUnitNode, template="Concat::Execute(in A,in B,out Result)"}
   RigVMFunctionReturnNode(Result=&RigVMFunction_NameConcat(A=@Concat.Result, B=@n1.C) {template="Concat::Execute(in A,in B,out Result)"}.Result)
  ]}
 Concat_1_ContainedGraph {nodes [
   n1 = RigVMFunctionEntryNode()
   Concat = &RigVMFunction_NameConcat(A=@n1.A, B=@n1.B) {template="Concat::Execute(in A,in B,out Result)"}
   RigVMFunctionReturnNode(Result=&RigVMFunction_NameConcat(A=@Concat.Result, B=@n1.C) {template="Concat::Execute(in A,in B,out Result)"}.Result)
  ]}
 Concat_1_ContainedGraph {nodes [
   n1 = RigVMFunctionEntryNode()
   Concat = &RigVMFunction_NameConcat(A=@n1.A, B=@n1.B) {template="Concat::Execute(in A,in B,out Result)"}
   RigVMFunctionReturnNode(Result=&RigVMFunction_NameConcat(A=@Concat.Result, B=@n1.C) {template="Concat::Execute(in A,in B,out Result)"}.Result)
  ]}
 Concat_1_ContainedGraph {nodes [
   n1 = RigVMFunctionEntryNode()
   Concat = &RigVMFunction_NameConcat(A=@n1.A, B=@n1.B) {template="Concat::Execute(in A,in B,out Result)"}
   RigVMFunctionReturnNode(Result=&RigVMFunction_NameConcat(A=@Concat.Result, B=@n1.C) {template="Concat::Execute(in A,in B,out Result)"}.Result)
  ]}
 Concat_1_ContainedGraph {nodes [
   n1 = RigVMFunctionEntryNode()
   Concat = &RigVMFunction_NameConcat(A=@n1.A, B=@n1.B) {template="Concat::Execute(in A,in B,out Result)"}
   RigVMFunctionReturnNode(Result=&RigVMFunction_NameConcat(A=@Concat.Result, B=@n1.C) {template="Concat::Execute(in A,in B,out Result)"}.Result)
  ]}
 Concat_1_ContainedGraph {nodes [
   n1 = RigVMFunctionEntryNode()
   Concat = &RigVMFunction_NameConcat(A=@n1.A, B=@n1.B) {template="Concat::Execute(in A,in B,out Result)"}
   RigVMFunctionReturnNode(Result=&RigVMFunction_NameConcat(A=@Concat.Result, B=@n1.C) {template="Concat::Execute(in A,in B,out Result)"}.Result)
  ]}
 Concat_1_ContainedGraph {nodes [
   n1 = RigVMFunctionEntryNode()
   Concat = &RigVMFunction_NameConcat(A=@n1.A, B=@n1.B) {template="Concat::Execute(in A,in B,out Result)"}
   RigVMFunctionReturnNode(Result=&RigVMFunction_NameConcat(A=@Concat.Result, B=@n1.C) {template="Concat::Execute(in A,in B,out Result)"}.Result)
  ]}
 Concat_1_ContainedGraph {nodes [
   n1 = RigVMFunctionEntryNode()
   Concat = &RigVMFunction_NameConcat(A=@n1.A, B=@n1.B) {template="Concat::Execute(in A,in B,out Result)"}
   RigVMFunctionReturnNode(Result=&RigVMFunction_NameConcat(A=@Concat.Result, B=@n1.C) {template="Concat::Execute(in A,in B,out Result)"}.Result)
  ]}
 Concat_1_ContainedGraph {nodes [
   n1 = RigVMFunctionEntryNode()
   Concat = &RigVMFunction_NameConcat(A=@n1.A, B=@n1.B) {template="Concat::Execute(in A,in B,out Result)"}
   RigVMFunctionReturnNode(Result=&RigVMFunction_NameConcat(A=@Concat.Result, B=@n1.C) {template="Concat::Execute(in A,in B,out Result)"}.Result)
  ]}
 Concat_1_ContainedGraph {nodes [
   n1 = RigVMFunctionEntryNode()
   Concat = &RigVMFunction_NameConcat(A=@n1.A, B=@n1.B) {template="Concat::Execute(in A,in B,out Result)"}
   RigVMFunctionReturnNode(Result=&RigVMFunction_NameConcat(A=@Concat.Result, B=@n1.C) {template="Concat::Execute(in A,in B,out Result)"}.Result)
  ]}
 Concat_1_ContainedGraph {nodes [
   n1 = RigVMFunctionEntryNode()
   Concat = &RigVMFunction_NameConcat(A=@n1.A, B=@n1.B) {template="Concat::Execute(in A,in B,out Result)"}
   RigVMFunctionReturnNode(Result=&RigVMFunction_NameConcat(A=@Concat.Result, B=@n1.C) {template="Concat::Execute(in A,in B,out Result)"}.Result)
  ]}
 Concat_1_ContainedGraph {nodes [
   n1 = RigVMFunctionEntryNode()
   Concat = &RigVMFunction_NameConcat(A=@n1.A, B=@n1.B) {template="Concat::Execute(in A,in B,out Result)"}
   RigVMFunctionReturnNode(Result=&RigVMFunction_NameConcat(A=@Concat.Result, B=@n1.C) {template="Concat::Execute(in A,in B,out Result)"}.Result)
  ]}
 Concat_4_1_ContainedGraph {nodes [
   n1 = RigVMFunctionEntryNode()
   n2 = &RigVMFunction_NameConcat(A=@n1.A, B=@n1.B) {template="Concat::Execute(in A,in B,out Result)"}
   RigVMFunctionReturnNode(Result=&RigVMFunction_NameConcat(A=&RigVMFunction_NameConcat(A=@n2.Result, B=@n1.C) {template="Concat::Execute(in A,in B,out Result)"}.Result, B=@n1.D) {template="Concat::Execute(in A,in B,out Result)"}.Result)
  ]}
 ConstructionGraph {nodes [
   RigUnit_PrepareForExecution = &RigUnit_PrepareForExecution()
   RigVMFunction_Sequence = &RigVMFunction_Sequence(ExecuteContext=@RigUnit_PrepareForExecution.ExecutePin)
   DISPATCH_RigDispatch_GetModuleMetadata = RigVMDispatchNode(Default=(Type=Bone,Name="None"), Name=Effector Null, NameSpace=Parent) {template="DISPATCH_RigDispatch_GetModuleMetadata(in Name,in NameSpace,in Default,out Value,out Found)"}
   RigVMVariableNode(Value=@DISPATCH_RigDispatch_GetModuleMetadata.Value, ExecuteContext=@RigVMFunction_Sequence.A) {variable=Effector Null}
   n1 = RigVMDispatchNode(Default=(Type=Bone,Name="None"), Name=Foot Control, NameSpace=Parent) {template="DISPATCH_RigDispatch_GetModuleMetadata(in Name,in NameSpace,in Default,out Value,out Found)"}
   > RigVMVariableNode(Value=@n1.Value) {variable=Foot Control}
   > RigVMVariableNode(Value=@n1.Found) {variable=Foot Control Metadata Found}
   > RigVMVariableNode(Value=RigVMDispatchNode(Default=(Type=Bone,Name="None"), Name=FK Parent, NameSpace=Parent) {template="DISPATCH_RigDispatch_GetModuleMetadata(in Name,in NameSpace,in Default,out Value,out Found)"}.Value) {variable=FK Parent}
   CONSTR Foot = RigVMFunctionReferenceNode(
    Ball Bone Connector=(Type=Connector,Name="Ball Bone")
    Foot Bone Connector=(Type=Connector,Name="Foot Bone")
    Parent Control Connector=(Type=Connector,Name="Parent Control")
    ExecuteContext=@RigVMFunction_Sequence.B
    Ball FK Settings=RigVMVariableNode() {variable=Ball FK Settings}.Value
    Foot BK Settings=RigVMVariableNode() {variable=Foot BK Settings}.Value
    Foot Roll Settings=RigVMVariableNode() {variable=Foot Roll Settings}.Value
    Heel IK Settings=RigVMVariableNode() {variable=Heel IK Settings}.Value
    Tip Settings=RigVMVariableNode() {variable=Tip SEttings}.Value
    Ball IK Settings=RigVMVariableNode() {variable=Ball IK Settings}.Value
   )
  ]}
 ContainedGraph {nodes [
   n1 = &RigVMFunction_ControlFlowBranch(ExecuteContext=(), Condition=RigVMVariableNode() {variable=IK Mode}.Value)
   /Script/ControlRig.RigUnit_SendEvent(Event=RequestAutoKey, bEnable=True, bOnlyDuringInteraction=True, ExecutePin=@n1.True, Item=RigVMVariableNode() {variable=IK Control}.Value) {is=struct RigVMUnitNode, template="SendEvent::Execute(in Event,in Item,in OffsetInSeconds,in bEnable,in bOnlyDuringInteraction)"}
   > &RigUnit_SendEvent(Event=RequestAutoKey, bEnable=True, bOnlyDuringInteraction=True, Item=RigVMVariableNode() {variable=FK Control}.Value) {template="SendEvent::Execute(in Event,in Item,in OffsetInSeconds,in bEnable,in bOnlyDuringInteraction)"}
   > RigVMVariableNode(Value=True) {variable=IK Mode}
   n2 = RigVMFunctionEntryNode()
   > RigVMVariableNode(Value=@n2.IKFK) {variable=IK Mode}
   > SequenceExecution = RigVMAggregateNode()
   n3 = RigVMVariableNode() {variable=IK Mode}
   /Script/ControlRig.RigUnit_SetControlVisibility(ExecutePin=@SequenceExecution.A, bVisible=@n3.Value, Item=RigVMVariableNode() {variable=IK Control}.Value) {is=struct RigVMUnitNode, template="SetControlVisibility::Execute(in Item,in Pattern,in bVisible)"}
   > &RigUnit_SetControlVisibility(bVisible=&RigVMFunction_MathBoolNot(Value=@n3.Value).Result, Item=RigVMVariableNode() {variable=FK Control}.Value) {template="SetControlVisibility::Execute(in Item,in Pattern,in bVisible)"}
   n4 = &RigVMFunction_ControlFlowBranch(ExecuteContext=@SequenceExecution.B, Condition=RigVMVariableNode() {variable=IK Mode}.Value)
   n5 = RigVMVariableNode() {variable=Bone}
   n6 = RigVMVariableNode() {variable=FK Control}
   &RigUnit_SetTransform(
    Weight=1
    ExecutePin=@n4.False
    Value=/Script/ControlRig.RigUnit_ProjectTransformToNewParent(bChildInitial=True, bOldParentInitial=True, Child=@n5.Value, NewParent=@n6.Value, OldParent=@n6.Value) {is=struct RigVMUnitNode}.Transform
    Item=@n5.Value
   )
   n7 = RigVMVariableNode() {variable=IK Control}
   n8 = RigVMVariableNode() {variable=Bone}
   > &RigUnit_SetTransform(
    Weight=1
    Value=&RigUnit_ProjectTransformToNewParent(bChildInitial=True, bOldParentInitial=True, Child=@n7.Value, NewParent=@n8.Value, OldParent=@n8.Value).Transform
    Item=@n7.Value
   )
   n9 = RigVMVariableNode() {variable=Bone}
   n10 = RigVMVariableNode() {variable=IK Control}
   SetTransform_13 = &RigUnit_SetRotation(
    Space=GlobalSpace
    Weight=1
    bInitial=False
    ExecutePin=@n4.True
    Value=&RigUnit_ProjectTransformToNewParent(bChildInitial=True, bOldParentInitial=True, Child=@n9.Value, NewParent=@n10.Value, OldParent=@n10.Value).Transform.Rotation
    Item=@n9.Value
   )
   n11 = &RigUnit_GetTransform(Item=@n9.Value)
   n12 = &RigUnit_GetTransform(Item=@n10.Value)
   Add = /Script/RigVM.RigVMFunction_MathVectorAdd(
    B=&RigUnit_GetTransform(Space=LocalSpace, Item=@n10.Value).Transform.Translation
    A=&RigVMFunction_MathTransformMakeRelative(Global=@n11.Transform, Parent=@n12.Transform).Local.Translation
   ) {is=struct RigVMUnitNode, template="Add::Execute(in A,in B,out Result)"}
   > &RigUnit_OffsetTransformForItem(
    OffsetTransform=(Rotation=(X=0,Y=0,Z=0,W=1),Translation=(X=0,Y=0,Z=0),Scale3D=(X=1,Y=1,Z=1))
    Weight=1
    bPropagateToChildren=True
    OffsetTransform.Translation=&RigVMFunction_MathTransformMakeRelative(
     Parent=@n11.Transform
     Global=/Script/RigVM.RigVMFunction_MathTransformMakeAbsolute(Local=(Rotation=(X=0,Y=0,Z=0,W=1),Translation=(X=0,Y=0,Z=0),Scale3D=(X=1,Y=1,Z=1)), Local.Translation=@Add.Result, Parent=@n12.Transform) {is=struct RigVMUnitNode, template="Make Absolute::Execute(in Local,in Parent,out Global)"}.Global
    ).Local.Translation
    Item=@n9.Value
   )
   n13 = RigVMVariableNode() {variable=Bone}
   n14 = RigVMVariableNode() {variable=FK Control}
   > n15 = &RigUnit_SetTransform(
    Weight=1
    Value=&RigUnit_ProjectTransformToNewParent(bChildInitial=True, bOldParentInitial=True, NewParent=@n13.Value, OldParent=@n13.Value, Child=@n14.Value).Transform
    Item=@n14.Value
   )
   n16 = &RigVMFunction_ControlFlowBranch(ExecuteContext=@n15.ExecutePin, Condition=RigVMVariableNode() {variable=IK Mode}.Value)
   &RigUnit_SendEvent(Event=RequestAutoKey, bEnable=True, bOnlyDuringInteraction=True, ExecutePin=@n16.False, Item=RigVMVariableNode() {variable=IK Control}.Value) {template="SendEvent::Execute(in Event,in Item,in OffsetInSeconds,in bEnable,in bOnlyDuringInteraction)"}
   > &RigUnit_SendEvent(Event=RequestAutoKey, bEnable=True, bOnlyDuringInteraction=True, Item=RigVMVariableNode() {variable=FK Control}.Value) {template="SendEvent::Execute(in Event,in Item,in OffsetInSeconds,in bEnable,in bOnlyDuringInteraction)"}
   > RigVMVariableNode(Value=True) {variable=IK Mode}
   MathBoolNotEquals_2_1 = RigVMDispatchNode(B=RigVMVariableNode() {variable=IK Mode}.Value, A=RigVMVariableNode() {variable=Previous IK Mode}.Value) {template="DISPATCH_RigVMDispatch_CoreNotEquals(in A,in B,out Result)"}
   n17 = &RigVMFunction_ControlFlowBranch(Condition=@MathBoolNotEquals_2_1.Result, ExecuteContext=@SequenceExecution.C)
   &RigUnit_SendEvent(Event=RequestAutoKey, bEnable=True, bOnlyDuringInteraction=True, ExecutePin=@n17.True, Item=RigVMVariableNode() {variable=IK Control}.Value) {template="SendEvent::Execute(in Event,in Item,in OffsetInSeconds,in bEnable,in bOnlyDuringInteraction)"}
   > &RigUnit_SendEvent(Event=RequestAutoKey, bEnable=True, bOnlyDuringInteraction=True, Item=RigVMVariableNode() {variable=FK Control}.Value) {template="SendEvent::Execute(in Event,in Item,in OffsetInSeconds,in bEnable,in bOnlyDuringInteraction)"}
   RigVMFunctionReturnNode(ExecuteContext=@SequenceExecution.D)
   /Script/RigVM.RigVMFunction_MathFloatGreater(
    B=0.5
    A=&RigUnit_GetTransform(Space=LocalSpace, Item=RigVMVariableNode() {variable=IKFK Switch Control}.Value).Transform.Translation.X
   ) {is=struct RigVMUnitNode, template="Greater::Execute(in A,in B,out Result)"}
   n18 = RigVMVariableNode() {variable=Toes Space}
   n19 = RigVMVariableNode() {variable=Bone}
   &RigUnit_SetTransform(
    ExecutePin=()
    Weight=1
    Value=&RigUnit_ProjectTransformToNewParent(bChildInitial=True, bOldParentInitial=True, Child=@n18.Value, NewParent=@n19.Value, OldParent=@n19.Value).Transform
    Item=@n18.Value
   )
  ]}
 ContainedGraph {nodes [
   n1 = RigVMFunctionEntryNode()
   SequenceExecution = &RigVMFunction_Sequence(ExecuteContext=@n1.ExecuteContext)
   SequenceExecution_1 = &RigVMFunction_Sequence(ExecuteContext=@SequenceExecution.B)
   SequenceExecution_2 = &RigVMFunction_Sequence(ExecuteContext=@SequenceExecution_1.B)
   RigVMFunctionReturnNode(A=@SequenceExecution.A, B=@SequenceExecution_1.A, C=@SequenceExecution_2.A, D=@SequenceExecution_2.B)
  ]}
 ContainedGraph {nodes [
   n1 = RigVMFunctionEntryNode()
   > RigVMVariableNode(Value=@n1.FK or IK) {variable=IK Mode}
   > SequenceExecution = RigVMAggregateNode()
   Is Right Side v2 = RigVMFunctionReferenceNode(Mirror Plane=@n1.Mirror Plane, Item=@n1.Roll Control, ExecuteContext=@SequenceExecution.A)
   > RigVMVariableNode(Value=@"Is Right Side v2".Right Side) {variable=Right Side}
   n2 = RigVMDispatchNode(ExecuteContext=@SequenceExecution.B, Array=RigVMVariableNode() {variable=BK Controls}.Value) {template="DISPATCH_RigVMDispatch_ArrayIterator(in Array,out Element,out Index,out Count,out Ratio)"}
   n3 = RigVMVariableNode() {variable=IK Mode}
   > SetControlVisibility = &RigUnit_SetControlVisibility(Item=@n2.Element, bVisible=@n3.Value) {template="SetControlVisibility::Execute(in Item,in Pattern,in bVisible)"}
   &RigUnit_SetControlVisibility(ExecutePin=@n2.Completed, bVisible=@n3.Value, Item=RigVMVariableNode() {variable=Roll Control}.Value) {template="SetControlVisibility::Execute(in Item,in Pattern,in bVisible)"}
   RigUnit_GetTransform = &RigUnit_GetTransform(Space=LocalSpace, Item=RigVMVariableNode() {variable=Roll Control}.Value)
   n4 = /Script/RigVM.RigVMFunction_MathFloatClamp(
    Maximum=180
    Minimum=0
    Value=/Script/RigVM.RigVMFunction_MathQuaternionToRotator(Value=@RigUnit_GetTransform.Transform.Rotation) {is=struct RigVMUnitNode, template="Cast::Execute(in Value,out Result)"}.Result.Yaw
   ) {is=struct RigVMUnitNode, template="Clamp::Execute(in Value,in Minimum,in Maximum,out Result)"}
   Subtract = &RigVMFunction_MathDoubleSub(A=1, B=@n4.Result)
   DISPATCH_RigVMDispatch_If = RigVMDispatchNode(False=@n4.Result, True=@Subtract.Result, Condition=RigVMVariableNode() {variable=Right Side}.Value) {template="DISPATCH_RigVMDispatch_If(in Condition,in True,in False,out Result)"}
   n5 = &RigUnit_GetTransform(
    Space=LocalSpace
    bInitial=false
    Item=RigVMDispatchNode(Index=0, Array=RigVMVariableNode() {variable=BK Controls}.Value) {template="DISPATCH_RigVMDispatch_ArrayGetAtIndex(in Array,in Index,out Element)"}.Element
   )
   n6 = &RigVMFunction_MathTransformMakeAbsolute(
    Local=&RigVMFunction_MathTransformMul(
     B=(Rotation=(X=0,Y=0,Z=0,W=1),Translation=(X=0,Y=0,Z=0),Scale3D=(X=1,Y=1,Z=1))
     A=&RigVMFunction_MathTransformMakeRelative(Global=(Rotation=(X=0,Y=0,Z=0,W=1),Translation=(X=0,Y=0,Z=0),Scale3D=(X=1,Y=1,Z=1)), Parent=@n5.Transform).Local
     B.Rotation=@n5.Transform.Rotation
    ) {template="Multiply::Execute(in A,in B,out Result)"}.Result
    Parent=@n5.Transform
   ) {template="Make Absolute::Execute(in Local,in Parent,out Global)"}
   n7 = &RigUnit_GetTransform(
    Space=LocalSpace
    bInitial=true
    Item=RigVMDispatchNode(Index=0, Array=RigVMVariableNode() {variable=BK Buffers}.Value) {template="DISPATCH_RigVMDispatch_ArrayGetAtIndex(in Array,in Index,out Element)"}.Element
   )
   RigUnit_SetTransform = &RigUnit_SetTransform(
    Space=LocalSpace
    Weight=1
    Item=RigVMDispatchNode(Index=0, Array=RigVMVariableNode() {variable=BK Buffers}.Value) {template="DISPATCH_RigVMDispatch_ArrayGetAtIndex(in Array,in Index,out Element)"}.Element
    Value=&RigVMFunction_MathTransformMul(
     A=(Rotation=(X=0,Y=0,Z=0,W=1),Translation=(X=0,Y=0,Z=0),Scale3D=(X=1,Y=1,Z=1))
     A.Rotation=/Script/RigVM.RigVMFunction_MathQuaternionFromRotatorV2(Value=(Pitch=0,Yaw=0,Roll=0), Value.Yaw=@DISPATCH_RigVMDispatch_If.Result) {is=struct RigVMUnitNode, template="Cast::Execute(in Value,out Result)"}.Result
     B=&RigVMFunction_MathTransformMul(A=@n6.Global, B=@n7.Transform) {template="Multiply::Execute(in A,in B,out Result)"}.Result
    ) {template="Multiply::Execute(in A,in B,out Result)"}.Result
    ExecutePin=@SequenceExecution.C
   )
   n8 = RigVMDispatchNode(
    Index=0
    Array=RigVMDispatchNode(Array=RigVMVariableNode() {variable=BK Buffers}.Value) {template="DISPATCH_RigVMDispatch_ArrayClone(in Array,out Clone)"}.Clone
    ExecuteContext=@SequenceExecution.D
   ) {template="DISPATCH_RigVMDispatch_ArrayRemove(io Array,in Index)"}
   > n9 = RigVMDispatchNode(Array=@n8.Array) {template="DISPATCH_RigVMDispatch_ArrayReverse(io Array)"}
   > n10 = RigVMDispatchNode(Array=@n9.Array) {template="DISPATCH_RigVMDispatch_ArrayIterator(in Array,out Element,out Index,out Count,out Ratio)"}
   n11 = &RigVMFunction_MathFloatMul(A=RigVMVariableNode() {variable=Roll Clamp}.Value, B=RigVMVariableNode() {variable=Roll Blend}.Value) {template="Multiply::Execute(in A,in B,out Result)"}
   MathIntEquals = RigVMDispatchNode(
    B=/Script/RigVM.RigVMFunction_MathIntSub(B=1, A=@n10.Count) {is=struct RigVMUnitNode, template="Subtract::Execute(in A,in B,out Result)"}.Result
    A=@n10.Index
   ) {template="DISPATCH_RigVMDispatch_CoreEquals(in A,in B,out Result)"}
   n12 = /Script/RigVM.RigVMFunction_MathIntToFloat(Value=@n10.Index) {is=struct RigVMUnitNode, template="Cast::Execute(in Value,out Result)"}
   n13 = &RigVMFunction_MathQuaternionToRotator(Value=&RigUnit_GetTransform(Space=LocalSpace, Item=RigVMVariableNode() {variable=Roll Control}.Value).Transform.Rotation) {template="Cast::Execute(in Value,out Result)"}
   n14 = &RigVMFunction_MathFloatClamp(
    Minimum=0
    Maximum=RigVMDispatchNode(True=180, False=@n11.Result, Condition=@MathIntEquals.Result) {template="DISPATCH_RigVMDispatch_If(in Condition,in True,in False,out Result)"}.Result
    Value=/Script/RigVM.RigVMFunction_MathFloatSub(
     B=&RigVMFunction_MathFloatMul(B=@n11.Result, A=@n12.Result) {template="Multiply::Execute(in A,in B,out Result)"}.Result
     A=/Script/RigVM.RigVMFunction_MathFloatNegate(Value=@n13.Result.Yaw) {is=struct RigVMUnitNode, template="Negate::Execute(in Value,out Result)"}.Result
    ) {is=struct RigVMUnitNode, template="Subtract::Execute(in A,in B,out Result)"}.Result
   ) {template="Clamp::Execute(in Value,in Minimum,in Maximum,out Result)"}
   n15 = RigVMDispatchNode(
    True=@n14.Result
    False=&RigVMFunction_MathFloatNegate(Value=@n14.Result) {template="Negate::Execute(in Value,out Result)"}.Result
    Condition=RigVMVariableNode() {variable=Right Side}.Value
   ) {template="DISPATCH_RigVMDispatch_If(in Condition,in True,in False,out Result)"}
   > &RigUnit_SetRotation(
    Weight=1
    Item=@n10.Element
    Value=&RigVMFunction_MathTransformMul(
     B=(Rotation=(X=0,Y=0,Z=0,W=1),Translation=(X=0,Y=0,Z=0),Scale3D=(X=1,Y=1,Z=1))
     B.Rotation=&RigVMFunction_MathQuaternionFromRotatorV2(Value=(Pitch=0,Yaw=0,Roll=0), Value.Yaw=@n15.Result) {template="Cast::Execute(in Value,out Result)"}.Result
     A=&RigUnit_GetTransform(Space=LocalSpace, bInitial=true, Item=@n10.Element).Transform
    ) {template="Multiply::Execute(in A,in B,out Result)"}.Result.Rotation
   )
   SequenceExecution_1 = RigVMAggregateNode(ExecuteContext=@SequenceExecution.E)
   n16 = RigVMDispatchNode(Index=1, Array=RigVMVariableNode() {variable=BK Buffers}.Value) {template="DISPATCH_RigVMDispatch_ArrayGetAtIndex(in Array,in Index,out Element)"}
   n17 = &RigUnit_GetTransform(
    Space=LocalSpace
    bInitial=false
    Item=RigVMDispatchNode(Index=1, Array=RigVMVariableNode() {variable=BK Controls}.Value) {template="DISPATCH_RigVMDispatch_ArrayGetAtIndex(in Array,in Index,out Element)"}.Element
   )
   n18 = &RigVMFunction_MathTransformMakeAbsolute(
    Local=&RigVMFunction_MathTransformMul(
     B=(Rotation=(X=0,Y=0,Z=0,W=1),Translation=(X=0,Y=0,Z=0),Scale3D=(X=1,Y=1,Z=1))
     A=&RigVMFunction_MathTransformMakeRelative(Global=(Rotation=(X=0,Y=0,Z=0,W=1),Translation=(X=0,Y=0,Z=0),Scale3D=(X=1,Y=1,Z=1)), Parent=@n17.Transform).Local
     B.Rotation=@n17.Transform.Rotation
    ) {template="Multiply::Execute(in A,in B,out Result)"}.Result
    Parent=@n17.Transform
   ) {template="Make Absolute::Execute(in Local,in Parent,out Global)"}
   n19 = &RigUnit_GetTransform(Space=LocalSpace, bInitial=true, Item=@n16.Element)
   &RigUnit_SetTransform(
    Space=LocalSpace
    Weight=1
    Item=@n16.Element
    Value=&RigVMFunction_MathTransformMul(
     A=(Rotation=(X=0,Y=0,Z=0,W=1),Translation=(X=0,Y=0,Z=0),Scale3D=(X=1,Y=1,Z=1))
     A.Rotation=&RigUnit_GetTransform(Space=LocalSpace, Item=@n16.Element).Transform.Rotation
     B=&RigVMFunction_MathTransformMul(A=@n18.Global, B=@n19.Transform) {template="Multiply::Execute(in A,in B,out Result)"}.Result
    ) {template="Multiply::Execute(in A,in B,out Result)"}.Result
    ExecutePin=@SequenceExecution_1.A
   )
   n20 = /Script/ControlRig.RigUnit_HierarchyGetParent(
    bDefaultParent=True
    Child=RigVMDispatchNode(Index=0, Array=RigVMVariableNode() {variable=BK Buffers}.Value) {template="DISPATCH_RigVMDispatch_ArrayGetAtIndex(in Array,in Index,out Element)"}.Element
   ) {is=struct RigVMUnitNode}
   n21 = RigVMVariableNode() {variable=Roll Buffer}
   RigUnit_SetRotation = &RigUnit_SetRotation(
    Space=GlobalSpace
    Weight=1
    Value=&RigUnit_ProjectTransformToNewParent(bChildInitial=True, bOldParentInitial=True, NewParent=@n20.Parent, OldParent=@n20.Parent, Child=@n21.Value).Transform.Rotation
    ExecutePin=@SequenceExecution_1.B
    Item=@n21.Value
   )
   RigVMFunctionReturnNode(ExecuteContext=@SequenceExecution_1.D)
   &RigVMFunction_MathFloatGreater(
    B=0.5
    A=&RigUnit_GetTransform(Space=LocalSpace, Item=RigVMVariableNode() {variable=IKFK Switch Control}.Value).Transform.Translation.X
   ) {template="Greater::Execute(in A,in B,out Result)"}
  ]}
 ContainedGraph {nodes [
   n1 = RigVMFunctionEntryNode()
   SequenceExecution = &RigVMFunction_Sequence(ExecuteContext=@n1.ExecuteContext)
   SequenceExecution_1 = &RigVMFunction_Sequence(ExecuteContext=@SequenceExecution.B)
   SequenceExecution_2 = &RigVMFunction_Sequence(ExecuteContext=@SequenceExecution_1.B)
   SequenceExecution_3 = &RigVMFunction_Sequence(ExecuteContext=@SequenceExecution_2.B)
   RigVMFunctionReturnNode(A=@SequenceExecution.A, B=@SequenceExecution_1.A, C=@SequenceExecution_2.A, D=@SequenceExecution_3.A, E=@SequenceExecution_3.B)
  ]}
 ContainedGraph {nodes [
   n1 = RigVMFunctionEntryNode()
   SequenceExecution_1 = &RigVMFunction_Sequence(ExecuteContext=@n1.ExecuteContext)
   SequenceExecution_1_1 = &RigVMFunction_Sequence(ExecuteContext=@SequenceExecution_1.B)
   SequenceExecution_1_2 = &RigVMFunction_Sequence(ExecuteContext=@SequenceExecution_1_1.B)
   RigVMFunctionReturnNode(A=@SequenceExecution_1.A, B=@SequenceExecution_1_1.A, C=@SequenceExecution_1_2.A, D=@SequenceExecution_1_2.B)
  ]}
 ContainedGraph {nodes [
   n1 = RigVMFunctionEntryNode()
   n2 = &RigVMFunction_MathQuaternionToRotator(Value=@n1.Input.Rotation) {template="Cast::Execute(in Value,out Result)"}
   RerouteNode = RigVMDispatchNode(
    Elements=(Rotation=(X=0,Y=0,Z=0,W=1),Translation=(X=0,Y=0,Z=0),Scale3D=(X=-1,Y=1,Z=1))
    Elements.Scale3D.Y=@n1.Input.Scale3D.Y
    Elements.Scale3D.Z=@n1.Input.Scale3D.Z
    Elements.Translation=@n1.Input.Translation
    Elements.Scale3D.X=&RigVMFunction_MathFloatNegate(Value=@n1.Input.Scale3D.X) {template="Negate::Execute(in Value,out Result)"}.Result
    Elements.Rotation=&RigVMFunction_MathQuaternionFromRotatorV2(
     Value=(Pitch=0,Yaw=0,Roll=0)
     Value.Roll=&RigVMFunction_MathFloatAdd(B=180, A=@n2.Result.Roll) {template="Add::Execute(in A,in B,out Result)"}.Result
     Value.Pitch=@n2.Result.Pitch
     Value.Yaw=@n2.Result.Yaw
    ) {template="Cast::Execute(in Value,out Result)"}.Result
   ) {template="DISPATCH_RigVMDispatch_MakeStruct(in Elements,out Struct)"}
   > RigVMFunctionReturnNode(Output=@RerouteNode.Struct)
  ]}
 ContainedGraph {nodes [
   RigVMFunctionEntryNode()
   > SequenceExecution = RigVMAggregateNode()
   n1 = RigVMDispatchNode(ExecuteContext=@SequenceExecution.A, Array=RigVMVariableNode() {variable=Bones}.Value) {template="DISPATCH_RigVMDispatch_ArrayIterator(in Array,out Element,out Index,out Count,out Ratio)"}
   n2 = &RigUnit_GetTransform(Item=RigVMDispatchNode(Index=0, Array=RigVMVariableNode() {variable=Bones}.Value) {template="DISPATCH_RigVMDispatch_ArrayGetAtIndex(in Array,in Index,out Element)"}.Element)
   n3 = &RigUnit_GetTransform(Item=@n1.Element)
   Get Negative Transform_1 = RigVMFunctionReferenceNode(
    Input=(Rotation=(X=0,Y=0,Z=0,W=1),Translation=(X=0,Y=0,Z=0),Scale3D=(X=0,Y=0,Z=0))
    Input.Scale3D=@n3.Transform.Scale3D
    Input.Translation=@n3.Transform.Translation
   )
   n4 = RigVMDispatchNode(
    False=(Rotation=(X=0,Y=0,Z=0,W=1),Translation=(X=0,Y=0,Z=0),Scale3D=(X=0,Y=0,Z=0))
    True=@"Get Negative Transform_1".Output
    False.Scale3D=@n3.Transform.Scale3D
    False.Translation=@n3.Transform.Translation
    False.Rotation=@n2.Transform.Rotation
    Condition=RigVMVariableNode() {variable=Negative Side}.Value
   ) {template="DISPATCH_RigVMDispatch_If(in Condition,in True,in False,out Result)"}
   > &RigUnit_SetTransform(
    Value=(Rotation=(X=0,Y=0,Z=0,W=1),Translation=(X=0,Y=0,Z=0),Scale3D=(X=1,Y=1,Z=1))
    Weight=1
    bInitial=true
    Item=RigVMDispatchNode(Index=@n1.Index, Array=RigVMVariableNode() {variable=BK Buffers}.Value) {template="DISPATCH_RigVMDispatch_ArrayGetAtIndex(in Array,in Index,out Element)"}.Element
    Value.Rotation=@n2.Transform.Rotation
    Value.Scale3D=@n4.Result.Scale3D
    Value.Translation=@n4.Result.Translation
   )
   > &RigUnit_SetTransform(
    Value=(Rotation=(X=0,Y=0,Z=0,W=1),Translation=(X=0,Y=0,Z=0),Scale3D=(X=1,Y=1,Z=1))
    Weight=1
    bInitial=true
    Item=RigVMDispatchNode(Index=@n1.Index, Array=RigVMVariableNode() {variable=BK Controls}.Value) {template="DISPATCH_RigVMDispatch_ArrayGetAtIndex(in Array,in Index,out Element)"}.Element
    Value.Rotation=@n2.Transform.Rotation
    Value.Scale3D=@n4.Result.Scale3D
    Value.Translation=@n4.Result.Translation
   )
   n5 = RigVMVariableNode() {variable=Bones}
   n6 = RigVMDispatchNode(Array=@n5.Value) {template="DISPATCH_RigVMDispatch_ArrayGetNum(in Array,out Num)"}
   n7 = &RigUnit_GetTransform(Item=RigVMDispatchNode(Index=&RigVMFunction_MathIntSub(B=1, A=@n6.Num) {template="Subtract::Execute(in A,in B,out Result)"}.Result, Array=@n5.Value) {template="DISPATCH_RigVMDispatch_ArrayGetAtIndex(in Array,in Index,out Element)"}.Element)
   n8 = RigVMDispatchNode(False=(X=1,Y=1,Z=1), True=(X=-1,Y=1,Z=1), Condition=RigVMVariableNode() {variable=Negative Side}.Value) {template="DISPATCH_RigVMDispatch_If(in Condition,in True,in False,out Result)"}
   &RigUnit_SetTransform(
    Value=(Rotation=(X=0,Y=0,Z=0,W=1),Translation=(X=0,Y=0,Z=0),Scale3D=(X=1,Y=1,Z=1))
    Weight=1
    bInitial=true
    Value.Translation=@n7.Transform.Translation
    Value.Rotation=@n2.Transform.Rotation
    Value.Scale3D=@n8.Result
    ExecutePin=@SequenceExecution.B
    Item=RigVMVariableNode() {variable=Roll Buffer}.Value
   )
   > &RigUnit_SetTransform(
    Value=(Rotation=(X=0,Y=0,Z=0,W=1),Translation=(X=0,Y=0,Z=0),Scale3D=(X=1,Y=1,Z=1))
    Weight=1
    bInitial=true
    Value.Translation=@n7.Transform.Translation
    Value.Rotation=@n2.Transform.Rotation
    Value.Scale3D=@n8.Result
    Item=RigVMVariableNode() {variable=Roll Control}.Value
   )
   RigVMFunctionReturnNode(ExecuteContext=@SequenceExecution.D)
  ]}
 ContainedGraph {nodes [
   n1 = RigVMFunctionEntryNode()
   SequenceExecution = &RigVMFunction_Sequence(ExecuteContext=@n1.ExecuteContext)
   SequenceExecution_1 = &RigVMFunction_Sequence(ExecuteContext=@SequenceExecution.B)
   SequenceExecution_2 = &RigVMFunction_Sequence(ExecuteContext=@SequenceExecution_1.B)
   RigVMFunctionReturnNode(A=@SequenceExecution.A, B=@SequenceExecution_1.A, C=@SequenceExecution_2.A, D=@SequenceExecution_2.B)
  ]}
 ContainedGraph {nodes [
   n1 = RigVMFunctionEntryNode()
   RigUnit_SequenceAggregate_1 = &RigVMFunction_Sequence(ExecuteContext=@n1.ExecuteContext)
   RigUnit_SequenceAggregate = RigVMAggregateNode(ExecuteContext=@RigUnit_SequenceAggregate_1.A)
   Greater = /Script/RigVM.RigVMFunction_MathDoubleGreater(B=0, A=@n1.Mirror Plane.X) {is=struct RigVMUnitNode, template="Greater::Execute(in A,in B,out Result)"}
   BranchNode_4 = &RigVMFunction_ControlFlowBranch(Condition=@Greater.Result, ExecuteContext=@RigUnit_SequenceAggregate.A)
   n2 = &RigVMFunction_MathDoubleGreater(B=0, A=@n1.Mirror Plane.Y) {template="Greater::Execute(in A,in B,out Result)"}
   BranchNode_1_1 = &RigVMFunction_ControlFlowBranch(ExecuteContext=@BranchNode_4.Completed, Condition=@n2.Result)
   n3 = &RigVMFunction_MathDoubleGreater(B=0, A=@n1.Mirror Plane.Z) {template="Greater::Execute(in A,in B,out Result)"}
   BranchNode_2_1 = &RigVMFunction_ControlFlowBranch(ExecuteContext=@BranchNode_1_1.Completed, Condition=@n3.Result)
   Cast = &RigVMFunction_MathBoolToFloat(Value=@Greater.Result)
   Dot__ = /Script/RigVM.RigVMFunction_MathVectorDot(
    B=(X=1,Y=1,Z=0)
    B.X=@Cast.Result
    B.Y=&RigVMFunction_MathBoolToFloat(Value=@n2.Result).Result
    B.Z=&RigVMFunction_MathBoolToFloat(Value=@n3.Result).Result
    A=RigVMVariableNode() {variable=vector pos}.Value
   ) {is=struct RigVMUnitNode, template="Dot,|::Execute(in A,in B,out Result)"}
   Less = /Script/RigVM.RigVMFunction_MathDoubleLess(B=0, A=@Dot__.Result) {is=struct RigVMUnitNode, template="Less::Execute(in A,in B,out Result)"}
   RigVMVariableNode(ExecuteContext=@BranchNode_2_1.Completed, Value=@Less.Result) {variable=is on right}
   n4 = RigVMVariableNode() {variable=vector pos}
   RigVMVariableNode(
    Value=(X=0,Y=0,Z=0)
    ExecuteContext=@BranchNode_2_1.True
    Value.Z=&RigUnit_GetTransform(bInitial=true, Item=@n1.Item).Transform.Translation.Z
    Value.X=@n4.Value.X
    Value.Y=@n4.Value.Y
   ) {variable=vector pos}
   Dot___1_1_1_1_1_2 = &RigVMFunction_MathVectorDot(
    B=(X=1,Y=1,Z=0)
    B.Y=&RigVMFunction_MathBoolToFloat(Value=@n2.Result).Result
    B.Z=&RigVMFunction_MathBoolToFloat(Value=@n3.Result).Result
    B.X=&RigVMFunction_MathBoolToFloat(Value=@Greater.Result).Result
    A=RigVMVariableNode() {variable=vector pos}.Value
   ) {template="Dot,|::Execute(in A,in B,out Result)"}
   > RigVMVariableNode(Value=&RigVMFunction_MathDoubleLess(B=0, A=@Dot___1_1_1_1_1_2.Result) {template="Less::Execute(in A,in B,out Result)"}.Result) {variable=is on right}
   n5 = RigVMVariableNode() {variable=vector pos}
   RigVMVariableNode(
    Value=(X=0,Y=0,Z=0)
    ExecuteContext=@BranchNode_1_1.True
    Value.Y=&RigUnit_GetTransform(bInitial=true, Item=@n1.Item).Transform.Translation.Y
    Value.X=@n5.Value.X
    Value.Z=@n5.Value.Z
   ) {variable=vector pos}
   Dot___1_1_1_1_1_1 = &RigVMFunction_MathVectorDot(
    B=(X=1,Y=1,Z=0)
    B.Y=&RigVMFunction_MathBoolToFloat(Value=@n2.Result).Result
    B.Z=&RigVMFunction_MathBoolToFloat(Value=@n3.Result).Result
    B.X=&RigVMFunction_MathBoolToFloat(Value=@Greater.Result).Result
    A=RigVMVariableNode() {variable=vector pos}.Value
   ) {template="Dot,|::Execute(in A,in B,out Result)"}
   > RigVMVariableNode(Value=&RigVMFunction_MathDoubleLess(B=0, A=@Dot___1_1_1_1_1_1.Result) {template="Less::Execute(in A,in B,out Result)"}.Result) {variable=is on right}
   n6 = RigVMVariableNode() {variable=vector pos}
   RigVMVariableNode(
    Value=(X=0,Y=0,Z=0)
    ExecuteContext=@BranchNode_4.True
    Value.X=&RigUnit_GetTransform(bInitial=true, Item=@n1.Item).Transform.Translation.X
    Value.Y=@n6.Value.Y
    Value.Z=@n6.Value.Z
   ) {variable=vector pos}
   Dot___1_1_1_1_1 = &RigVMFunction_MathVectorDot(
    B=(X=1,Y=1,Z=0)
    B.Y=&RigVMFunction_MathBoolToFloat(Value=@n2.Result).Result
    B.Z=&RigVMFunction_MathBoolToFloat(Value=@n3.Result).Result
    B.X=&RigVMFunction_MathBoolToFloat(Value=@Greater.Result).Result
    A=RigVMVariableNode() {variable=vector pos}.Value
   ) {template="Dot,|::Execute(in A,in B,out Result)"}
   > RigVMVariableNode(Value=&RigVMFunction_MathDoubleLess(B=0, A=@Dot___1_1_1_1_1.Result) {template="Less::Execute(in A,in B,out Result)"}.Result) {variable=is on right}
   n7 = &RigVMFunction_MathDoubleLess(B=0, A=@n1.Mirror Plane.X) {template="Less::Execute(in A,in B,out Result)"}
   BranchNode_4_1 = &RigVMFunction_ControlFlowBranch(Condition=@n7.Result, ExecuteContext=@RigUnit_SequenceAggregate.D)
   n8 = RigVMVariableNode() {variable=vector pos}
   RigVMVariableNode(
    Value=(X=0,Y=0,Z=0)
    ExecuteContext=@BranchNode_4_1.True
    Value.X=&RigUnit_GetTransform(bInitial=true, Item=@n1.Item).Transform.Translation.X
    Value.Y=@n8.Value.Y
    Value.Z=@n8.Value.Z
   ) {variable=vector pos}
   n9 = &RigVMFunction_MathDoubleLess(B=0, A=@n1.Mirror Plane.Y) {template="Less::Execute(in A,in B,out Result)"}
   n10 = &RigVMFunction_MathDoubleLess(B=0, A=@n1.Mirror Plane.Z) {template="Less::Execute(in A,in B,out Result)"}
   Dot___1_1_1_1 = &RigVMFunction_MathVectorDot(
    B=(X=1,Y=1,Z=0)
    B.Y=&RigVMFunction_MathBoolToFloat(Value=@n9.Result).Result
    B.Z=&RigVMFunction_MathBoolToFloat(Value=@n10.Result).Result
    B.X=&RigVMFunction_MathBoolToFloat(Value=@n7.Result).Result
    A=RigVMVariableNode() {variable=vector pos}.Value
   ) {template="Dot,|::Execute(in A,in B,out Result)"}
   > RigVMVariableNode(Value=&RigVMFunction_MathDoubleGreater(B=0, A=@Dot___1_1_1_1.Result) {template="Greater::Execute(in A,in B,out Result)"}.Result) {variable=is on right}
   BranchNode_1_1_1 = &RigVMFunction_ControlFlowBranch(Condition=@n9.Result, ExecuteContext=@RigUnit_SequenceAggregate.E)
   n11 = RigVMVariableNode() {variable=vector pos}
   RigVMVariableNode(
    Value=(X=0,Y=0,Z=0)
    ExecuteContext=@BranchNode_1_1_1.True
    Value.Y=&RigUnit_GetTransform(bInitial=true, Item=@n1.Item).Transform.Translation.Y
    Value.X=@n11.Value.X
    Value.Z=@n11.Value.Z
   ) {variable=vector pos}
   Dot___1_1 = &RigVMFunction_MathVectorDot(
    B=(X=1,Y=1,Z=0)
    B.Y=&RigVMFunction_MathBoolToFloat(Value=@n9.Result).Result
    B.Z=&RigVMFunction_MathBoolToFloat(Value=@n10.Result).Result
    B.X=&RigVMFunction_MathBoolToFloat(Value=@n7.Result).Result
    A=RigVMVariableNode() {variable=vector pos}.Value
   ) {template="Dot,|::Execute(in A,in B,out Result)"}
   > RigVMVariableNode(Value=&RigVMFunction_MathDoubleGreater(B=0, A=@Dot___1_1.Result) {template="Greater::Execute(in A,in B,out Result)"}.Result) {variable=is on right}
   BranchNode_2_1_1 = &RigVMFunction_ControlFlowBranch(Condition=@n10.Result, ExecuteContext=@RigUnit_SequenceAggregate.F)
   n12 = RigVMVariableNode() {variable=vector pos}
   RigVMVariableNode(
    Value=(X=0,Y=0,Z=0)
    ExecuteContext=@BranchNode_2_1_1.True
    Value.Z=&RigUnit_GetTransform(bInitial=true, Item=@n1.Item).Transform.Translation.Z
    Value.X=@n12.Value.X
    Value.Y=@n12.Value.Y
   ) {variable=vector pos}
   Dot___1_1_1 = &RigVMFunction_MathVectorDot(
    B=(X=1,Y=1,Z=0)
    B.Y=&RigVMFunction_MathBoolToFloat(Value=@n9.Result).Result
    B.Z=&RigVMFunction_MathBoolToFloat(Value=@n10.Result).Result
    B.X=&RigVMFunction_MathBoolToFloat(Value=@n7.Result).Result
    A=RigVMVariableNode() {variable=vector pos}.Value
   ) {template="Dot,|::Execute(in A,in B,out Result)"}
   > RigVMVariableNode(Value=&RigVMFunction_MathDoubleGreater(B=0, A=@Dot___1_1_1.Result) {template="Greater::Execute(in A,in B,out Result)"}.Result) {variable=is on right}
   RigVMFunctionReturnNode(ExecuteContext=@RigUnit_SequenceAggregate_1.B, Right Side=RigVMVariableNode() {variable=is on right}.Value)
   &RigVMFunction_MathDoubleLess(B=0, A=@Dot___1_1.Result) {template="Less::Execute(in A,in B,out Result)"}
   &RigVMFunction_MathDoubleLess(B=0, A=@Dot___1_1_1.Result) {template="Less::Execute(in A,in B,out Result)"}
   &RigVMFunction_MathDoubleLess(B=0, A=@Dot___1_1_1_1.Result) {template="Less::Execute(in A,in B,out Result)"}
  ]}
 ContainedGraph {nodes [
   n1 = RigVMFunctionEntryNode()
   RigUnit_SequenceAggregate = &RigVMFunction_Sequence(ExecuteContext=@n1.ExecuteContext)
   RigUnit_SequenceAggregate_1 = &RigVMFunction_Sequence(ExecuteContext=@RigUnit_SequenceAggregate.B)
   RigUnit_SequenceAggregate_2 = &RigVMFunction_Sequence(ExecuteContext=@RigUnit_SequenceAggregate_1.B)
   RigUnit_SequenceAggregate_3 = &RigVMFunction_Sequence(ExecuteContext=@RigUnit_SequenceAggregate_2.B)
   RigUnit_SequenceAggregate_4 = &RigVMFunction_Sequence(ExecuteContext=@RigUnit_SequenceAggregate_3.B)
   RigUnit_SequenceAggregate_5 = &RigVMFunction_Sequence(ExecuteContext=@RigUnit_SequenceAggregate_4.B)
   RigVMFunctionReturnNode(
    A=@RigUnit_SequenceAggregate.A
    B=@RigUnit_SequenceAggregate_1.A
    C=@RigUnit_SequenceAggregate_2.A
    D=@RigUnit_SequenceAggregate_3.A
    E=@RigUnit_SequenceAggregate_4.A
    F=@RigUnit_SequenceAggregate_5.A
    G=@RigUnit_SequenceAggregate_5.B
   )
  ]}
 New Function_ContainedGraph {nodes [
   n1 = RigVMFunctionEntryNode()
   n2 = &RigVMFunction_Sequence(ExecuteContext=@n1.ExecuteContext)
   RigVMFunction_Sequence = RigVMAggregateNode(ExecuteContext=@n2.A)
   Concat_1_1_2_1_2 = RigVMAggregateNode(B=_, C=ball, A=RigVMVariableNode() {variable=prefix}.Value)
   Control Stack at Item_2_1 = RigVMFunctionReferenceNode(
    Add Bottom Null=False
    Add Null=False
    Add Secondary Control=False
    Control Suffix=_fk_ctrl
    Null Suffix=_null
    Secondary Settings="(DisplayName=\"None\",Shape=(bVisible=True,Name=\"Default\",Color=(R=1,G=0,B=0,A=1),Transform=(Rotation=(X=0,Y=0,Z=0,W=1),Translation=(X=0,Y=0,Z=0),Scale3D=(X=1,Y=1,Z=1))),Proxy=(bIsProxy=False,ShapeVisibility=BasedOnSelection),InitialSpace=LocalSpace,bUsePref ...(truncated)"
    Tag Controls=true
    Zero Controls=true
    Name=@Concat_1_1_2_1_2.Result
    Item=/Script/ControlRig.RigUnit_ResolveConnector(SkipSocket=true, Connector=RigVMVariableNode() {variable=Ball Bone Connector}.Value) {is=struct RigVMUnitNode}.Result
    ExecuteContext=@RigVMFunction_Sequence.E
    Settings=RigVMVariableNode() {variable=Ball FK Settings}.Value
    Parent=RigVMVariableNode() {variable=FK Parent}.Value
   )
   > RigVMVariableNode(Value=@"Control Stack at Item_2_1".Control) {variable=ball fk control}
   > RigVMDispatchNode(Name=Ball FK Control, NameSpace=Self, Value=@"Control Stack at Item_2_1".Control) {template="DISPATCH_RigDispatch_SetModuleMetadata(in Name,in NameSpace,in Value,out Success)"}
   n3 = &RigVMFunction_Sequence(ExecuteContext=@RigVMFunction_Sequence.H)
   n4 = RigVMVariableNode() {variable=Effector Null}
   RigVMFunction_ControlFlowBranch = &RigVMFunction_ControlFlowBranch(Condition=/Script/ControlRig.RigUnit_ItemExists(Item=@n4.Value) {is=struct RigVMUnitNode}.Exists, ExecuteContext=@n3.A)
   /Script/ControlRig.RigUnit_SetDefaultParent(ExecutePin=@RigVMFunction_ControlFlowBranch.True, Parent=RigVMVariableNode() {variable=foot ik null}.Value, Child=@n4.Value) {is=struct RigVMUnitNode}
   Finalize Controls = RigVMCollapseNode(ExecuteContext=@RigVMFunction_Sequence.I, Module Settings=RigVMVariableNode() {variable=Module Settings}.Value)
   n5 = &RigVMFunction_ControlFlowBranch(ExecuteContext=@RigVMFunction_Sequence.J, Condition=RigVMVariableNode() {variable=Foot Control Metadata Found}.Value)
   Set Offset Transform from Current_0 = RigVMFunctionReferenceNode(ExecuteContext=@n5.True, Control=RigVMVariableNode() {variable=Foot Control}.Value)
   n6 = &RigVMFunction_ControlFlowBranch(Condition=RigVMVariableNode() {variable=Foot Control Metadata Found}.Value, ExecuteContext=@RigVMFunction_Sequence.C)
   RigVMVariableNode(Value=&RigUnit_ResolveConnector(SkipSocket=true, Connector=@n1.Parent Control Connector).Result, ExecuteContext=@n6.False) {variable=foot control parent}
   RigVMVariableNode(ExecuteContext=@n6.True, Value=RigVMVariableNode() {variable=Foot Control}.Value) {variable=foot control parent}
   RigVMVariableNode(
    Value=&RigUnit_ResolveConnector(Connector=(Type=Connector,Name="Foot Bone"), SkipSocket=true).Result.Name
    ExecuteContext=@RigVMFunction_Sequence.B
   ) {variable=prefix}
   > RigVMVariableNode(Value=&RigUnit_ResolveConnector(Connector=(Type=Connector,Name="Foot Bone"), SkipSocket=False).Result) {variable=foot bone}
   > RigVMVariableNode(Value=&RigUnit_ResolveConnector(Connector=(Type=Connector,Name="Ball Bone"), SkipSocket=False).Result) {variable=ball bone}
   Concat = RigVMAggregateNode(B=_, C=heel, A=RigVMVariableNode() {variable=prefix}.Value)
   n7 = &RigUnit_GetTransform(bInitial=true, Item=&RigUnit_ResolveConnector(Connector=(Type=Connector,Name="Foot Bone"), SkipSocket=true).Result)
   > n8 = /Script/ControlRig.RigUnit_HierarchyAddBone(Parent=(Type=Bone,Name="None"), Space=LocalSpace, Name=@Concat.Result, Transform=@n7.Transform) {is=struct RigVMUnitNode}
   > RigVMVariableNode(Value=@n8.Item) {variable=heel bone}
   > Set Transform_1 = &RigUnit_SetRotation(Value=(X=-0.5,Y=0.5,Z=-0.5,W=0.5), Weight=1, bInitial=true, Item=@n8.Item)
   > Set Transform = /Script/ControlRig.RigUnit_SetTranslation(
    Space=GlobalSpace
    Value=(X=0,Y=0,Z=0)
    Weight=1
    bInitial=true
    bPropagateToChildren=True
    Value.X=@n7.Transform.Translation.X
    Value.Y=@n7.Transform.Translation.Y
    Value.Z=&RigUnit_GetTransform(bInitial=true, Item=&RigUnit_ResolveConnector(Connector=(Type=Connector,Name="Ball Bone"), SkipSocket=true).Result).Transform.Translation.Z
    Item=@n8.Item
   ) {is=struct RigVMUnitNode, template="Set Transform::Execute(in Item,in Space,in bInitial,in Value,in Weight,in bPropagateToChildren)"}
   Concat_1 = RigVMAggregateNode(B=_, C=tip, A=RigVMVariableNode() {variable=prefix}.Value)
   n9 = &RigUnit_ResolveConnector(Connector=(Type=Connector,Name="Ball Bone"), SkipSocket=true)
   > n10 = &RigUnit_HierarchyAddBone(
    Parent=(Type=Bone,Name="None")
    Space=LocalSpace
    Name=@Concat_1.Result
    Transform=&RigUnit_GetTransform(bInitial=False, Item=@n9.Result).Transform
   )
   > RigVMVariableNode(Value=@n10.Item) {variable=tip bone}
   n11 = /Script/ControlRig.RigUnit_AimBoneMath(
    DebugSettings=(bEnabled=False,Scale=10,WorldOffset=(Rotation=(X=0,Y=0,Z=0,W=1),Translation=(X=0,Y=0,Z=0),Scale3D=(X=1,Y=1,Z=1)))
    Primary=(Weight=1,Axis=(X=1,Y=0,Z=0),Target=(X=1,Y=0,Z=0),Kind=Location,Space=(Type=Bone,Name="None"))
    Secondary=(Weight=1,Axis=(X=0,Y=0,Z=-1),Target=(X=1,Y=0,Z=0),Kind=Location,Space=(Type=Bone,Name="None"))
    Weight=1
    InputTransform=&RigUnit_GetTransform(bInitial=true, Item=@n9.Result).Transform
    Primary.Target=&RigUnit_GetTransform(bInitial=true, Item=RigVMVariableNode() {variable=heel bone}.Value).Transform.Translation
   ) {is=struct RigVMUnitNode}
   > &RigUnit_SetTransform(Space=LocalSpace, Weight=1, bInitial=true, Value=@n11.Result, Item=@n10.Item)
   > Set Transform_3_1 = &RigUnit_SetRotation(
    Weight=1
    bInitial=true
    Value=&RigUnit_GetTransform(bInitial=False, Item=RigVMVariableNode() {variable=tip bone}.Value).Transform.Rotation
    Item=RigVMVariableNode() {variable=heel bone}.Value
   )
   > &RigUnit_SetDefaultParent(Child=@n10.Item, Parent=@n8.Item)
   > &RigUnit_SetDefaultParent(
    Child=@n8.Item
    Parent=&RigUnit_HierarchyGetParent(
     bDefaultParent=True
     Child=&RigUnit_ResolveConnector(Connector=(Type=Connector,Name="Foot Primary"), SkipSocket=true).Result
    ).Parent
   )
   Concat_1_1 = RigVMAggregateNode(B=_, C=ball_ik, A=RigVMVariableNode() {variable=prefix}.Value)
   Control Stack at Item = RigVMFunctionReferenceNode(
    Add Bottom Null=False
    Add Null=False
    Add Secondary Control=False
    Control Suffix=_ctrl
    Null Suffix=_null
    Orient to World=true
    Secondary Settings="(DisplayName=\"None\",Shape=(bVisible=True,Name=\"Default\",Color=(R=1,G=0,B=0,A=1),Transform=(Rotation=(X=0,Y=0,Z=0,W=1),Translation=(X=0,Y=0,Z=0),Scale3D=(X=1,Y=1,Z=1))),Proxy=(bIsProxy=False,ShapeVisibility=BasedOnSelection),InitialSpace=LocalSpace,bUsePref ...(truncated)"
    Tag Controls=true
    Zero Controls=true
    Name=@Concat_1_1.Result
    Item=&RigUnit_ResolveConnector(SkipSocket=true, Connector=RigVMVariableNode() {variable=Ball Bone Connector}.Value).Result
    Parent=RigVMVariableNode() {variable=foot control parent}.Value
    Settings=RigVMVariableNode() {variable=Ball IK Settings}.Value
    ExecuteContext=@RigVMFunction_Sequence.D
   )
   > DISPATCH_RigDispatch_SetModuleMetadata = RigVMDispatchNode(Name=Ball IK Control, NameSpace=Self, Value=@"Control Stack at Item".Control) {template="DISPATCH_RigDispatch_SetModuleMetadata(in Name,in NameSpace,in Value,out Success)"}
   > RigVMVariableNode(Value=@"Control Stack at Item".Control) {variable=ball ik control}
   Concat_1_1_1 = RigVMAggregateNode(B=_, C=bk, A=RigVMVariableNode() {variable=prefix}.Value)
   > Control Stack at Item_1 = RigVMFunctionReferenceNode(
    Add Bottom Null=False
    Add Null=true
    Add Secondary Control=False
    Control Suffix=_ctrl
    Null Suffix=_null
    Secondary Settings="(DisplayName=\"None\",Shape=(bVisible=True,Name=\"Default\",Color=(R=1,G=0,B=0,A=1),Transform=(Rotation=(X=0,Y=0,Z=0,W=1),Translation=(X=0,Y=0,Z=0),Scale3D=(X=1,Y=1,Z=1))),Proxy=(bIsProxy=False,ShapeVisibility=BasedOnSelection),InitialSpace=LocalSpace,bUsePref ...(truncated)"
    Tag Controls=true
    Zero Controls=true
    Name=@Concat_1_1_1.Result
    Item=&RigUnit_ResolveConnector(SkipSocket=true, Connector=RigVMVariableNode() {variable=Ball Bone Connector}.Value).Result
    Parent=RigVMVariableNode() {variable=foot control parent}.Value
    Settings=RigVMVariableNode() {variable=Foot BK Settings}.Value
   )
   > Set Transform_1_1_1 = &RigUnit_SetRotation(Value=(X=-0.5,Y=0.5,Z=-0.5,W=0.5), Weight=1, bInitial=true, Item=@"Control Stack at Item_1".Null)
   > RigVMDispatchNode(Name=Foot BK Control, NameSpace=Self, Value=@"Control Stack at Item_1".Control) {template="DISPATCH_RigDispatch_SetModuleMetadata(in Name,in NameSpace,in Value,out Success)"}
   > RigVMDispatchNode(Name=Foot BK Buffer, NameSpace=Self, Value=@"Control Stack at Item_1".Null) {template="DISPATCH_RigDispatch_SetModuleMetadata(in Name,in NameSpace,in Value,out Success)"}
   > RigVMVariableNode(Value=@"Control Stack at Item_1".Control) {variable=foot bk control}
   > RigVMVariableNode(Value=@"Control Stack at Item_1".Null) {variable=foot bk null}
   Concat_1_1_2_1_1 = RigVMAggregateNode(B=_, C=foot_ik, A=RigVMVariableNode() {variable=prefix}.Value)
   n12 = RigVMVariableNode() {variable=Foot Bone Connector}
   > n13 = &RigUnit_HierarchyAddNull(
    Parent=(Type=Bone,Name="None")
    Space=LocalSpace
    Name=@Concat_1_1_2_1_1.Result
    Transform=&RigUnit_GetTransform(bInitial=true, Item=&RigUnit_ResolveConnector(SkipSocket=true, Connector=@n12.Value).Result).Transform
   )
   > RigVMVariableNode(Value=@n13.Item) {variable=foot ik null}
   > Set Transform_1_1_1_2 = &RigUnit_SetRotation(Value=(X=-0.5,Y=0.5,Z=-0.5,W=0.5), Weight=1, bInitial=true, Item=@n13.Item)
   Concat_1_1_1_1 = RigVMAggregateNode(B=_, C=roll, A=RigVMVariableNode() {variable=prefix}.Value)
   > Control Stack at Item_1_1_1 = RigVMFunctionReferenceNode(
    Add Bottom Null=False
    Add Null=true
    Add Secondary Control=False
    Control Suffix=_ctrl
    Null Suffix=_null
    Secondary Settings="(DisplayName=\"None\",Shape=(bVisible=True,Name=\"Default\",Color=(R=1,G=0,B=0,A=1),Transform=(Rotation=(X=0,Y=0,Z=0,W=1),Translation=(X=0,Y=0,Z=0),Scale3D=(X=1,Y=1,Z=1))),Proxy=(bIsProxy=False,ShapeVisibility=BasedOnSelection),InitialSpace=LocalSpace,bUsePref ...(truncated)"
    Tag Controls=true
    Zero Controls=true
    Name=@Concat_1_1_1_1.Result
    Item=&RigUnit_ResolveConnector(SkipSocket=true, Connector=RigVMVariableNode() {variable=Foot Bone Connector}.Value).Result
    Parent=RigVMVariableNode() {variable=foot control parent}.Value
    Settings=RigVMVariableNode() {variable=Foot Roll Settings}.Value
   )
   > Set Transform_1_1_1_1 = &RigUnit_SetRotation(Value=(X=-0.5,Y=0.5,Z=-0.5,W=0.5), Weight=1, bInitial=true, Item=@"Control Stack at Item_1_1_1".Null)
   > RigVMDispatchNode(Name=Foot Roll Control, NameSpace=Self, Value=@"Control Stack at Item_1_1_1".Control) {template="DISPATCH_RigDispatch_SetModuleMetadata(in Name,in NameSpace,in Value,out Success)"}
   > RigVMVariableNode(Value=@"Control Stack at Item_1_1_1".Control) {variable=foot roll control}
   > RigVMDispatchNode(Name=Foot Roll Buffer, NameSpace=Self, Value=@"Control Stack at Item_1_1_1".Null) {template="DISPATCH_RigDispatch_SetModuleMetadata(in Name,in NameSpace,in Value,out Success)"}
   > RigVMVariableNode(Value=@"Control Stack at Item_1_1_1".Null) {variable=foot roll null}
   Concat_1_1_1_1_1 = RigVMAggregateNode(B=_, C=heel, A=RigVMVariableNode() {variable=prefix}.Value)
   > Control Stack at Item_1_1_2 = RigVMFunctionReferenceNode(
    Add Bottom Null=False
    Add Null=False
    Add Secondary Control=False
    Control Suffix=_ctrl
    Null Suffix=_null
    Secondary Settings="(DisplayName=\"None\",Shape=(bVisible=True,Name=\"Default\",Color=(R=1,G=0,B=0,A=1),Transform=(Rotation=(X=0,Y=0,Z=0,W=1),Translation=(X=0,Y=0,Z=0),Scale3D=(X=1,Y=1,Z=1))),Proxy=(bIsProxy=False,ShapeVisibility=BasedOnSelection),InitialSpace=LocalSpace,bUsePref ...(truncated)"
    Tag Controls=true
    Zero Controls=true
    Name=@Concat_1_1_1_1_1.Result
    Item=RigVMVariableNode() {variable=heel bone}.Value
    Parent=RigVMVariableNode() {variable=foot control parent}.Value
    Settings=RigVMVariableNode() {variable=Heel IK Settings}.Value
   )
   > RigVMVariableNode(Value=@"Control Stack at Item_1_1_2".Control) {variable=heel ik control}
   > RigVMDispatchNode(Name=Heel IK Control, NameSpace=Self, Value=@"Control Stack at Item_1_1_2".Control) {template="DISPATCH_RigDispatch_SetModuleMetadata(in Name,in NameSpace,in Value,out Success)"}
   Concat_1_1_1_1_1_1 = RigVMAggregateNode(B=_, C=tip, A=RigVMVariableNode() {variable=prefix}.Value)
   > Control Stack at Item_1_1 = RigVMFunctionReferenceNode(
    Add Bottom Null=False
    Add Null=False
    Add Secondary Control=False
    Control Suffix=_ctrl
    Null Suffix=_null
    Secondary Settings="(DisplayName=\"None\",Shape=(bVisible=True,Name=\"Default\",Color=(R=1,G=0,B=0,A=1),Transform=(Rotation=(X=0,Y=0,Z=0,W=1),Translation=(X=0,Y=0,Z=0),Scale3D=(X=1,Y=1,Z=1))),Proxy=(bIsProxy=False,ShapeVisibility=BasedOnSelection),InitialSpace=LocalSpace,bUsePref ...(truncated)"
    Tag Controls=true
    Zero Controls=true
    Name=@Concat_1_1_1_1_1_1.Result
    Item=RigVMVariableNode() {variable=tip bone}.Value
    Parent=RigVMVariableNode() {variable=foot control parent}.Value
    Settings=RigVMVariableNode() {variable=Tip Settings}.Value
   )
   > RigVMVariableNode(Value=@"Control Stack at Item_1_1".Control) {variable=tip control}
   > RigVMDispatchNode(Name=Tip Control, NameSpace=Self, Value=@"Control Stack at Item_1_1".Control) {template="DISPATCH_RigDispatch_SetModuleMetadata(in Name,in NameSpace,in Value,out Success)"}
   Concat_1_1_1_1_1_1_1 = RigVMAggregateNode(B=_, C=Blend, A=RigVMVariableNode() {variable=prefix}.Value)
   > SpawnControl = /Script/ControlRig.RigUnit_HierarchyAddControlFloat(
    InitialValue=1
    OffsetSpace=LocalSpace
    OffsetTransform=(Rotation=(X=0,Y=0,Z=0,W=1),Translation=(X=0,Y=0,Z=0),Scale3D=(X=1,Y=1,Z=1))
    Settings="(PrimaryAxis=X,bIsScale=False,Limits=(Limit=(bMinimum=True,bMaximum=True),MinValue=0,MaxValue=100,bDrawLimits=True),Shape=(bVisible=false,Name=\"Default\",Color=(R=1,G=0,B=0,A=1),Transform=(Rotation=(X=0,Y=0,Z=0,W=1),Translation=(X=0,Y=0,Z=0),Scale3D=(X=1,Y= ...(truncated)"
    Name=@Concat_1_1_1_1_1_1_1.Result
    Parent=RigVMVariableNode() {variable=foot roll control}.Value
   ) {
    is=struct RigVMUnitNode
    template="SpawnControl::Execute(in InitialValue,in Settings,in OffsetTransform,in OffsetSpace,in Parent,in Name,out Item)"
   }
   > RigVMVariableNode(Value=@SpawnControl.Item) {variable=ball fk control}
   > RigVMDispatchNode(Name=Ball FK Control, NameSpace=Self, Value=@SpawnControl.Item) {template="DISPATCH_RigDispatch_SetModuleMetadata(in Name,in NameSpace,in Value,out Success)"}
   RerouteNode_11 = RigVMDispatchNode(
    Values=((Type=Bone,Name="heel_l"),((),Type=Bone,Name="tip_l"),((),Type=Bone,Name="ball_l"),((),Type=Bone,Name="foot_l"))
    Values.0=RigVMVariableNode() {variable=heel bone}.Value
    Values.1=RigVMVariableNode() {variable=tip bone}.Value
    Values.3=RigVMVariableNode() {variable=foot bone}.Value
    Values.2=RigVMVariableNode() {variable=ball bone}.Value
   ) {template="DISPATCH_RigVMDispatch_ArrayMake(in Values,out Array)"}
   RerouteNode_18 = RigVMDispatchNode(
    Values=((Type=Control,Name="heel_l_ctrl"),((),Type=Control,Name="tip_l_ctrl"),((),Type=Control,Name="foot_bk1_l_ctrl"),((),Type=Null,Name="foot_ik_l"))
    Values.2=RigVMVariableNode() {variable=foot bk control}.Value
    Values.0=RigVMVariableNode() {variable=heel ik control}.Value
    Values.1=RigVMVariableNode() {variable=tip control}.Value
    Values.3=RigVMVariableNode() {variable=foot ik null}.Value
   ) {template="DISPATCH_RigVMDispatch_ArrayMake(in Values,out Array)"}
   RerouteNode_23 = RigVMDispatchNode(
    Values=((Type=Null,Name="heel_l"),((),Type=Null,Name="tip_l"),((),Type=Null,Name="foot_bk1_l"),((),Type=Null,Name="foot_ik_l"))
    Values.2=RigVMVariableNode() {variable=foot bk control}.Value
    Values.0=RigVMVariableNode() {variable=heel bone}.Value
    Values.1=RigVMVariableNode() {variable=tip bone}.Value
    Values.3=RigVMVariableNode() {variable=foot ik null}.Value
   ) {template="DISPATCH_RigVMDispatch_ArrayMake(in Values,out Array)"}
   Setup Reverse Foot = RigVMFunctionReferenceNode(
    Bones=@RerouteNode_11.Array
    BK Controls=@RerouteNode_18.Array
    BK Buffers=@RerouteNode_23.Array
    Roll Control=RigVMVariableNode() {variable=foot roll control}.Value
    Roll Buffer=RigVMVariableNode() {variable=foot roll null}.Value
    ExecuteContext=@RigVMFunction_Sequence.G
   )
   Concat_1_1_2 = RigVMAggregateNode(B=_, C=heel_null, A=RigVMVariableNode() {variable=prefix}.Value)
   n14 = &RigUnit_HierarchyAddNull(
    Parent=(Type=Bone,Name="None")
    Space=LocalSpace
    Name=@Concat_1_1_2.Result
    Transform=&RigUnit_GetTransform(bInitial=False, Item=RigVMVariableNode() {variable=heel bone}.Value).Transform
    ExecutePin=@RigVMFunction_Sequence.F
   )
   > RigVMVariableNode(Value=@n14.Item) {variable=heel null}
   > RigVMDispatchNode(Name=heel buffer, NameSpace=Self, Value=@n14.Item) {template="DISPATCH_RigDispatch_SetModuleMetadata(in Name,in NameSpace,in Value,out Success)"}
   Concat_1_1_2_1 = RigVMAggregateNode(B=_, C=tip_null, A=RigVMVariableNode() {variable=prefix}.Value)
   > n15 = &RigUnit_HierarchyAddNull(
    Parent=(Type=Bone,Name="None")
    Space=LocalSpace
    Name=@Concat_1_1_2_1.Result
    Transform=&RigUnit_GetTransform(bInitial=False, Item=RigVMVariableNode() {variable=tip bone}.Value).Transform
   )
   > RigVMVariableNode(Value=@n15.Item) {variable=tip null}
   > RigVMDispatchNode(Name=tip buffer, NameSpace=Self, Value=@n15.Item) {template="DISPATCH_RigDispatch_SetModuleMetadata(in Name,in NameSpace,in Value,out Success)"}
   > &RigUnit_SetDefaultParent(Parent=@n14.Item, Child=@n15.Item)
   > &RigUnit_SetDefaultParent(Child=RigVMVariableNode() {variable=foot roll null}.Value, Parent=RigVMVariableNode() {variable=foot bk control}.Value)
   > &RigUnit_SetDefaultParent(Parent=RigVMVariableNode() {variable=tip null}.Value, Child=RigVMVariableNode() {variable=foot bk null}.Value)
   > &RigUnit_SetDefaultParent(Child=RigVMVariableNode() {variable=tip control}.Value, Parent=RigVMVariableNode() {variable=heel null}.Value)
   n16 = RigVMVariableNode() {variable=ball ik control}
   > &RigUnit_SetDefaultParent(Parent=RigVMVariableNode() {variable=tip null}.Value, Child=@n16.Value)
   > Set Offset Transform from Current_0_1 = RigVMFunctionReferenceNode(Control=@n16.Value)
   > &RigUnit_SetDefaultParent(Child=RigVMVariableNode() {variable=heel null}.Value, Parent=RigVMVariableNode() {variable=foot control parent}.Value)
   > &RigUnit_SetDefaultParent(Child=RigVMVariableNode() {variable=foot ik null}.Value, Parent=RigVMVariableNode() {variable=foot bk control}.Value)
   RigVMFunctionReturnNode(ExecuteContext=@n2.B)
   SpawnAnimationChannel = /Script/ControlRig.RigUnit_HierarchyAddAnimationChannelBool(
    ExecutePin=()
    InitialValue=true
    LimitsEnabled=()
    MaximumValue=True
    MinimumValue=False
    Name=foot_fk_ik_switch
    Parent=&RigUnit_ResolveConnector(Connector=(Type=Connector,Name="Settings Control"), SkipSocket=true).Result
   ) {
    is=struct RigVMUnitNode
    template="SpawnAnimationChannel::Execute(in InitialValue,in MinimumValue,in MaximumValue,in LimitsEnabled,in Parent,in Name,out Item)"
   }
   > RigVMVariableNode(Value=@SpawnAnimationChannel.Item) {variable=fk ik switch}
   > RigVMDispatchNode(Name=fk ik switch, NameSpace=Self, Value=(Type=Bone,Name="None")) {template="DISPATCH_RigDispatch_SetModuleMetadata(in Name,in NameSpace,in Value,out Success)"}
   Set Transform_1_1 = &RigUnit_SetRotation(ExecutePin=(), Value=(X=-0.5,Y=0.5,Z=-0.5,W=0.5), Weight=1, bInitial=true, Item=@n10.Item)
  ]}
 New Function_ContainedGraph {nodes [
   n1 = RigVMFunctionEntryNode()
   n2 = /Script/ControlRig.RigUnit_FindItemsWithMetadataTag(NameSpace=Self, Tag=Controls) {is=struct RigVMUnitNode}
   > Scale Module Controls = RigVMFunctionReferenceNode(
    Global Scale=RigVMDispatchNode(Default=1, Name=Control Size, NameSpace=Root) {template="DISPATCH_RigDispatch_GetModuleMetadata(in Name,in NameSpace,in Default,out Value,out Found)"}.Value
    Local Scale=@n1.Module Settings.ControlSize_12_#1
    Controls=@n2.Items
   )
   > Color Module Controls = RigVMFunctionReferenceNode(
    CenterColor=@n1.Module Settings.CenterColor_9_#2
    LeftColor=@n1.Module Settings.LeftColor_3_#3
    Override=@n1.Module Settings.OverrideColor_48_#4
    RightColor=@n1.Module Settings.RightColor_5_#5
    Controls=@n2.Items
   )
   > RigVMFunctionReturnNode()
  ]}
 New Function_ContainedGraph {nodes [
   RigVMFunctionEntryNode()
   > RigVMFunctionReturnNode()
  ]}
 New Function_ContainedGraph {nodes [
   n1 = RigVMFunctionEntryNode()
   Concat_4 = RigVMAggregateNode(B=@n1.name, A=@n1.prefix, C=@n1.side, D=@n1.suffix)
   > RigVMFunctionReturnNode(Result=@Concat_4.Result)
  ]}
 New Function_ContainedGraph {nodes [
   RigVMFunctionEntryNode()
   > RigVMFunction_Sequence = RigVMAggregateNode()
   Get_Module_Metadata_3 = RigVMDispatchNode(Default=(X=0,Y=0,Z=0), Name=Mirror Axis, NameSpace=Root) {template="DISPATCH_RigDispatch_GetModuleMetadata(in Name,in NameSpace,in Default,out Value,out Found)"}
   Is Right Side v2 = RigVMFunctionReferenceNode(
    Mirror Plane=@Get_Module_Metadata_3.Value
    Item=&RigUnit_ResolveConnector(Connector=(Type=Connector,Name="Foot Primary"), SkipSocket=true).Result
    ExecuteContext=@RigVMFunction_Sequence.A
   )
   > RigVMVariableNode(Value=@"Is Right Side v2".Right Side) {variable=is on right side}
   Connect to Module Metadata_1 = RigVMFunctionReferenceNode(
    Connector=(Type=Connector,Name="Settings Control")
    Module Metadata Name=Settings Control
    Module Metadata NameSpace=Root
    ExecuteContext=@RigVMFunction_Sequence.B
   )
   Connect to Module Metadata_1_1_1 = RigVMFunctionReferenceNode(
    Module Metadata NameSpace=Root
    Module Metadata Name=RigVMDispatchNode(False=foot_l, True=foot_r, Condition=RigVMVariableNode() {variable=is on right side}.Value) {template="DISPATCH_RigVMDispatch_If(in Condition,in True,in False,out Result)"}.Result
    Connector=/Script/ControlRig.RigUnit_Item(Item=(Type=Connector,Name="Foot Bone")) {is=struct RigVMUnitNode}.Item
    ExecuteContext=@RigVMFunction_Sequence.C
   )
   n1 = &RigVMFunction_ControlFlowBranch(ExecuteContext=@"Connect to Module Metadata_1_1_1".ExecuteContext, Condition=@"Connect to Module Metadata_1_1_1".Found)
   n2 = /Script/ControlRig.RigUnit_GetCandidates() {is=struct RigVMUnitNode}
   n3 = &RigVMFunction_ControlFlowBranch(
    Condition=RigVMDispatchNode(B=@n2.Connector, A=&RigUnit_Item(Item=(Type=Connector,Name="Foot Bone")).Item) {template="DISPATCH_RigVMDispatch_CoreEquals(in A,in B,out Result)"}.Result
    ExecuteContext=@n1.False
   )
   /Script/ControlRig.RigUnit_DiscardMatches(
    Excluded=/Script/ControlRig.RigUnit_CollectionNameSearchArray(PartialName=corrective, TypeToSearch=Bone) {is=struct RigVMUnitNode}.Items
    ExecutePin=@n3.True
   ) {is=struct RigVMUnitNode}
   > n4 = RigVMDispatchNode(Array=@n2.Candidates) {template="DISPATCH_RigVMDispatch_ArrayIterator(in Array,out Element,out Index,out Count,out Ratio)"}
   > Find first in Hierarchy = RigVMFunctionReferenceNode(
    Children=false
    Ignore=()
    Search for=("foot")
    Type to Search=Bone
    Item=&RigUnit_ResolveConnector(Connector=(Type=Connector,Name="Foot Primary"), SkipSocket=true).Result
   )
   n5 = &RigVMFunction_ControlFlowBranch(
    Condition=RigVMDispatchNode(B=@n4.Element, A=@"Find first in Hierarchy".Item Found) {template="DISPATCH_RigVMDispatch_CoreEquals(in A,in B,out Result)"}.Result
    ExecuteContext=@"Find first in Hierarchy".ExecuteContext
   )
   /Script/ControlRig.RigUnit_SetDefaultMatch(Default=@n4.Element, ExecutePin=@n5.True) {is=struct RigVMUnitNode, template="Set Default Match::Execute(in Default)"}
   n6 = &RigUnit_GetCandidates()
   n7 = &RigVMFunction_ControlFlowBranch(
    Condition=RigVMDispatchNode(B=@n6.Connector, A=&RigUnit_Item(Item=(Type=Connector,Name="Ball Bone")).Item) {template="DISPATCH_RigVMDispatch_CoreEquals(in A,in B,out Result)"}.Result
    ExecuteContext=@RigVMFunction_Sequence.D
   )
   n8 = RigVMDispatchNode(Array=@n6.Candidates, ExecuteContext=@n7.True) {template="DISPATCH_RigVMDispatch_ArrayIterator(in Array,out Element,out Index,out Count,out Ratio)"}
   n9 = &RigVMFunction_ControlFlowBranch(
    ExecuteContext=@n8.ExecuteContext
    Condition=RigVMDispatchNode(B=@n8.Element, A=&RigUnit_ResolveConnector(Connector=(Type=Connector,Name="Foot Primary"), SkipSocket=true).Result) {template="DISPATCH_RigVMDispatch_CoreEquals(in A,in B,out Result)"}.Result
   )
   &RigUnit_SetDefaultMatch(Default=@n8.Element, ExecutePin=@n9.True) {template="Set Default Match::Execute(in Default)"}
   n10 = &RigUnit_GetCandidates()
   n11 = &RigVMFunction_ControlFlowBranch(
    Condition=RigVMDispatchNode(B=@n10.Connector, A=&RigUnit_Item(Item=(Type=Connector,Name="Parent Control")).Item) {template="DISPATCH_RigVMDispatch_CoreEquals(in A,in B,out Result)"}.Result
    ExecuteContext=@RigVMFunction_Sequence.E
   )
   n12 = RigVMDispatchNode(Array=@n10.Candidates, ExecuteContext=@n11.True) {template="DISPATCH_RigVMDispatch_ArrayIterator(in Array,out Element,out Index,out Count,out Ratio)"}
   n13 = &RigVMFunction_ControlFlowBranch(
    ExecuteContext=@n12.ExecuteContext
    Condition=RigVMDispatchNode(
     B=RigVMDispatchNode(Default=(Type=Bone,Name="None"), Name=Foot Control, NameSpace=Parent) {template="DISPATCH_RigDispatch_GetModuleMetadata(in Name,in NameSpace,in Default,out Value,out Found)"}.Value
     A=@n12.Element
    ) {template="DISPATCH_RigVMDispatch_CoreEquals(in A,in B,out Result)"}.Result
   )
   &RigUnit_SetDefaultMatch(Default=@n12.Element, ExecutePin=@n13.True) {template="Set Default Match::Execute(in Default)"}
   RigVMFunctionReturnNode(ExecuteContext=@RigVMFunction_Sequence.F)
  ]}
 New Function_ContainedGraph {nodes [
   n1 = RigVMFunctionEntryNode()
   Get_Module_Metadata_4 = RigVMDispatchNode(Default=(X=1,Y=0,Z=0), Name=Mirror Axis, NameSpace=Root) {template="DISPATCH_RigDispatch_GetModuleMetadata(in Name,in NameSpace,in Default,out Value,out Found)"}
   Not_Equals_1 = RigVMDispatchNode(B=0, A=@Get_Module_Metadata_4.Value.Y) {template="DISPATCH_RigVMDispatch_CoreNotEquals(in A,in B,out Result)"}
   Not_Equals_2 = RigVMDispatchNode(B=0, A=@Get_Module_Metadata_4.Value.Z) {template="DISPATCH_RigVMDispatch_CoreNotEquals(in A,in B,out Result)"}
   Not_Equals = RigVMDispatchNode(B=0, A=@Get_Module_Metadata_4.Value.X) {template="DISPATCH_RigVMDispatch_CoreNotEquals(in A,in B,out Result)"}
   Dot___1 = &RigVMFunction_MathVectorDot(
    A=(X=0,Y=0,Z=0)
    B=(X=1,Y=1,Z=0)
    B.Y=&RigVMFunction_MathBoolToFloat(Value=@Not_Equals_1.Result).Result
    B.Z=&RigVMFunction_MathBoolToFloat(Value=@Not_Equals_2.Result).Result
    B.X=&RigVMFunction_MathBoolToFloat(Value=@Not_Equals.Result).Result
    A.X=RigVMDispatchNode(False=0, True=&RigUnit_GetTransform(bInitial=true, Item=@n1.Item).Transform.Translation.X, Condition=@Not_Equals.Result) {template="DISPATCH_RigVMDispatch_If(in Condition,in True,in False,out Result)"}.Result
    A.Y=RigVMDispatchNode(False=0, True=&RigUnit_GetTransform(bInitial=true, Item=@n1.Item).Transform.Translation.Y, Condition=@Not_Equals_1.Result) {template="DISPATCH_RigVMDispatch_If(in Condition,in True,in False,out Result)"}.Result
    A.Z=RigVMDispatchNode(False=0, True=&RigUnit_GetTransform(bInitial=true, Item=@n1.Item).Transform.Translation.Z, Condition=@Not_Equals_2.Result) {template="DISPATCH_RigVMDispatch_If(in Condition,in True,in False,out Result)"}.Result
   ) {template="Dot,|::Execute(in A,in B,out Result)"}
   Greater = &RigVMFunction_MathDoubleGreater(B=0, A=@Get_Module_Metadata_4.Value.X) {template="Greater::Execute(in A,in B,out Result)"}
   RigVMFunction_MathBoolOr = RigVMAggregateNode(
    A=@Greater.Result
    B=&RigVMFunction_MathDoubleGreater(B=0, A=@Get_Module_Metadata_4.Value.Y) {template="Greater::Execute(in A,in B,out Result)"}.Result
    C=&RigVMFunction_MathDoubleGreater(B=0, A=@Get_Module_Metadata_4.Value.Z) {template="Greater::Execute(in A,in B,out Result)"}.Result
   )
   RigVMFunction_MathBoolNot = &RigVMFunction_MathBoolNot(Value=RigVMDispatchNode(
     True=&RigVMFunction_MathDoubleGreater(B=0, A=@Dot___1.Result) {template="Greater::Execute(in A,in B,out Result)"}.Result
     False=&RigVMFunction_MathDoubleLess(B=0, A=@Dot___1.Result) {template="Less::Execute(in A,in B,out Result)"}.Result
     Condition=@RigVMFunction_MathBoolOr.Result
    ) {template="DISPATCH_RigVMDispatch_If(in Condition,in True,in False,out Result)"}.Result)
   > RigVMFunctionReturnNode(Result=@RigVMFunction_MathBoolNot.Result)
  ]}
 RigVMFunctionLibrary {nodes [
   CONNECT Foot = RigVMCollapseNode(ExecuteContext=())
   CONSTR Foot = RigVMCollapseNode(
    Ball Bone Connector=(Type=Connector,Name="Ball Bone")
    Ball FK Settings="(DisplayName=\"None\",bSelectable=True,InitialSpace=LocalSpace,bUsePreferredRotationOrder=False,PreferredRotationOrder=YZX,Limits=(LimitTranslationX=(bMinimum=False,bMaximum=False),LimitTranslationY=(bMinimum=False,bMaximum=False),LimitTranslationZ=(bMinimum ...(truncated)"
    Ball IK Settings="(DisplayName=\"None\",bSelectable=True,InitialSpace=LocalSpace,bUsePreferredRotationOrder=False,PreferredRotationOrder=YZX,Limits=(LimitTranslationX=(bMinimum=False,bMaximum=False),LimitTranslationY=(bMinimum=False,bMaximum=False),LimitTranslationZ=(bMinimum ...(truncated)"
    ExecuteContext=()
    Foot BK Settings="(DisplayName=\"None\",bSelectable=True,InitialSpace=LocalSpace,bUsePreferredRotationOrder=False,PreferredRotationOrder=YZX,Limits=(LimitTranslationX=(bMinimum=False,bMaximum=False),LimitTranslationY=(bMinimum=False,bMaximum=False),LimitTranslationZ=(bMinimum ...(truncated)"
    Foot Bone Connector=(Type=Connector,Name="Foot Bone")
    Foot Roll Settings="(DisplayName=\"None\",bSelectable=True,InitialSpace=LocalSpace,bUsePreferredRotationOrder=False,PreferredRotationOrder=YZX,Limits=(LimitTranslationX=(bMinimum=False,bMaximum=False),LimitTranslationY=(bMinimum=False,bMaximum=False),LimitTranslationZ=(bMinimum ...(truncated)"
    Heel IK Settings="(DisplayName=\"None\",bSelectable=True,InitialSpace=LocalSpace,bUsePreferredRotationOrder=False,PreferredRotationOrder=YZX,Limits=(LimitTranslationX=(bMinimum=False,bMaximum=False),LimitTranslationY=(bMinimum=False,bMaximum=False),LimitTranslationZ=(bMinimum ...(truncated)"
    Parent Control Connector=(Type=Connector,Name="Foot Control Parent")
    Tip Settings="(DisplayName=\"None\",bSelectable=True,InitialSpace=LocalSpace,bUsePreferredRotationOrder=False,PreferredRotationOrder=YZX,Limits=(LimitTranslationX=(bMinimum=False,bMaximum=False),LimitTranslationY=(bMinimum=False,bMaximum=False),LimitTranslationZ=(bMinimum ...(truncated)"
   )
   Format Name = RigVMCollapseNode(name=arm, prefix=None, side=_fk_ik_switch, suffix=_fk_ik_switch)
   Forward Foot = RigVMCollapseNode(Bone=(), ExecuteContext=(), FK Control=(), IK Control=(), IKFK=True, IKFK Switch Control=(), Toes Space=())
   Forward Reverse Foot = RigVMCollapseNode(
    BK Buffers=()
    BK Controls=()
    ExecuteContext=()
    IKFK Switch Control=()
    Mirror Plane=(X=1,Y=0,Z=0)
    Roll Buffer=()
    Roll Clamp=30
    Roll Control=()
   )
   Get Negative Transform = RigVMCollapseNode(Input=(Rotation=(X=0,Y=0,Z=0,W=1),Translation=(X=0,Y=0,Z=0),Scale3D=(X=0,Y=0,Z=0)))
   Get Right Side = RigVMCollapseNode(Item=(Type=None,Name="None"))
   INV Foot = RigVMCollapseNode(ExecuteContext=())
   Is Right Side v2 = RigVMCollapseNode(ExecuteContext=(), Item=(Type=Bone), Mirror Plane=(X=1,Y=1,Z=0))
   Setup Reverse Foot = RigVMCollapseNode(BK Buffers=(), BK Controls=(), Bones=(), ExecuteContext=(), Roll Buffer=(), Roll Control=())
  ]}
 RigVMFunction_MathBoolOr_1_ContainedGraph {nodes [
   n1 = RigVMFunctionEntryNode()
   RigVMFunction_MathBoolOr = &RigVMFunction_MathBoolOr(A=@n1.A, B=@n1.B)
   RigVMFunctionReturnNode(Result=&RigVMFunction_MathBoolOr(B=@n1.C, A=@RigVMFunction_MathBoolOr.Result).Result)
  ]}
 RigVMFunction_Sequence_1_ContainedGraph {nodes [
   n1 = RigVMFunctionEntryNode()
   RigVMFunction_Sequence = &RigVMFunction_Sequence(ExecuteContext=@n1.ExecuteContext)
   n2 = &RigVMFunction_Sequence(ExecuteContext=@RigVMFunction_Sequence.B)
   n3 = &RigVMFunction_Sequence(ExecuteContext=@n2.B)
   n4 = &RigVMFunction_Sequence(ExecuteContext=@n3.B)
   n5 = &RigVMFunction_Sequence(ExecuteContext=@n4.B)
   RigVMFunctionReturnNode(A=@RigVMFunction_Sequence.A, B=@n2.A, C=@n3.A, D=@n4.A, E=@n5.A, F=@n5.B)
  ]}
 RigVMFunction_Sequence_1_ContainedGraph {nodes [
   n1 = RigVMFunctionEntryNode()
   RigVMFunction_Sequence = &RigVMFunction_Sequence(ExecuteContext=@n1.ExecuteContext)
   n2 = &RigVMFunction_Sequence(ExecuteContext=@RigVMFunction_Sequence.B)
   n3 = &RigVMFunction_Sequence(ExecuteContext=@n2.B)
   n4 = &RigVMFunction_Sequence(ExecuteContext=@n3.B)
   n5 = &RigVMFunction_Sequence(ExecuteContext=@n4.B)
   n6 = &RigVMFunction_Sequence(ExecuteContext=@n5.B)
   n7 = &RigVMFunction_Sequence(ExecuteContext=@n6.B)
   n8 = &RigVMFunction_Sequence(ExecuteContext=@n7.B)
   n9 = &RigVMFunction_Sequence(ExecuteContext=@n8.B)
   RigVMFunctionReturnNode(A=@RigVMFunction_Sequence.A, B=@n2.A, C=@n3.A, D=@n4.A, E=@n5.A, F=@n6.A, G=@n7.A, H=@n8.A, I=@n9.A, J=@n9.B)
  ]}
 RigVMFunction_Sequence_1_ContainedGraph {nodes [
   n1 = RigVMFunctionEntryNode()
   RigVMFunction_Sequence = &RigVMFunction_Sequence(ExecuteContext=@n1.ExecuteContext)
   n2 = &RigVMFunction_Sequence(ExecuteContext=@RigVMFunction_Sequence.B)
   n3 = &RigVMFunction_Sequence(ExecuteContext=@n2.B)
   n4 = &RigVMFunction_Sequence(ExecuteContext=@n3.B)
   n5 = &RigVMFunction_Sequence(ExecuteContext=@n4.B)
   RigVMFunctionReturnNode(A=@RigVMFunction_Sequence.A, B=@n2.A, C=@n3.A, D=@n4.A, E=@n5.A, F=@n5.B)
  ]}
 RigVMModel {nodes [
   RigUnit_BeginExecution = &RigUnit_BeginExecution()
   Get_Metadata = RigVMDispatchNode(Default=None, Item=(Type=Bone,Name="None"), Name=fk_ik_switch, NameSpace=Parent) {template="DISPATCH_RigDispatch_GetMetadata(in Item,in Name,in NameSpace,in Default,out Value,out Found)"}
   Get_Module_Metadata_3 = RigVMDispatchNode(Default=(X=0,Y=0,Z=0), Name=Mirror Axis, NameSpace=Root) {template="DISPATCH_RigDispatch_GetModuleMetadata(in Name,in NameSpace,in Default,out Value,out Found)"}
   DISPATCH_RigDispatch_GetModuleMetadata = RigVMDispatchNode(Default=true, Name=Current FK IK Switch, NameSpace=Parent) {template="DISPATCH_RigDispatch_GetModuleMetadata(in Name,in NameSpace,in Default,out Value,out Found)"}
   Forward Reverse Foot_2 = RigVMFunctionReferenceNode(
    BK Buffers=((Type=None,Name="None"),(Type=None,Name="None"),(Type=None,Name="None"))
    BK Controls=((Type=None,Name="None"),(Type=None,Name="None"),(Type=None,Name="None"))
    IKFK Switch Control=(Type=Control,Name="leg_l_fk_ik_switch")
    Roll Blend=0.5
    Roll Clamp=40
    Roll Control=RigVMDispatchNode(Default=(Type=Bone,Name="None"), Name=Foot Roll Control, NameSpace=Self) {template="DISPATCH_RigDispatch_GetModuleMetadata(in Name,in NameSpace,in Default,out Value,out Found)"}.Value
    Roll Buffer=RigVMDispatchNode(Default=(Type=Bone,Name="None"), Name=Foot Roll Buffer, NameSpace=Self) {template="DISPATCH_RigDispatch_GetModuleMetadata(in Name,in NameSpace,in Default,out Value,out Found)"}.Value
    BK Buffers.0=RigVMDispatchNode(Default=(Type=Bone,Name="None"), Name=heel buffer, NameSpace=Self) {template="DISPATCH_RigDispatch_GetModuleMetadata(in Name,in NameSpace,in Default,out Value,out Found)"}.Value
    BK Buffers.1=RigVMDispatchNode(Default=(Type=Bone,Name="None"), Name=tip buffer, NameSpace=Self) {template="DISPATCH_RigDispatch_GetModuleMetadata(in Name,in NameSpace,in Default,out Value,out Found)"}.Value
    BK Buffers.2=RigVMDispatchNode(Default=(Type=Bone,Name="None"), Name=Foot BK Buffer, NameSpace=Self) {template="DISPATCH_RigDispatch_GetModuleMetadata(in Name,in NameSpace,in Default,out Value,out Found)"}.Value
    BK Controls.2=RigVMDispatchNode(Default=(Type=Bone,Name="None"), Name=Foot BK Control, NameSpace=Self) {template="DISPATCH_RigDispatch_GetModuleMetadata(in Name,in NameSpace,in Default,out Value,out Found)"}.Value
    BK Controls.1=RigVMDispatchNode(Default=(Type=Bone,Name="None"), Name=Tip Control, NameSpace=Self) {template="DISPATCH_RigDispatch_GetModuleMetadata(in Name,in NameSpace,in Default,out Value,out Found)"}.Value
    BK Controls.0=RigVMDispatchNode(Default=(Type=Bone,Name="None"), Name=Heel IK Control, NameSpace=Self) {template="DISPATCH_RigDispatch_GetModuleMetadata(in Name,in NameSpace,in Default,out Value,out Found)"}.Value
    IKFK Switch Control.Name=@Get_Metadata.Value
    Mirror Plane=@Get_Module_Metadata_3.Value
    FK or IK=RigVMDispatchNode(True=False, False=@DISPATCH_RigDispatch_GetModuleMetadata.Value, Condition=RigVMVariableNode() {variable=Force FK Mode}.Value) {template="DISPATCH_RigVMDispatch_If(in Condition,in True,in False,out Result)"}.Result
    ExecuteContext=@RigUnit_BeginExecution.ExecutePin
   )
   RigVMDispatchNode(Default=(Type=Bone,Name="None"), Name=fk ik switch, NameSpace=Self) {template="DISPATCH_RigDispatch_GetModuleMetadata(in Name,in NameSpace,in Default,out Value,out Found)"}
  ]}
 RigVMModel Backwards Solve Graph {nodes [
   n1 = /Script/ControlRig.RigUnit_InverseExecution() {is=struct RigVMUnitNode}
   RigVMVariableNode(Value=true, ExecuteContext=@n1.ExecutePin) {variable=Force FK Mode}
   > /Script/ControlRig.RigUnit_SwitchParent(
    Mode=ParentItem
    bMaintainGlobal=True
    Child=RigVMVariableNode() {variable=Effector Null}.Value
    Parent=RigVMVariableNode() {variable=Foot Control}.Value
   ) {is=struct RigVMUnitNode}
   > &RigUnit_ParentConstraint(
    AdvancedSettings=(InterpolationType=Average,RotationOrderForFilter=XZY)
    Parents=((Item=(Type=Bone,Name="None"),Weight=1))
    Weight=1
    Parents.0.Item=&RigUnit_ResolveConnector(Connector=(Type=Connector,Name="Ball Bone"), SkipSocket=true).Result
    Child=RigVMVariableNode() {variable=foot bk control}.Value
   )
   > &RigUnit_ParentConstraint(
    AdvancedSettings=(InterpolationType=Average,RotationOrderForFilter=XZY)
    Parents=((Item=(Type=Bone,Name="None"),Weight=1))
    Weight=1
    Child=&RigUnit_HierarchyGetParent(bDefaultParent=True, Child=RigVMVariableNode() {variable=foot bk control}.Value).Parent
    Parents.0.Item=&RigUnit_ResolveConnector(Connector=(Type=Connector,Name="Ball Bone"), SkipSocket=true).Result
   )
  ]}
 RigVMModel Connector Event Graph {nodes [
   n1 = /Script/ControlRig.RigUnit_ConnectorExecution() {is=struct RigVMUnitNode}
   RigVMFunction_Sequence = RigVMAggregateNode(ExecuteContext=@n1.ExecuteContext)
   Get_Module_Metadata_3 = RigVMDispatchNode(Default=(X=0,Y=0,Z=0), Name=Mirror Axis, NameSpace=Root) {template="DISPATCH_RigDispatch_GetModuleMetadata(in Name,in NameSpace,in Default,out Value,out Found)"}
   Is Right Side v2 = RigVMFunctionReferenceNode(
    Mirror Plane=@Get_Module_Metadata_3.Value
    Item=&RigUnit_ResolveConnector(Connector=(Type=Connector,Name="Foot Primary"), SkipSocket=true).Result
    ExecuteContext=@RigVMFunction_Sequence.A
   )
   Connect to Module Metadata_1 = RigVMFunctionReferenceNode(
    Connector=(Type=Connector,Name="Settings Control")
    Module Metadata Name=Settings Control
    Module Metadata NameSpace=Root
    ExecuteContext=@RigVMFunction_Sequence.B
   )
   Connect to Module Metadata_1_1_1 = RigVMFunctionReferenceNode(
    Module Metadata NameSpace=Root
    Module Metadata Name=RigVMDispatchNode(Condition=False, False=foot_l, True=foot_r) {template="DISPATCH_RigVMDispatch_If(in Condition,in True,in False,out Result)"}.Result
    Connector=&RigUnit_Item(Item=(Type=Connector,Name="Foot Bone")).Item
    ExecuteContext=@RigVMFunction_Sequence.C
   )
   n2 = &RigVMFunction_ControlFlowBranch(ExecuteContext=@"Connect to Module Metadata_1_1_1".ExecuteContext, Condition=@"Connect to Module Metadata_1_1_1".Found)
   n3 = &RigUnit_GetCandidates()
   n4 = &RigVMFunction_ControlFlowBranch(
    Condition=RigVMDispatchNode(B=@n3.Connector, A=&RigUnit_Item(Item=(Type=Connector,Name="Foot Bone")).Item) {template="DISPATCH_RigVMDispatch_CoreEquals(in A,in B,out Result)"}.Result
    ExecuteContext=@n2.False
   )
   &RigUnit_DiscardMatches(Excluded=&RigUnit_CollectionNameSearchArray(PartialName=corrective, TypeToSearch=Bone).Items, ExecutePin=@n4.True)
   > n5 = RigVMDispatchNode(Array=@n3.Candidates) {template="DISPATCH_RigVMDispatch_ArrayIterator(in Array,out Element,out Index,out Count,out Ratio)"}
   > Find first in Hierarchy = RigVMFunctionReferenceNode(
    Children=false
    Ignore=()
    Search for=("foot")
    Type to Search=Bone
    Item=&RigUnit_ResolveConnector(Connector=(Type=Connector,Name="Foot Primary"), SkipSocket=true).Result
   )
   n6 = &RigVMFunction_ControlFlowBranch(
    Condition=RigVMDispatchNode(B=@n5.Element, A=@"Find first in Hierarchy".Item Found) {template="DISPATCH_RigVMDispatch_CoreEquals(in A,in B,out Result)"}.Result
    ExecuteContext=@"Find first in Hierarchy".ExecuteContext
   )
   &RigUnit_SetDefaultMatch(Default=@n5.Element, ExecutePin=@n6.True) {template="Set Default Match::Execute(in Default)"}
   n7 = &RigUnit_GetCandidates()
   n8 = &RigVMFunction_ControlFlowBranch(
    Condition=RigVMDispatchNode(B=@n7.Connector, A=&RigUnit_Item(Item=(Type=Connector,Name="Ball Bone")).Item) {template="DISPATCH_RigVMDispatch_CoreEquals(in A,in B,out Result)"}.Result
    ExecuteContext=@RigVMFunction_Sequence.D
   )
   n9 = RigVMDispatchNode(Array=@n7.Candidates, ExecuteContext=@n8.True) {template="DISPATCH_RigVMDispatch_ArrayIterator(in Array,out Element,out Index,out Count,out Ratio)"}
   n10 = &RigVMFunction_ControlFlowBranch(
    ExecuteContext=@n9.ExecuteContext
    Condition=RigVMDispatchNode(B=@n9.Element, A=&RigUnit_ResolveConnector(Connector=(Type=Connector,Name="Foot Primary"), SkipSocket=true).Result) {template="DISPATCH_RigVMDispatch_CoreEquals(in A,in B,out Result)"}.Result
   )
   &RigUnit_SetDefaultMatch(Default=@n9.Element, ExecutePin=@n10.True) {template="Set Default Match::Execute(in Default)"}
   n11 = &RigUnit_GetCandidates()
   n12 = &RigVMFunction_ControlFlowBranch(
    Condition=RigVMDispatchNode(B=@n11.Connector, A=&RigUnit_Item(Item=(Type=Connector,Name="Parent Control")).Item) {template="DISPATCH_RigVMDispatch_CoreEquals(in A,in B,out Result)"}.Result
    ExecuteContext=@RigVMFunction_Sequence.E
   )
   n13 = RigVMDispatchNode(Array=@n11.Candidates, ExecuteContext=@n12.True) {template="DISPATCH_RigVMDispatch_ArrayIterator(in Array,out Element,out Index,out Count,out Ratio)"}
   n14 = &RigVMFunction_ControlFlowBranch(
    ExecuteContext=@n13.ExecuteContext
    Condition=RigVMDispatchNode(
     B=RigVMDispatchNode(Default=(Type=Bone,Name="None"), Name=Foot Control, NameSpace=Parent) {template="DISPATCH_RigDispatch_GetModuleMetadata(in Name,in NameSpace,in Default,out Value,out Found)"}.Value
     A=@n13.Element
    ) {template="DISPATCH_RigVMDispatch_CoreEquals(in A,in B,out Result)"}.Result
   )
   &RigUnit_SetDefaultMatch(Default=@n13.Element, ExecutePin=@n14.True) {template="Set Default Match::Execute(in Default)"}
  ]}
 RigVMModel Post Forwards Solve Graph {nodes [
   n1 = /Script/ControlRig.RigUnit_PostBeginExecution() {is=struct RigVMUnitNode}
   Get_Metadata = RigVMDispatchNode(Default=None, Item=(Type=Bone,Name="None"), Name=fk_ik_switch, NameSpace=Parent) {template="DISPATCH_RigDispatch_GetMetadata(in Item,in Name,in NameSpace,in Default,out Value,out Found)"}
   DISPATCH_RigDispatch_GetModuleMetadata = RigVMDispatchNode(Default=true, Name=Current FK IK Switch, NameSpace=Parent) {template="DISPATCH_RigDispatch_GetModuleMetadata(in Name,in NameSpace,in Default,out Value,out Found)"}
   Forward Foot = RigVMFunctionReferenceNode(
    IKFK Switch Control=(Type=Control,Name="leg_l_fk_ik_switch")
    Previous IK Mode=false
    Toes Space=(Type=Null,Name="toes_l_space")
    IK Control=RigVMDispatchNode(Default=(Type=Bone,Name="None"), Name=Ball IK Control, NameSpace=Self) {template="DISPATCH_RigDispatch_GetModuleMetadata(in Name,in NameSpace,in Default,out Value,out Found)"}.Value
    FK Control=RigVMDispatchNode(Default=(Type=Bone,Name="None"), Name=Ball FK Control, NameSpace=Self) {template="DISPATCH_RigDispatch_GetModuleMetadata(in Name,in NameSpace,in Default,out Value,out Found)"}.Value
    IKFK Switch Control.Name=@Get_Metadata.Value
    IKFK=RigVMDispatchNode(True=False, False=@DISPATCH_RigDispatch_GetModuleMetadata.Value, Condition=RigVMVariableNode() {variable=Force FK Mode}.Value) {template="DISPATCH_RigVMDispatch_If(in Condition,in True,in False,out Result)"}.Result
    ExecuteContext=@n1.ExecutePin
    Bone=&RigUnit_ResolveConnector(Connector=(Type=Connector,Name="Ball Bone"), SkipSocket=true).Result
   )
  ]}
]
hierarchy [
 Foot Primary {type=Connector, loc=<0 0 0>, rot=<0 0 0 1>}
 Ball Bone {type=Connector, loc=<0 0 0>, rot=<0 0 0 1>}
 Foot Bone {type=Connector, loc=<0 0 0>, rot=<0 0 0 1>}
 Settings Control {type=Connector, loc=<0 0 0>, rot=<0 0 0 1>}
 Parent Control {type=Connector, loc=<0 0 0>, rot=<0 0 0 1>}
]
variables [
 Ball FK Settings {
  type=FRigUnit_HierarchyAddControlTransform_Settings
  default="(InitialSpace=LocalSpace,bUsePreferredRotationOrder=False,PreferredRotationOrder=YZX,Limits=(LimitTranslationX=(bMinimum=False,bMaximum=False),LimitTranslationY=(bMinimum=False,bMaximum=False),LimitTranslationZ=(bMinimum=False,bMaximum=False),LimitPitch=(b ...(truncated)"
 }
 Ball IK Control {type=FRigElementKey, default=(Type=None,Name="")}
 Ball IK Settings {
  type=FRigUnit_HierarchyAddControlTransform_Settings
  default="(InitialSpace=LocalSpace,bUsePreferredRotationOrder=False,PreferredRotationOrder=YZX,Limits=(LimitTranslationX=(bMinimum=False,bMaximum=False),LimitTranslationY=(bMinimum=False,bMaximum=False),LimitTranslationZ=(bMinimum=False,bMaximum=False),LimitPitch=(b ...(truncated)"
 }
 Effector Null {type=FRigElementKey, default=(Type=Bone,Name="")}
 FK Parent {type=FRigElementKey, default=(Type=Bone,Name="")}
 Foot BK Settings {
  type=FRigUnit_HierarchyAddControlTransform_Settings
  default="(InitialSpace=LocalSpace,bUsePreferredRotationOrder=False,PreferredRotationOrder=YZX,Limits=(LimitTranslationX=(bMinimum=False,bMaximum=False),LimitTranslationY=(bMinimum=False,bMaximum=False),LimitTranslationZ=(bMinimum=False,bMaximum=False),LimitPitch=(b ...(truncated)"
 }
 Foot Control {type=FRigElementKey, default=(Type=Bone,Name="")}
 Foot Control Metadata Found {type=bool, default=False}
 Foot Roll Settings {
  type=FRigUnit_HierarchyAddControlTransform_Settings
  default="(InitialSpace=LocalSpace,bUsePreferredRotationOrder=False,PreferredRotationOrder=YZX,Limits=(LimitTranslationX=(bMinimum=False,bMaximum=False),LimitTranslationY=(bMinimum=False,bMaximum=False),LimitTranslationZ=(bMinimum=False,bMaximum=False),LimitPitch=(b ...(truncated)"
 }
 Force FK Mode {type=bool, default=False}
 Heel IK Settings {
  type=FRigUnit_HierarchyAddControlTransform_Settings
  default="(InitialSpace=LocalSpace,bUsePreferredRotationOrder=False,PreferredRotationOrder=YZX,Limits=(LimitTranslationX=(bMinimum=False,bMaximum=False),LimitTranslationY=(bMinimum=False,bMaximum=False),LimitTranslationZ=(bMinimum=False,bMaximum=False),LimitPitch=(b ...(truncated)"
 }
 Module Settings {
  type=FUserDefinedStruct_#6
  default="(OverrideColor_48_#4=False,LeftColor_3_#3=(R=0,G=0.0231,B=0.4115,A=1),RightColor_5_#5=(R=1,G=0,B=0,A=1),CenterColor_9_#2=(R=1,G=1,B=0,A= ...(truncated)"
 }
 Tip Settings {
  type=FRigUnit_HierarchyAddControlTransform_Settings
  default="(InitialSpace=LocalSpace,bUsePreferredRotationOrder=False,PreferredRotationOrder=YZX,Limits=(LimitTranslationX=(bMinimum=False,bMaximum=False),LimitTranslationY=(bMinimum=False,bMaximum=False),LimitTranslationZ=(bMinimum=False,bMaximum=False),LimitPitch=(b ...(truncated)"
 }
 foot bk control {type=FRigElementKey, default=(Type=None,Name="")}
]
