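# Control Rig module that rigs a piston: two bones that keep aiming at each other.
# Construction builds start and end controls or plain nulls, plus an up-vector null with optional control, and Reverse swaps the two ends.
# Forward solve aims each bottom null at the opposite control and constrains the bones to them, with Sticky gating one bone.
# Backwards solve restores controls from tagged item metadata, and connector events suggest default bone and parent matches.
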ControlRigBlueprint /ControlRigModules/Modules58/Piston
graphs [
 Concat_1_1_2_1_ContainedGraph {nodes [
   n1 = RigVMFunctionEntryNode()
   RigVMFunctionReturnNode(Result=/Script/RigVM.RigVMFunction_NameConcat(A=&RigVMFunction_NameConcat(A=@n1.A, B=@n1.B) {template="Concat::Execute(in A,in B,out Result)"}.Result, B=@n1.C) {is=struct RigVMUnitNode, template="Concat::Execute(in A,in B,out Result)"}.Result)
  ]}
 Concat_1_1_2_1_ContainedGraph {nodes [
   n1 = RigVMFunctionEntryNode()
   RigVMFunctionReturnNode(Result=&RigVMFunction_NameConcat(A=&RigVMFunction_NameConcat(A=@n1.A, B=@n1.B) {template="Concat::Execute(in A,in B,out Result)"}.Result, B=@n1.C) {template="Concat::Execute(in A,in B,out Result)"}.Result)
  ]}
 ConstructionGraph {nodes [
   RigUnit_PrepareForExecution = &RigUnit_PrepareForExecution()
   n1 = RigVMVariableNode() {variable=Reverse}
   DISPATCH_RigVMDispatch_If = RigVMDispatchNode(False=(Type=Connector,Name="End Bone"), True=(Type=Connector,Name="Start Bone"), Condition=@n1.Value) {template="DISPATCH_RigVMDispatch_If(in Condition,in True,in False,out Result)"}
   n2 = RigVMVariableNode() {variable=End Control Position}
   Subtract = &RigVMFunction_MathDoubleSub(A=1, B=@n2.Value)
   CONSTR Piston = RigVMFunctionReferenceNode(
    Parent Control Connector=(Type=Connector,Name="Parent Control")
    Parent Up Connector=(Type=Connector,Name="Parent Up")
    End Connector=@DISPATCH_RigVMDispatch_If.Result
    Start Connector=RigVMDispatchNode(False=(Type=Connector,Name="Start Bone"), True=(Type=Connector,Name="End Bone"), Condition=@n1.Value) {template="DISPATCH_RigVMDispatch_If(in Condition,in True,in False,out Result)"}.Result
    Parent Start Connector=RigVMDispatchNode(False=(Type=Connector,Name="Parent Bone Start"), True=(Type=Connector,Name="Parent Bone End"), Condition=@n1.Value) {template="DISPATCH_RigVMDispatch_If(in Condition,in True,in False,out Result)"}.Result
    Parent End Connector=RigVMDispatchNode(False=(Type=Connector,Name="Parent Bone End"), True=(Type=Connector,Name="Parent Bone Start"), Condition=@n1.Value) {template="DISPATCH_RigVMDispatch_If(in Condition,in True,in False,out Result)"}.Result
    End Control Position=RigVMDispatchNode(True=@Subtract.Result, False=@n2.Value, Condition=RigVMVariableNode() {variable=Reverse}.Value) {template="DISPATCH_RigVMDispatch_If(in Condition,in True,in False,out Result)"}.Result
    ExecuteContext=@RigUnit_PrepareForExecution.ExecutePin
    End Control Settings=RigVMVariableNode() {variable=End Control Settings}.Value
    Start Control Settings=RigVMVariableNode() {variable=Start Control Settings}.Value
    Name=RigVMVariableNode() {variable=Name}.Value
    Up Control Settings=RigVMVariableNode() {variable=Up Control Settings}.Value
    As Controls=RigVMVariableNode() {variable=As Controls}.Value
    Module Settings=RigVMVariableNode() {variable=Module Settings}.Value
   )
  ]}
 ContainedGraph {nodes [
   n1 = RigVMFunctionEntryNode()
   > RigVMFunction_Sequence = RigVMAggregateNode()
   DISPATCH_RigDispatch_GetMetadata = RigVMDispatchNode(Default=(Type=Bone,Name="None"), Name=Bottom Null, NameSpace=Self, Item=@n1.Start Ctrl) {template="DISPATCH_RigDispatch_GetMetadata(in Item,in Name,in NameSpace,in Default,out Value,out Found)"}
   RigVMVariableNode(Value=@DISPATCH_RigDispatch_GetMetadata.Value, ExecuteContext=@RigVMFunction_Sequence.A) {variable=Start Bottom Null}
   > RigVMVariableNode(Value=RigVMDispatchNode(Default=(Type=Bone,Name="None"), Name=Bottom Null, NameSpace=Self, Item=@n1.End Ctrl) {template="DISPATCH_RigDispatch_GetMetadata(in Item,in Name,in NameSpace,in Default,out Value,out Found)"}.Value) {variable=End Bottom Null}
   n2 = RigVMVariableNode() {variable=Start Bottom Null}
   > &RigUnit_AimConstraintLocalSpaceOffset(
    AdvancedSettings=(DebugSettings=(bEnabled=false,Scale=10,WorldOffset=(Rotation=(X=0,Y=0,Z=0,W=1),Translation=(X=0,Y=0,Z=0),Scale3D=(X=1,Y=1,Z=1))),RotationOrderForFilter=XZY)
    Filter=(bX=True,bY=True,bZ=True)
    Parents=((Item=(Type=Bone),Weight=1))
    Weight=1
    WorldUp=(Target=(X=0,Y=0,Z=1),Kind=Location,Space=(Type=None,Name="None"))
    Parents.0.Item=@n1.End Ctrl
    bMaintainOffset=@n1.Maintain Offset
    WorldUp.Space=@n1.Up Ctrl
    Child=@n2.Value
    AimAxis=RigVMVariableNode() {variable=Axis}.Value
    UpAxis=RigVMVariableNode() {variable=Up Vector Axis}.Value
   )
   n3 = RigVMVariableNode() {variable=End Bottom Null}
   > &RigUnit_AimConstraintLocalSpaceOffset(
    AdvancedSettings=(DebugSettings=(bEnabled=false,Scale=10,WorldOffset=(Rotation=(X=0,Y=0,Z=0,W=1),Translation=(X=0,Y=0,Z=0),Scale3D=(X=1,Y=1,Z=1))),RotationOrderForFilter=XZY)
    Filter=(bX=True,bY=True,bZ=True)
    Parents=((Item=(Type=Bone),Weight=1))
    Weight=1
    WorldUp=(Target=(X=0,Y=0,Z=1),Kind=Location,Space=(Type=None,Name="None"))
    bMaintainOffset=@n1.Maintain Offset
    Parents.0.Item=@n1.Start Ctrl
    WorldUp.Space=@n1.Up Ctrl
    Child=@n3.Value
    AimAxis=RigVMVariableNode() {variable=Axis}.Value
    UpAxis=RigVMVariableNode() {variable=Up Vector Axis}.Value
   )
   > /Script/ControlRig.RigUnit_PositionConstraintLocalSpaceOffset(
    Filter=(bX=True,bY=True,bZ=True)
    Parents=((Item=(Type=Bone),Weight=1))
    Weight=1
    bMaintainOffset=True
    Child=RigVMVariableNode() {variable=End Bone}.Value
    Parents.0.Item=@n2.Value
   ) {is=struct RigVMUnitNode}
   > &RigUnit_RotationConstraintLocalSpaceOffset(
    AdvancedSettings=(InterpolationType=Average,RotationOrderForFilter=XZY)
    Filter=(bX=True,bY=True,bZ=True)
    Parents=((Item=(Type=Bone,Name="None"),Weight=1))
    Weight=1
    Parents.0.Item=@n2.Value
    Child=RigVMVariableNode() {variable=End Bone}.Value
   )
   > DrawAxis_1 = RigVMFunctionReferenceNode(Active=false, Scale=15, Thickness=2, Item=RigVMVariableNode() {variable=End Bone}.Value)
   n4 = &RigVMFunction_ControlFlowBranch(ExecuteContext=@DrawAxis_1.ExecuteContext, Condition=@n1.Sticky)
   &RigUnit_PositionConstraintLocalSpaceOffset(
    Filter=(bX=True,bY=True,bZ=True)
    Parents=((Item=(Type=Bone),Weight=1))
    Weight=1
    bMaintainOffset=True
    Parents.0.Item=@n1.End Ctrl
    ExecutePin=@n4.True
    Child=RigVMVariableNode() {variable=Start Bone}.Value
   )
   > &RigUnit_RotationConstraintLocalSpaceOffset(
    AdvancedSettings=(InterpolationType=Average,RotationOrderForFilter=XZY)
    Filter=(bX=True,bY=true,bZ=true)
    Parents=((Item=(Type=Bone,Name="None"),Weight=1))
    Weight=1
    Parents.0.Item=@n3.Value
    Child=RigVMVariableNode() {variable=Start Bone}.Value
   )
   RigUnit_SequenceAggregate = &RigVMFunction_Sequence(ExecuteContext=@RigVMFunction_Sequence.C)
   n5 = &RigVMFunction_ControlFlowBranch(Condition=@n1.Debug Draw, ExecuteContext=@RigUnit_SequenceAggregate.A)
   DebugTransformMutableItemSpace_1 = /Script/RigVM.RigVMFunction_DebugTransformMutableNoSpace(
    Color=(R=1,G=0,B=0,A=1)
    DebugDrawSettings=(DepthPriority=SDPG_Foreground,Lifetime=-1)
    Mode=Axes
    Scale=5
    Thickness=0.5
    WorldOffset=(Rotation=(X=0,Y=0,Z=0,W=1),Translation=(X=0,Y=0,Z=0),Scale3D=(X=1,Y=1,Z=1))
    bEnabled=True
    ExecutePin=@n5.True
    Transform=&RigUnit_GetTransform(Item=@n1.Start Bone).Transform
   ) {is=struct RigVMUnitNode}
   > DebugTransformMutableItemSpace_1_1 = &RigVMFunction_DebugTransformMutableNoSpace(
    Color=(R=1,G=0,B=0,A=1)
    DebugDrawSettings=(DepthPriority=SDPG_Foreground,Lifetime=-1)
    Mode=Axes
    Scale=5
    Thickness=0.5
    WorldOffset=(Rotation=(X=0,Y=0,Z=0,W=1),Translation=(X=0,Y=0,Z=0),Scale3D=(X=1,Y=1,Z=1))
    bEnabled=True
    Transform=&RigUnit_GetTransform(Item=@n1.End Bone).Transform
   )
   RigVMFunctionReturnNode(ExecuteContext=@RigUnit_SequenceAggregate.B)
   DrawAxis = RigVMFunctionReferenceNode(Active=true, ExecuteContext=(), Scale=15, Thickness=2, Item=RigVMVariableNode() {variable=End Bone}.Value)
  ]}
 ContainedGraph {nodes [
   n1 = RigVMFunctionEntryNode()
   > n2 = RigVMDispatchNode(Array=@n1.Array) {template="DISPATCH_RigVMDispatch_ArrayIterator(in Array,out Element,out Index,out Count,out Ratio)"}
   LessEqual = /Script/RigVM.RigVMFunction_MathIntLessEqual(A=@n2.Index, B=@n1.Max) {is=struct RigVMUnitNode, template="LessEqual::Execute(in A,in B,out Result)"}
   n3 = &RigVMFunction_ControlFlowBranch(ExecuteContext=@n2.ExecuteContext, Condition=@LessEqual.Result)
   RigVMDispatchNode(Element=@n2.Element, ExecuteContext=@n3.True, Array=RigVMVariableNode() {variable=final_array}.Value) {template="DISPATCH_RigVMDispatch_ArrayAdd(io Array,in Element,out Index)"}
   RigVMFunctionReturnNode(ExecuteContext=@n2.Completed, Value=RigVMVariableNode() {variable=final_array}.Value)
  ]}
 Multiply_1_ContainedGraph {nodes [
   n1 = RigVMFunctionEntryNode()
   Multiply = /Script/RigVM.RigVMFunction_MathVectorMul(A=@n1.A, B=@n1.B) {is=struct RigVMUnitNode, template="Multiply::Execute(in A,in B,out Result)"}
   RigVMFunctionReturnNode(Result=&RigVMFunction_MathVectorMul(B=@n1.C, A=@Multiply.Result) {template="Multiply::Execute(in A,in B,out Result)"}.Result)
  ]}
 New Function_1_ContainedGraph {nodes [
   n1 = RigVMFunctionEntryNode()
   RigVMFunction_Sequence = &RigVMFunction_Sequence(ExecuteContext=@n1.ExecuteContext)
   n2 = /Script/ControlRig.RigUnit_GetCandidates() {is=struct RigVMUnitNode}
   n3 = &RigVMFunction_ControlFlowBranch(
    Condition=RigVMDispatchNode(B=@n1.Connector, A=@n2.Connector) {template="DISPATCH_RigVMDispatch_CoreEquals(in A,in B,out Result)"}.Result
    ExecuteContext=@RigVMFunction_Sequence.A
   )
   RigVMDispatchNode(Array=@n2.Candidates, ExecuteContext=@n3.True) {template="DISPATCH_RigVMDispatch_ArrayIterator(in Array,out Element,out Index,out Count,out Ratio)"}
   n4 = /Script/ControlRig.RigUnit_CollectionChildrenArray(
    TypeToSearch=Bone
    bDefaultChildren=True
    bIncludeParent=true
    bRecursive=true
    Parent=/Script/ControlRig.RigUnit_ResolveConnector(SkipSocket=true, Connector=@n1.Primary Connector) {is=struct RigVMUnitNode}.Result
   ) {is=struct RigVMUnitNode}
   > Clamp Array = RigVMFunctionReferenceNode(Array=@n4.Items, Max=@n1.Clamp)
   > /Script/ControlRig.RigUnit_SetDefaultMatch(Default=RigVMDispatchNode(Index=-1, Array=@n4.Items) {template="DISPATCH_RigVMDispatch_ArrayGetAtIndex(in Array,in Index,out Element)"}.Element) {is=struct RigVMUnitNode, template="Set Default Match::Execute(in Default)"}
   RigVMFunctionReturnNode(ExecuteContext=@RigVMFunction_Sequence.B)
  ]}
 New Function_ContainedGraph {nodes [
   n1 = RigVMFunctionEntryNode()
   > RigVMFunction_Sequence_1 = RigVMAggregateNode()
   RigVMFunction_Sequence = RigVMAggregateNode(ExecuteContext=@RigVMFunction_Sequence_1.B)
   RigVMVariableNode(
    Value=RigVMDispatchNode(Default=(Type=Bone,Name="None"), Name=Start Control, NameSpace=Self) {template="DISPATCH_RigDispatch_GetModuleMetadata(in Name,in NameSpace,in Default,out Value,out Found)"}.Value
    ExecuteContext=@RigVMFunction_Sequence.A
   ) {variable=start ctrl}
   > RigVMVariableNode(Value=&RigUnit_ResolveConnector(SkipSocket=true, Connector=RigVMVariableNode() {variable=Start Connector}.Value).Result) {variable=start bone}
   RigVMVariableNode(
    Value=RigVMDispatchNode(Default=(Type=Bone,Name="None"), Name=End Control, NameSpace=Self) {template="DISPATCH_RigDispatch_GetModuleMetadata(in Name,in NameSpace,in Default,out Value,out Found)"}.Value
    ExecuteContext=@RigVMFunction_Sequence.B
   ) {variable=end ctrl}
   > RigVMVariableNode(Value=&RigUnit_ResolveConnector(SkipSocket=true, Connector=RigVMVariableNode() {variable=End Connector}.Value).Result) {variable=end bone}
   Build Piston = RigVMFunctionReferenceNode(
    Maintain Offset=true
    Axis=@n1.Axis
    Debug Draw=@n1.Debug Draw
    Sticky=@n1.Sticky
    Up Vector Axis=@n1.Up Vector Axis
    ExecuteContext=@RigVMFunction_Sequence.C
    Start Ctrl=RigVMVariableNode() {variable=start ctrl}.Value
    End Ctrl=RigVMVariableNode() {variable=end ctrl}.Value
    Start Bone=RigVMVariableNode() {variable=end bone}.Value
    End Bone=RigVMVariableNode() {variable=start bone}.Value
    Up Ctrl=RigVMVariableNode() {variable=Up Control}.Value
   )
   n2 = &RigUnit_ResolveConnector(Connector=(Type=Connector,Name="Parent Bone End"), SkipSocket=true)
   RigVMFunction_ControlFlowBranch = &RigVMFunction_ControlFlowBranch(ExecuteContext=@"Build Piston".ExecuteContext, Condition=@n2.bIsConnected)
   n3 = &RigUnit_ResolveConnector(Connector=(Type=Connector,Name="Parent Bone Start"), SkipSocket=true)
   n4 = &RigVMFunction_ControlFlowBranch(Condition=@n3.bIsConnected, ExecuteContext=@RigVMFunction_ControlFlowBranch.Completed)
   Parent_1 = RigVMFunctionReferenceNode(
    Child=RigVMDispatchNode(Default=(Type=Bone,Name="None"), Name=Null, NameSpace=Self, Item=RigVMVariableNode() {variable=start ctrl}.Value) {template="DISPATCH_RigDispatch_GetMetadata(in Item,in Name,in NameSpace,in Default,out Value,out Found)"}.Value
    New Parent=@n3.Result
    ExecuteContext=@n4.True
   )
   Parent = RigVMFunctionReferenceNode(
    Child=RigVMDispatchNode(Default=(Type=Bone,Name="None"), Name=Null, NameSpace=Self, Item=RigVMVariableNode() {variable=end ctrl}.Value) {template="DISPATCH_RigDispatch_GetMetadata(in Item,in Name,in NameSpace,in Default,out Value,out Found)"}.Value
    New Parent=@n2.Result
    ExecuteContext=@RigVMFunction_ControlFlowBranch.True
   )
   RigVMFunctionReturnNode(ExecuteContext=@RigVMFunction_Sequence_1.C)
   DrawItems = RigVMFunctionReferenceNode(
    Active=false
    ExecuteContext=()
    Items=((Type=None,Name="None"),(Type=None,Name="None"))
    Scale=10
    Thickness=1
    Items.0=RigVMVariableNode() {variable=start ctrl}.Value
    Items.1=RigVMVariableNode() {variable=end ctrl}.Value
   )
  ]}
 New Function_ContainedGraph {nodes [
   n1 = RigVMFunctionEntryNode()
   DISPATCH_RigVMDispatch_CoreEquals = RigVMDispatchNode(B=None, A=@n1.Local Suffix) {template="DISPATCH_RigVMDispatch_CoreEquals(in A,in B,out Result)"}
   DISPATCH_RigVMDispatch_If = RigVMDispatchNode(Condition=@DISPATCH_RigVMDispatch_CoreEquals.Result, True=@n1.Global Suffix, False=@n1.Local Suffix) {template="DISPATCH_RigVMDispatch_If(in Condition,in True,in False,out Result)"}
   > RigVMFunctionReturnNode(Valid Suffix=@DISPATCH_RigVMDispatch_If.Result)
  ]}
 New Function_ContainedGraph {nodes [
   n1 = RigVMFunctionEntryNode()
   > RigVMFunction_Sequence_1 = RigVMAggregateNode()
   DISPATCH_RigVMDispatch_CoreEquals = RigVMDispatchNode(B=None, A=@n1.Name) {template="DISPATCH_RigVMDispatch_CoreEquals(in A,in B,out Result)"}
   n2 = &RigVMFunction_ControlFlowBranch(Condition=@DISPATCH_RigVMDispatch_CoreEquals.Result, ExecuteContext=@RigVMFunction_Sequence_1.A)
   RigVMVariableNode(Value=@n1.Name, ExecuteContext=@n2.False) {variable=prefix}
   RigVMVariableNode(Value=&RigUnit_ResolveConnector(SkipSocket=true, Connector=@n1.Start Connector).Result.Name, ExecuteContext=@n2.True) {variable=prefix}
   RigVMFunction_Sequence = RigVMAggregateNode(ExecuteContext=@RigVMFunction_Sequence_1.B)
   n3 = &RigVMFunction_ControlFlowBranch(ExecuteContext=@RigVMFunction_Sequence.A, Condition=RigVMVariableNode() {variable=As Controls}.Value)
   Concat_1_1_2 = RigVMAggregateNode(B=_start, C=_ctrl, A=RigVMVariableNode() {variable=prefix}.Value)
   n4 = RigVMVariableNode() {variable=Start Connector}
   n5 = &RigUnit_HierarchyAddNull(
    Space=GlobalSpace
    Name=@Concat_1_1_2.Result
    Transform=&RigUnit_GetTransform(bInitial=true, Item=&RigUnit_ResolveConnector(SkipSocket=true, Connector=@n4.Value).Result).Transform
    Parent=&RigUnit_ResolveConnector(SkipSocket=true, Connector=RigVMVariableNode() {variable=Parent Control Connector}.Value).Result
    ExecutePin=@n3.False
   )
   > RigVMDispatchNode(Name=Start Control, NameSpace=Self, Value=@n5.Item) {template="DISPATCH_RigDispatch_SetModuleMetadata(in Name,in NameSpace,in Value,out Success)"}
   > RigVMVariableNode(Value=@n5.Item) {variable=Start Control}
   > Add Null Above_1 = RigVMFunctionReferenceNode(New Suffix=_null, Replace Suffix=_ctrl, Item=@n5.Item)
   > RigVMDispatchNode(Name=Null, NameSpace=Self, Value=@"Add Null Above_1".Null, Item=@n5.Item) {template="DISPATCH_RigDispatch_SetMetadata(in Item,in Name,in NameSpace,in Value,out Success)"}
   > Add Null Below_1 = RigVMFunctionReferenceNode(
    New Suffix=_ctrl_null
    Replace Suffix=_ctrl
    Transform=(Rotation=(X=0,Y=0,Z=0,W=1),Translation=(X=0,Y=0,Z=0),Scale3D=(X=1,Y=1,Z=1))
    Use Transform=False
    Item=@n5.Item
   )
   > RigVMDispatchNode(Name=Bottom Null, NameSpace=Self, Value=@"Add Null Below_1".Null, Item=@n5.Item) {template="DISPATCH_RigDispatch_SetMetadata(in Item,in Name,in NameSpace,in Value,out Success)"}
   Control Stack at Item_1 = RigVMFunctionReferenceNode(
    Add Bottom Null=true
    Add Null=true
    Add Secondary Control=False
    Control Suffix=_ctrl
    Null Suffix=_null
    Secondary Settings="(DisplayName=\"None\",Shape=(bVisible=True,Name=\"Default\",Color=(R=1,G=0,B=0,A=1),Transform=(Rotation=(X=0,Y=0,Z=0,W=1),Translation=(X=0,Y=0,Z=0),Scale3D=(X=1,Y=1,Z=1))),Proxy=(bIsProxy=False,ShapeVisibility=BasedOnSelection),InitialSpace=LocalSpace,bUsePref ...(truncated)"
    Tag Controls=true
    Zero Controls=true
    Name=&RigVMFunction_NameConcat(B=_start, A=RigVMVariableNode() {variable=prefix}.Value) {template="Concat::Execute(in A,in B,out Result)"}.Result
    Settings=@n1.Start Control Settings
    Item=&RigUnit_ResolveConnector(SkipSocket=true, Connector=RigVMVariableNode() {variable=Start Connector}.Value).Result
    Parent=&RigUnit_ResolveConnector(SkipSocket=true, Connector=RigVMVariableNode() {variable=Parent Control Connector}.Value).Result
    ExecuteContext=@n3.True
   )
   > RigVMDispatchNode(Name=Start Control, NameSpace=Self, Value=@"Control Stack at Item_1".Control) {template="DISPATCH_RigDispatch_SetModuleMetadata(in Name,in NameSpace,in Value,out Success)"}
   > RigVMVariableNode(Value=@"Control Stack at Item_1".Control) {variable=Start Control}
   n6 = &RigVMFunction_ControlFlowBranch(ExecuteContext=@RigVMFunction_Sequence.C, Condition=RigVMVariableNode() {variable=As Controls}.Value)
   Concat_1_1_2_1 = RigVMAggregateNode(B=_end, C=_ctrl, A=RigVMVariableNode() {variable=prefix}.Value)
   n7 = RigVMVariableNode() {variable=End Connector}
   n8 = &RigUnit_HierarchyAddNull(
    Space=GlobalSpace
    Name=@Concat_1_1_2_1.Result
    Transform=&RigUnit_GetTransform(bInitial=true, Item=&RigUnit_ResolveConnector(SkipSocket=true, Connector=@n7.Value).Result).Transform
    Parent=&RigUnit_ResolveConnector(SkipSocket=true, Connector=RigVMVariableNode() {variable=Parent Control Connector}.Value).Result
    ExecutePin=@n6.False
   )
   > RigVMDispatchNode(Name=End Control, NameSpace=Self, Value=@n8.Item) {template="DISPATCH_RigDispatch_SetModuleMetadata(in Name,in NameSpace,in Value,out Success)"}
   > Add Null Above = RigVMFunctionReferenceNode(New Suffix=_null, Replace Suffix=_ctrl, Item=@n8.Item)
   > DISPATCH_RigDispatch_SetMetadata = RigVMDispatchNode(Name=Null, NameSpace=Self, Value=@"Add Null Above".Null, Item=@n8.Item) {template="DISPATCH_RigDispatch_SetMetadata(in Item,in Name,in NameSpace,in Value,out Success)"}
   > Add Null Below = RigVMFunctionReferenceNode(
    New Suffix=_ctrl_null
    Replace Suffix=_ctrl
    Transform=(Rotation=(X=0,Y=0,Z=0,W=1),Translation=(X=0,Y=0,Z=0),Scale3D=(X=1,Y=1,Z=1))
    Use Transform=False
    Item=@n8.Item
   )
   > RigVMDispatchNode(Name=Bottom Null, NameSpace=Self, Value=@"Add Null Below".Null, Item=@n8.Item) {template="DISPATCH_RigDispatch_SetMetadata(in Item,in Name,in NameSpace,in Value,out Success)"}
   > RigVMDispatchNode(Name=Parent Control, NameSpace=Self, Value=@n8.Item) {template="DISPATCH_RigDispatch_SetModuleMetadata(in Name,in NameSpace,in Value,out Success)"}
   > RigVMDispatchNode(Name=FK Parent, NameSpace=Self, Value=@n8.Item) {template="DISPATCH_RigDispatch_SetModuleMetadata(in Name,in NameSpace,in Value,out Success)"}
   n9 = &RigUnit_GetTransform(
    bInitial=False
    Item=&RigUnit_ResolveConnector(SkipSocket=true, Connector=RigVMVariableNode() {variable=Start Connector}.Value).Result
   )
   Control Stack at Position = RigVMFunctionReferenceNode(
    Add Bottom Null=true
    Add Null=true
    Add Secondary Control=False
    Control Suffix=_ctrl
    Global Position=(Rotation=(X=0,Y=0,Z=0,W=1),Translation=(X=0,Y=0,Z=0),Scale3D=(X=1,Y=1,Z=1))
    Null Suffix=_null
    Secondary Settings="(DisplayName=\"None\",Shape=(bVisible=True,Name=\"Default\",Color=(R=1,G=0,B=0,A=1),Transform=(Rotation=(X=0,Y=0,Z=0,W=1),Translation=(X=0,Y=0,Z=0),Scale3D=(X=1,Y=1,Z=1))),Proxy=(bIsProxy=False,ShapeVisibility=BasedOnSelection),InitialSpace=LocalSpace,bUsePref ...(truncated)"
    Tag Controls=true
    Zero Controls=true
    Name=&RigVMFunction_NameConcat(B=_end, A=RigVMVariableNode() {variable=prefix}.Value) {template="Concat::Execute(in A,in B,out Result)"}.Result
    Global Position.Rotation=@n9.Transform.Rotation
    Global Position.Scale3D=@n9.Transform.Scale3D
    Parent=&RigUnit_ResolveConnector(SkipSocket=true, Connector=RigVMVariableNode() {variable=Parent Control Connector}.Value).Result
    ExecuteContext=@n6.True
    Global Position.Translation=/Script/RigVM.RigVMFunction_MathTransformTransformVector(
     Transform=@n9.Transform
     Location=&RigVMFunction_MathVectorMul(
      B=(X=25,Y=25,Z=25)
      B.X=@n1.End Control Position
      B.Y=@n1.End Control Position
      B.Z=@n1.End Control Position
      A=@n1.Module Settings.PrimaryBoneAxis_32_#1
     ) {template="Multiply::Execute(in A,in B,out Result)"}.Result
    ) {is=struct RigVMUnitNode}.Result
    Settings=RigVMVariableNode() {variable=End Control Settings}.Value
   )
   > RigVMDispatchNode(Name=End Control, NameSpace=Self, Value=@"Control Stack at Position".Control) {template="DISPATCH_RigDispatch_SetModuleMetadata(in Name,in NameSpace,in Value,out Success)"}
   > RigVMDispatchNode(Name=Parent Control, NameSpace=Self, Value=@"Control Stack at Position".Control) {template="DISPATCH_RigDispatch_SetModuleMetadata(in Name,in NameSpace,in Value,out Success)"}
   > RigVMDispatchNode(Name=FK Parent, NameSpace=Self, Value=@"Control Stack at Position".Control) {template="DISPATCH_RigDispatch_SetModuleMetadata(in Name,in NameSpace,in Value,out Success)"}
   Finalize Controls = RigVMCollapseNode(ExecuteContext=@RigVMFunction_Sequence.D, Module Settings=RigVMVariableNode() {variable=Module Settings}.Value)
   n10 = &RigUnit_CollectionChildrenArray(
    TypeToSearch=All
    bDefaultChildren=True
    bIncludeParent=False
    bRecursive=False
    Parent=&RigUnit_ResolveConnector(SkipSocket=true, Connector=RigVMVariableNode() {variable=End Connector}.Value).Result
   )
   DISPATCH_RigVMDispatch_ArrayGetNum = RigVMDispatchNode(Array=@n10.Items) {template="DISPATCH_RigVMDispatch_ArrayGetNum(in Array,out Num)"}
   Greater = /Script/RigVM.RigVMFunction_MathIntGreater(B=0, A=@DISPATCH_RigVMDispatch_ArrayGetNum.Num) {is=struct RigVMUnitNode, template="Greater::Execute(in A,in B,out Result)"}
   RigVMFunction_ControlFlowBranch = &RigVMFunction_ControlFlowBranch(Condition=@Greater.Result, ExecuteContext=@RigVMFunction_Sequence.E)
   DISPATCH_RigVMDispatch_ArrayGetAtIndex = RigVMDispatchNode(Index=0, Array=@n10.Items) {template="DISPATCH_RigVMDispatch_ArrayGetAtIndex(in Array,in Index,out Element)"}
   Concat = &RigVMFunction_NameConcat(B=_socket, A=@DISPATCH_RigVMDispatch_ArrayGetAtIndex.Element.Name) {template="Concat::Execute(in A,in B,out Result)"}
   /Script/ControlRig.RigUnit_HierarchyAddSocket(
    Color=(R=1,G=1,B=1,A=1)
    Space=LocalSpace
    Transform=(Rotation=(X=0,Y=0,Z=0,W=1),Translation=(X=0,Y=0,Z=0),Scale3D=(X=1,Y=1,Z=1))
    Name=@Concat.Result
    Parent=@DISPATCH_RigVMDispatch_ArrayGetAtIndex.Element
    ExecutePin=@RigVMFunction_ControlFlowBranch.True
   ) {is=struct RigVMUnitNode}
   n11 = RigVMVariableNode() {variable=Start Control}
   n12 = &RigUnit_GetTransform(bInitial=true, Item=@n11.Value)
   n13 = RigVMVariableNode() {variable=Up Vector Offset Factor}
   Multiply = RigVMAggregateNode(
    C=(X=0,Y=0,Z=0)
    A=@n12.Transform.Translation
    B=RigVMVariableNode() {variable=Up Vector Axis}.Value
    C.X=@n13.Value
    C.Y=@n13.Value
    C.Z=@n13.Value
   )
   RigVMFunction_MathTransformTransformVector = &RigVMFunction_MathTransformTransformVector(Transform=@n12.Transform, Location=@Multiply.Result)
   n14 = &RigUnit_HierarchyAddNull(
    Parent=(Type=Bone,Name="None")
    Space=LocalSpace
    Transform=(Rotation=(X=0,Y=0,Z=0,W=1),Translation=(X=0,Y=0,Z=0),Scale3D=(X=1,Y=1,Z=1))
    Name=&RigVMFunction_NameConcat(B=_up, A=RigVMVariableNode() {variable=prefix}.Value) {template="Concat::Execute(in A,in B,out Result)"}.Result
    Transform.Rotation=@n12.Transform.Rotation
    Transform.Scale3D=@n12.Transform.Scale3D
    Transform.Translation=@RigVMFunction_MathTransformTransformVector.Result
    ExecutePin=@RigVMFunction_Sequence.B
   )
   > /Script/ControlRig.RigUnit_SetDefaultParent(Child=@n14.Item, Parent=@n11.Value) {is=struct RigVMUnitNode}
   > RigVMVariableNode(Value=@n14.Item) {variable=Up Control}
   > DISPATCH_RigDispatch_SetModuleMetadata = RigVMDispatchNode(Name=Up Null, NameSpace=Self, Value=@n14.Item) {template="DISPATCH_RigDispatch_SetModuleMetadata(in Name,in NameSpace,in Value,out Success)"}
   n15 = &RigVMFunction_ControlFlowBranch(ExecuteContext=@DISPATCH_RigDispatch_SetModuleMetadata.ExecuteContext, Condition=RigVMVariableNode() {variable=Add Up Vector Ctrl}.Value)
   n16 = &RigUnit_ResolveConnector(SkipSocket=true, Connector=RigVMVariableNode() {variable=Parent Up Connector}.Value)
   DISPATCH_RigVMDispatch_If = RigVMDispatchNode(True=@n16.Result, Condition=@n16.bIsConnected, False=RigVMVariableNode() {variable=Start Control}.Value) {template="DISPATCH_RigVMDispatch_If(in Condition,in True,in False,out Result)"}
   Control Stack at Item_2 = RigVMFunctionReferenceNode(
    Add Bottom Null=False
    Add Null=False
    Add Secondary Control=False
    Control Suffix=_ctrl
    Name=up
    Null Suffix=_null
    Secondary Settings="(DisplayName=\"None\",Shape=(bVisible=True,Name=\"Default\",Color=(R=1,G=0,B=0,A=1),Transform=(Rotation=(X=0,Y=0,Z=0,W=1),Translation=(X=0,Y=0,Z=0),Scale3D=(X=1,Y=1,Z=1))),Proxy=(bIsProxy=False,ShapeVisibility=BasedOnSelection),InitialSpace=LocalSpace,bUsePref ...(truncated)"
    Settings="(DisplayName=\"None\",Shape=(bVisible=True,Name=\"Default\",Color=(R=1,G=0,B=0,A=1),Transform=(Rotation=(X=0,Y=0,Z=0,W=1),Translation=(X=0,Y=0,Z=0),Scale3D=(X=1,Y=1,Z=1))),Proxy=(bIsProxy=False,ShapeVisibility=BasedOnSelection),InitialSpace=LocalSpace,bUsePref ...(truncated)"
    Tag Controls=true
    Zero Controls=true
    Parent=@DISPATCH_RigVMDispatch_If.Result
    Item=@n14.Item
    ExecuteContext=@n15.True
   )
   > &RigUnit_SetDefaultParent(Parent=@"Control Stack at Item_2".Control, Child=RigVMVariableNode() {variable=Up Control}.Value)
   RigVMFunctionReturnNode(ExecuteContext=@RigVMFunction_Sequence_1.C)
   RigVMFunction_MathBoolNot = &RigVMFunction_MathBoolNot(Value=False)
   n17 = &RigVMFunction_ControlFlowBranch(ExecuteContext=(), Condition=@RigVMFunction_MathBoolNot.Result)
   n18 = RigVMVariableNode() {variable=Start Connector}
   n19 = &RigUnit_HierarchyAddNull(
    Space=LocalSpace
    Name=&RigVMFunction_NameConcat(B=_start_null, A=RigVMVariableNode() {variable=prefix}.Value) {template="Concat::Execute(in A,in B,out Result)"}.Result
    Transform=&RigUnit_GetTransform(bInitial=False, Item=&RigUnit_ResolveConnector(SkipSocket=true, Connector=@n18.Value).Result).Transform
    Parent=&RigUnit_ResolveConnector(SkipSocket=true, Connector=RigVMVariableNode() {variable=Parent End Connector}.Value).Result
    ExecutePin=@n17.True
   )
   > RigVMDispatchNode(Name=Start Control, NameSpace=Self, Value=@n19.Item) {template="DISPATCH_RigDispatch_SetModuleMetadata(in Name,in NameSpace,in Value,out Success)"}
   > RigVMVariableNode(Value=@n19.Item) {variable=Start Control}
  ]}
 New Function_ContainedGraph {nodes [
   n1 = RigVMFunctionEntryNode()
   n2 = /Script/ControlRig.RigUnit_FindItemsWithMetadataTag(NameSpace=Self, Tag=Controls) {is=struct RigVMUnitNode}
   > Scale Module Controls = RigVMFunctionReferenceNode(
    Global Scale=RigVMDispatchNode(Default=1, Name=Control Size, NameSpace=Root) {template="DISPATCH_RigDispatch_GetModuleMetadata(in Name,in NameSpace,in Default,out Value,out Found)"}.Value
    Local Scale=@n1.Module Settings.ControlSize_12_#2
    Controls=@n2.Items
   )
   > Color Module Controls = RigVMFunctionReferenceNode(
    CenterColor=@n1.Module Settings.CenterColor_9_#3
    LeftColor=@n1.Module Settings.LeftColor_3_#4
    Override=@n1.Module Settings.OverrideColor_48_#5
    RightColor=@n1.Module Settings.RightColor_5_#6
    Controls=@n2.Items
   )
   > RigVMFunctionReturnNode()
  ]}
 New Function_ContainedGraph {nodes [
   RigVMFunctionEntryNode()
   > RigVMFunction_Sequence = RigVMAggregateNode()
   DISPATCH_RigVMDispatch_ArrayIterator = RigVMDispatchNode(Array=&RigUnit_FindItemsWithMetadataTag(NameSpace=Self, Tag=Controls).Items, ExecuteContext=@RigVMFunction_Sequence.A) {template="DISPATCH_RigVMDispatch_ArrayIterator(in Array,out Element,out Index,out Count,out Ratio)"}
   DISPATCH_RigDispatch_GetMetadata = RigVMDispatchNode(Default=(Type=Bone,Name="None"), Name=Item, NameSpace=Self, Item=@DISPATCH_RigVMDispatch_ArrayIterator.Element) {template="DISPATCH_RigDispatch_GetMetadata(in Item,in Name,in NameSpace,in Default,out Value,out Found)"}
   RigVMFunction_ControlFlowBranch = &RigVMFunction_ControlFlowBranch(Condition=@DISPATCH_RigDispatch_GetMetadata.Found, ExecuteContext=@DISPATCH_RigVMDispatch_ArrayIterator.ExecuteContext)
   &RigUnit_SetTransform(
    Weight=1
    bInitial=False
    Item=@DISPATCH_RigVMDispatch_ArrayIterator.Element
    Value=/Script/ControlRig.RigUnit_ProjectTransformToNewParent(
     bChildInitial=True
     bNewParentInitial=False
     bOldParentInitial=True
     NewParent=@DISPATCH_RigDispatch_GetMetadata.Value
     OldParent=@DISPATCH_RigDispatch_GetMetadata.Value
     Child=@DISPATCH_RigVMDispatch_ArrayIterator.Element
    ) {is=struct RigVMUnitNode}.Transform
    ExecutePin=@RigVMFunction_ControlFlowBranch.True
   )
   RigVMFunctionReturnNode(ExecuteContext=@RigVMFunction_Sequence.C)
  ]}
 RigVMFunctionLibrary {nodes [
   Build Piston = RigVMCollapseNode(
    Axis=(X=1,Y=0,Z=0)
    Debug Draw=true
    End Bone=()
    End Ctrl=()
    ExecuteContext=()
    Maintain Offset=true
    Start Bone=(Type=Bone)
    Start Ctrl=()
    Sticky=true
    Up Ctrl=(Type=Control)
    Up Vector Axis=(X=0,Y=0,Z=1)
   )
   CONSTR Piston = RigVMCollapseNode(
    As Controls=true
    End Connector=(Type=Connector,Name="End")
    End Control Settings="(DisplayName=\"None\",bSelectable=True,InitialSpace=LocalSpace,bUsePreferredRotationOrder=False,PreferredRotationOrder=YZX,Limits=(LimitTranslationX=(bMinimum=False,bMaximum=False),LimitTranslationY=(bMinimum=False,bMaximum=False),LimitTranslationZ=(bMinimum ...(truncated)"
    ExecuteContext=()
    Module Settings="(OverrideColor_48_#5=False,LeftColor_3_#4=(R=0,G=0.0231,B=0.4115,A=1),RightColor_5_#6=(R=1,G=0,B=0,A=1),CenterColor_9_#3=(R=1,G=1,B=0,A= ...(truncated)"
    Parent Control Connector=(Type=Connector,Name="Parent Control")
    Parent End Connector=(Type=Connector,Name="Parent Control")
    Parent Start Connector=(Type=Connector,Name="Parent Control")
    Parent Up Connector=(Type=Connector,Name="Parent Control")
    Start Connector=(Type=Connector,Name="Start")
    Start Control Settings="(DisplayName=\"None\",bSelectable=True,InitialSpace=LocalSpace,bUsePreferredRotationOrder=False,PreferredRotationOrder=YZX,Limits=(LimitTranslationX=(bMinimum=False,bMaximum=False),LimitTranslationY=(bMinimum=False,bMaximum=False),LimitTranslationZ=(bMinimum ...(truncated)"
    Up Control Settings="(DisplayName=\"None\",bSelectable=True,InitialSpace=LocalSpace,bUsePreferredRotationOrder=False,PreferredRotationOrder=YZX,Limits=(LimitTranslationX=(bMinimum=False,bMaximum=False),LimitTranslationY=(bMinimum=False,bMaximum=False),LimitTranslationZ=(bMinimum ...(truncated)"
   )
   Clamp Array = RigVMCollapseNode(Array=(), ExecuteContext=(), Max=0)
   Connect to Leaf Bones = RigVMCollapseNode(Clamp=0, Connector=(Type=Connector,Name="End"), ExecuteContext=(), Primary Connector=(Type=Connector,Name="Finger"))
   FWD Piston = RigVMCollapseNode(
    Axis=(X=1,Y=0,Z=0)
    Debug Draw=true
    End Connector=(Type=Connector,Name="End")
    ExecuteContext=()
    Start Connector=(Type=Connector,Name="Start")
    Sticky=true
    Up Vector Axis=(X=0,Y=0,Z=1)
   )
   Get Module Suffix = RigVMCollapseNode(ExecuteContext=(), Global Suffix=Suffix, Local Suffix=None)
   INV Piston = RigVMCollapseNode(ExecuteContext=())
  ]}
 RigVMFunction_Sequence_1_1_ContainedGraph {nodes [
   n1 = RigVMFunctionEntryNode()
   n2 = &RigVMFunction_Sequence(ExecuteContext=@n1.ExecuteContext)
   n3 = &RigVMFunction_Sequence(ExecuteContext=@n2.B)
   RigVMFunctionReturnNode(A=@n2.A, B=@n3.A, C=@n3.B)
  ]}
 RigVMFunction_Sequence_1_1_ContainedGraph {nodes [
   n1 = RigVMFunctionEntryNode()
   n2 = &RigVMFunction_Sequence(ExecuteContext=@n1.ExecuteContext)
   n3 = &RigVMFunction_Sequence(ExecuteContext=@n2.B)
   RigVMFunctionReturnNode(A=@n2.A, B=@n3.A, C=@n3.B)
  ]}
 RigVMFunction_Sequence_1_ContainedGraph {nodes [
   n1 = RigVMFunctionEntryNode()
   RigVMFunction_Sequence = &RigVMFunction_Sequence(ExecuteContext=@n1.ExecuteContext)
   n2 = &RigVMFunction_Sequence(ExecuteContext=@RigVMFunction_Sequence.B)
   n3 = &RigVMFunction_Sequence(ExecuteContext=@n2.B)
   RigVMFunctionReturnNode(A=@RigVMFunction_Sequence.A, B=@n2.A, C=@n3.A, D=@n3.B)
  ]}
 RigVMFunction_Sequence_1_ContainedGraph {nodes [
   n1 = RigVMFunctionEntryNode()
   RigVMFunction_Sequence = &RigVMFunction_Sequence(ExecuteContext=@n1.ExecuteContext)
   n2 = &RigVMFunction_Sequence(ExecuteContext=@RigVMFunction_Sequence.B)
   RigVMFunctionReturnNode(A=@RigVMFunction_Sequence.A, B=@n2.A, C=@n2.B)
  ]}
 RigVMFunction_Sequence_1_ContainedGraph {nodes [
   n1 = RigVMFunctionEntryNode()
   RigVMFunction_Sequence = &RigVMFunction_Sequence(ExecuteContext=@n1.ExecuteContext)
   n2 = &RigVMFunction_Sequence(ExecuteContext=@RigVMFunction_Sequence.B)
   n3 = &RigVMFunction_Sequence(ExecuteContext=@n2.B)
   n4 = &RigVMFunction_Sequence(ExecuteContext=@n3.B)
   RigVMFunctionReturnNode(A=@RigVMFunction_Sequence.A, B=@n2.A, C=@n3.A, D=@n4.A, E=@n4.B)
  ]}
 RigVMFunction_Sequence_1_ContainedGraph {nodes [
   n1 = RigVMFunctionEntryNode()
   RigVMFunction_Sequence = &RigVMFunction_Sequence(ExecuteContext=@n1.ExecuteContext)
   n2 = &RigVMFunction_Sequence(ExecuteContext=@RigVMFunction_Sequence.B)
   RigVMFunctionReturnNode(A=@RigVMFunction_Sequence.A, B=@n2.A, C=@n2.B)
  ]}
 RigVMFunction_Sequence_2_ContainedGraph {nodes [
   n1 = RigVMFunctionEntryNode()
   RigVMFunction_Sequence = &RigVMFunction_Sequence(ExecuteContext=@n1.ExecuteContext)
   n2 = &RigVMFunction_Sequence(ExecuteContext=@RigVMFunction_Sequence.B)
   RigVMFunctionReturnNode(A=@RigVMFunction_Sequence.A, B=@n2.A, C=@n2.B)
  ]}
 RigVMModel {nodes [
   RigUnit_BeginExecution = &RigUnit_BeginExecution()
   n1 = RigVMVariableNode() {variable=Reverse}
   DISPATCH_RigVMDispatch_If = RigVMDispatchNode(False=(Type=Connector,Name="End Bone"), True=(Type=Connector,Name="Start Bone"), Condition=@n1.Value) {template="DISPATCH_RigVMDispatch_If(in Condition,in True,in False,out Result)"}
   FWD Piston = RigVMFunctionReferenceNode(
    End Connector=@DISPATCH_RigVMDispatch_If.Result
    Start Connector=RigVMDispatchNode(False=(Type=Connector,Name="Start Bone"), True=(Type=Connector,Name="End Bone"), Condition=@n1.Value) {template="DISPATCH_RigVMDispatch_If(in Condition,in True,in False,out Result)"}.Result
    ExecuteContext=@RigUnit_BeginExecution.ExecutePin
    Sticky=RigVMVariableNode() {variable=Sticky}.Value
    Axis=RigVMVariableNode() {variable=Module Settings}.Value.PrimaryBoneAxis_32_#1
    Up Vector Axis=RigVMVariableNode() {variable=Up Vector Axis}.Value
    Debug Draw=RigVMVariableNode() {variable=Debug}.Value
   )
  ]}
 RigVMModel Backwards Solve Graph {nodes [
   n1 = /Script/ControlRig.RigUnit_InverseExecution() {is=struct RigVMUnitNode}
   INV Piston = RigVMFunctionReferenceNode(ExecuteContext=@n1.ExecutePin)
  ]}
 RigVMModel Connector Event Graph {nodes [
   n1 = /Script/ControlRig.RigUnit_ConnectorExecution() {is=struct RigVMUnitNode}
   RigVMFunction_Sequence = RigVMAggregateNode(ExecuteContext=@n1.ExecuteContext)
   n2 = &RigUnit_GetCandidates()
   n3 = &RigVMFunction_ControlFlowBranch(
    Condition=RigVMDispatchNode(B=(Type=Connector,Name="Start Bone"), A=@n2.Connector) {template="DISPATCH_RigVMDispatch_CoreEquals(in A,in B,out Result)"}.Result
    ExecuteContext=@RigVMFunction_Sequence.A
   )
   n4 = RigVMDispatchNode(ExecuteContext=@n3.True, Array=@n2.Candidates) {template="DISPATCH_RigVMDispatch_ArrayIterator(in Array,out Element,out Index,out Count,out Ratio)"}
   DISPATCH_RigVMDispatch_CoreEquals = RigVMDispatchNode(B=@n4.Element, A=&RigUnit_ResolveConnector(Connector=(Type=Connector,Name="Piston Primary"), SkipSocket=true).Result) {template="DISPATCH_RigVMDispatch_CoreEquals(in A,in B,out Result)"}
   n5 = &RigVMFunction_ControlFlowBranch(ExecuteContext=@n4.ExecuteContext, Condition=@DISPATCH_RigVMDispatch_CoreEquals.Result)
   &RigUnit_SetDefaultMatch(Default=@n4.Element, ExecutePin=@n5.True) {template="Set Default Match::Execute(in Default)"}
   n6 = &RigUnit_GetCandidates()
   n7 = &RigVMFunction_ControlFlowBranch(
    Condition=RigVMDispatchNode(B=(Type=Connector,Name="End Bone"), A=@n6.Connector) {template="DISPATCH_RigVMDispatch_CoreEquals(in A,in B,out Result)"}.Result
    ExecuteContext=@RigVMFunction_Sequence.B
   )
   n8 = RigVMDispatchNode(ExecuteContext=@n7.True, Array=@n6.Candidates) {template="DISPATCH_RigVMDispatch_ArrayIterator(in Array,out Element,out Index,out Count,out Ratio)"}
   n9 = &RigUnit_CollectionChildrenArray(
    TypeToSearch=Bone
    bDefaultChildren=True
    bIncludeParent=False
    bRecursive=false
    Parent=&RigUnit_ResolveConnector(Connector=(Type=Connector,Name="Start Bone"), SkipSocket=true).Result
   )
   DISPATCH_RigVMDispatch_ArrayGetNum = RigVMDispatchNode(Array=@n9.Items) {template="DISPATCH_RigVMDispatch_ArrayGetNum(in Array,out Num)"}
   Greater = &RigVMFunction_MathIntGreater(B=0, A=@DISPATCH_RigVMDispatch_ArrayGetNum.Num) {template="Greater::Execute(in A,in B,out Result)"}
   RigVMFunction_ControlFlowBranch = &RigVMFunction_ControlFlowBranch(ExecuteContext=@n8.ExecuteContext, Condition=@Greater.Result)
   &RigUnit_SetDefaultMatch(
    Default=RigVMDispatchNode(Index=-1, Array=@n9.Items) {template="DISPATCH_RigVMDispatch_ArrayGetAtIndex(in Array,in Index,out Element)"}.Element
    ExecutePin=@RigVMFunction_ControlFlowBranch.True
   ) {template="Set Default Match::Execute(in Default)"}
   Connect to Module Metadata_1 = RigVMFunctionReferenceNode(
    Connector=(Type=Connector,Name="Parent Control")
    Module Metadata Name=Parent Control
    Module Metadata NameSpace=Parent
    ExecuteContext=@RigVMFunction_Sequence.C
   )
   > Connect to Module Metadata_1_1 = RigVMFunctionReferenceNode(Connector=(Type=Connector,Name="Parent Control"), Module Metadata Name=FK Parent, Module Metadata NameSpace=Parent)
   n10 = &RigUnit_GetCandidates()
   n11 = &RigVMFunction_ControlFlowBranch(
    Condition=RigVMDispatchNode(B=(Type=Connector,Name="Parent Bone Start"), A=@n10.Connector) {template="DISPATCH_RigVMDispatch_CoreEquals(in A,in B,out Result)"}.Result
    ExecuteContext=@RigVMFunction_Sequence.D
   )
   n12 = &RigUnit_GetCandidates()
   n13 = &RigVMFunction_ControlFlowBranch(
    Condition=RigVMDispatchNode(B=(Type=Connector,Name="Parent Bone End"), A=@n12.Connector) {template="DISPATCH_RigVMDispatch_CoreEquals(in A,in B,out Result)"}.Result
    ExecuteContext=@n11.Completed
   )
   n14 = RigVMDispatchNode(ExecuteContext=@n13.True, Array=@n12.Candidates) {template="DISPATCH_RigVMDispatch_ArrayIterator(in Array,out Element,out Index,out Count,out Ratio)"}
   DISPATCH_RigVMDispatch_If = RigVMDispatchNode(False=(Type=Connector,Name="End Bone"), True=(Type=Connector,Name="Start Bone"), Condition=RigVMVariableNode() {variable=Reverse}.Value) {template="DISPATCH_RigVMDispatch_If(in Condition,in True,in False,out Result)"}
   n15 = &RigUnit_ResolveConnector(SkipSocket=true, Connector=@DISPATCH_RigVMDispatch_If.Result)
   n16 = &RigVMFunction_ControlFlowBranch(
    ExecuteContext=@n14.ExecuteContext
    Condition=RigVMDispatchNode(A=@n14.Element, B=/Script/ControlRig.RigUnit_HierarchyGetParent(bDefaultParent=True, Child=@n15.Result) {is=struct RigVMUnitNode}.Parent) {template="DISPATCH_RigVMDispatch_CoreEquals(in A,in B,out Result)"}.Result
   )
   &RigUnit_SetDefaultMatch(Default=@n14.Element, ExecutePin=@n16.True) {template="Set Default Match::Execute(in Default)"}
   n17 = RigVMDispatchNode(ExecuteContext=@n11.True, Array=@n10.Candidates) {template="DISPATCH_RigVMDispatch_ArrayIterator(in Array,out Element,out Index,out Count,out Ratio)"}
   n18 = &RigUnit_ResolveConnector(
    SkipSocket=true
    Connector=RigVMDispatchNode(False=(Type=Connector,Name="Start Bone"), True=(Type=Connector,Name="End Bone"), Condition=RigVMVariableNode() {variable=Reverse}.Value) {template="DISPATCH_RigVMDispatch_If(in Condition,in True,in False,out Result)"}.Result
   )
   n19 = &RigVMFunction_ControlFlowBranch(
    ExecuteContext=@n17.ExecuteContext
    Condition=RigVMDispatchNode(A=@n17.Element, B=&RigUnit_HierarchyGetParent(bDefaultParent=True, Child=@n18.Result).Parent) {template="DISPATCH_RigVMDispatch_CoreEquals(in A,in B,out Result)"}.Result
   )
   &RigUnit_SetDefaultMatch(Default=@n17.Element, ExecutePin=@n19.True) {template="Set Default Match::Execute(in Default)"}
   FindClosestBone = RigVMFunctionReferenceNode(Bones=(), ExecuteContext=(), Initial=true, Item=(Type=None,Name="None"))
  ]}
]
hierarchy [
 Piston Primary {type=Connector, loc=<0 0 0>, rot=<0 0 0 1>}
 End Bone {type=Connector, loc=<0 0 0>, rot=<0 0 0 1>}
 Start Bone {type=Connector, loc=<0 0 0>, rot=<0 0 0 1>}
 Parent Bone Start {type=Connector, loc=<0 0 0>, rot=<0 0 0 1>}
 Parent Bone End {type=Connector, loc=<0 0 0>, rot=<0 0 0 1>}
 Parent Up {type=Connector, loc=<0 0 0>, rot=<0 0 0 1>}
 Parent Control {type=Connector, loc=<0 0 0>, rot=<0 0 0 1>}
]
variables [
 Add Up Vector Ctrl {type=bool, default=False}
 As Controls {type=bool, default=True}
 Debug {type=bool, default=False}
 End Control Position {type=double, default=25}
 End Control Settings {
  type=FRigUnit_HierarchyAddControlTransform_Settings
  default="(InitialSpace=LocalSpace,bUsePreferredRotationOrder=False,PreferredRotationOrder=YZX,Limits=(LimitTranslationX=(bMinimum=False,bMaximum=False),LimitTranslationY=(bMinimum=False,bMaximum=False),LimitTranslationZ=(bMinimum=False,bMaximum=False),LimitPitch=(b ...(truncated)"
 }
 Module Settings {
  type=FUserDefinedStruct_#7
  default="(OverrideColor_48_#5=False,LeftColor_3_#4=(R=0,G=0.0231,B=0.4115,A=1),RightColor_5_#6=(R=1,G=0,B=0,A=1),CenterColor_9_#3=(R=1,G=1,B=0,A= ...(truncated)"
 }
 Name {type=FName, default=None}
 Reverse {type=bool, default=False}
 Start Control Settings {
  type=FRigUnit_HierarchyAddControlTransform_Settings
  default="(InitialSpace=LocalSpace,bUsePreferredRotationOrder=False,PreferredRotationOrder=YZX,Limits=(LimitTranslationX=(bMinimum=False,bMaximum=False),LimitTranslationY=(bMinimum=False,bMaximum=False),LimitTranslationZ=(bMinimum=False,bMaximum=False),LimitPitch=(b ...(truncated)"
 }
 Sticky {type=bool, default=True}
 Up Control {type=FRigElementKey, default=(Type=None,Name="")}
 Up Control Settings {
  type=FRigUnit_HierarchyAddControlTransform_Settings
  default="(InitialSpace=LocalSpace,bUsePreferredRotationOrder=False,PreferredRotationOrder=YZX,Limits=(LimitTranslationX=(bMinimum=False,bMaximum=False),LimitTranslationY=(bMinimum=False,bMaximum=False),LimitTranslationZ=(bMinimum=False,bMaximum=False),LimitPitch=(b ...(truncated)"
 }
 Up Vector Axis {type=FVector, default=(X=0,Y=1,Z=0)}
 Up Vector Offset Factor {type=double, default=0.1}
]
