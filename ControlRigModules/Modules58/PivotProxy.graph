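# Control Rig module that adds a proxy pivot control for moving its PivotTarget around another point.
# Construction spawns the control under Parent Control, with configurable shape, scale and driven controls.
# When not interacting, forward solve snaps it to PivotSource (falling back to PivotTarget) plus Local offset, optionally sphere-traced downward within a distance limit, and caches the target's relative transform.
# The Interaction event reapplies that offset, so the target follows the pivot.
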ControlRigBlueprint /ControlRigModules/Modules58/PivotProxy
graphs [
 ConstructionGraph {nodes [
   RigUnit_PrepareForExecution = &RigUnit_PrepareForExecution()
   n1 = /Script/ControlRig.RigUnit_ResolveConnector(Connector=(Type=Connector,Name="PivotSource"), SkipSocket=False) {is=struct RigVMUnitNode}
   DISPATCH_RigVMDispatch_If = RigVMDispatchNode(
    True=@n1.Result
    Condition=@n1.bIsConnected
    False=&RigUnit_ResolveConnector(Connector=(Type=Connector,Name="PivotTarget"), SkipSocket=true).Result
   ) {template="DISPATCH_RigVMDispatch_If(in Condition,in True,in False,out Result)"}
   Concat = /Script/RigVM.RigVMFunction_NameConcat(B=_pivot, A=@DISPATCH_RigVMDispatch_If.Result.Name) {is=struct RigVMUnitNode, template="Concat::Execute(in A,in B,out Result)"}
   RigUnit_GetTransform = &RigUnit_GetTransform(Item=(Type=Bone,Name="Body"), bInitial=False)
   Scale = /Script/RigVM.RigVMFunction_MathVectorScale(Value=(X=1,Y=1,Z=1), Factor=RigVMVariableNode() {variable=Scale}.Value) {is=struct RigVMUnitNode, template="Scale::Execute(in Value,in Factor,out Result)"}
   SpawnControl = /Script/ControlRig.RigUnit_HierarchyAddControlTransform(
    InitialValue=(Rotation=(X=0,Y=0,Z=0,W=1),Translation=(X=0,Y=0,Z=0),Scale3D=(X=1,Y=1,Z=1))
    OffsetSpace=GlobalSpace
    Settings="(InitialSpace=LocalSpace,bUsePreferredRotationOrder=False,PreferredRotationOrder=YZX,Limits=(LimitTranslationX=(bMinimum=False,bMaximum=False),LimitTranslationY=(bMinimum=False,bMaximum=False),LimitTranslationZ=(bMinimum=False,bMaximum=False),LimitPitch=(b ...(truncated)"
    Name=@Concat.Result
    Parent=&RigUnit_ResolveConnector(Connector=(Type=Connector,Name="Parent Control"), SkipSocket=False).Result
    OffsetTransform=@RigUnit_GetTransform.Transform
    ExecutePin=@RigUnit_PrepareForExecution.ExecutePin
    Settings.Shape.Transform.Scale3D=@Scale.Result
    Settings.Proxy.DrivenControls=RigVMVariableNode() {variable=ControlsToRunBackwards}.Value
    Settings.Shape.Name=RigVMVariableNode() {variable=Control Shape}.Value
   ) {
    is=struct RigVMUnitNode
    template="SpawnControl::Execute(in InitialValue,in Settings,in OffsetTransform,in OffsetSpace,in Parent,in Name,out Item)"
   }
   > RigVMVariableNode(Value=@SpawnControl.Item) {variable=Control}
   > Scale Module Controls = RigVMFunctionReferenceNode(
    Controls=((Type=None,Name="None"))
    Global Scale=RigVMDispatchNode(Default=1, Name=Control Size, NameSpace=Root) {template="DISPATCH_RigDispatch_GetModuleMetadata(in Name,in NameSpace,in Default,out Value,out Found)"}.Value
    Controls.0=@SpawnControl.Item
    Local Scale=RigVMVariableNode() {variable=Module Settings}.Value.ControlSize_12_#1
   )
  ]}
 RigVMFunctionLibrary {links [], nodes []}
 RigVMModel {nodes [
   n1 = &RigUnit_BeginExecution()
   RigVMFunction_ControlFlowBranch = &RigVMFunction_ControlFlowBranch(ExecuteContext=@n1.ExecutePin, Condition=/Script/ControlRig.RigUnit_IsInteracting() {is=struct RigVMUnitNode}.bIsInteracting)
   n2 = &RigUnit_ResolveConnector(Connector=(Type=Connector,Name="PivotSource"), SkipSocket=False)
   DISPATCH_RigVMDispatch_If = RigVMDispatchNode(
    False=&RigUnit_ResolveConnector(Connector=(Type=Connector,Name="PivotTarget"), SkipSocket=true).Result
    True=@n2.Result
    Condition=@n2.bIsConnected
   ) {template="DISPATCH_RigVMDispatch_If(in Condition,in True,in False,out Result)"}
   n3 = /Script/ControlRig.RigUnit_HierarchyGetParent(bDefaultParent=True, Child=@DISPATCH_RigVMDispatch_If.Result) {is=struct RigVMUnitNode}
   n4 = RigVMVariableNode() {variable=UseInitial}
   n5 = /Script/RigVM.RigVMFunction_MathTransformMakeAbsolute(
    Parent=&RigVMFunction_MathTransformMakeAbsolute(
     Parent=&RigUnit_GetTransform(bInitial=False, Item=@n3.Parent).Transform
     Local=&RigUnit_GetTransform(Space=LocalSpace, Item=@DISPATCH_RigVMDispatch_If.Result, bInitial=@n4.Value).Transform
    ) {template="Make Absolute::Execute(in Local,in Parent,out Global)"}.Global
    Local=RigVMVariableNode() {variable=Local}.Value
   ) {is=struct RigVMUnitNode, template="Make Absolute::Execute(in Local,in Parent,out Global)"}
   Add = /Script/RigVM.RigVMFunction_MathVectorAdd(B=(X=0,Y=0,Z=-1000), A=@n5.Global.Translation) {is=struct RigVMUnitNode, template="Add::Execute(in A,in B,out Result)"}
   n6 = /Script/ControlRig.RigUnit_SphereTraceByTraceChannel(Radius=5, TraceChannel=TraceTypeQuery1, End=@Add.Result, Start=@n5.Global.Translation) {is=struct RigVMUnitNode}
   Subtract = /Script/RigVM.RigVMFunction_MathVectorSub(
    A=RigVMDispatchNode(False=@n5.Global.Translation, True=@n6.HitLocation, Condition=@n6.bHit) {template="DISPATCH_RigVMDispatch_If(in Condition,in True,in False,out Result)"}.Result
    B=@n5.Global.Translation
   ) {is=struct RigVMUnitNode, template="Subtract::Execute(in A,in B,out Result)"}
   RigVMFunction_MathVectorLength = /Script/RigVM.RigVMFunction_MathVectorLength(Value=@Subtract.Result) {is=struct RigVMUnitNode, template="Length::Execute(in Value,out Result)"}
   Minimum = /Script/RigVM.RigVMFunction_MathDoubleMin(A=@RigVMFunction_MathVectorLength.Result, B=RigVMVariableNode() {variable=MaxProjectionDistance}.Value) {is=struct RigVMUnitNode, template="Minimum::Execute(in A,in B,out Result)"}
   RigVMFunction_MathVectorSetLength = /Script/RigVM.RigVMFunction_MathVectorSetLength(Length=@Minimum.Result, Value=@Subtract.Result) {is=struct RigVMUnitNode, template="SetLength::Execute(in Value,in Length,out Result)"}
   &RigUnit_SetTransform(
    Weight=1
    bInitial=False
    Value=RigVMDispatchNode(
     True=(Rotation=(X=0,Y=0,Z=0,W=1),Translation=(X=0,Y=0,Z=0),Scale3D=(X=1,Y=1,Z=1))
     True.Translation=&RigVMFunction_MathVectorAdd(A=@n5.Global.Translation, B=@RigVMFunction_MathVectorSetLength.Result) {template="Add::Execute(in A,in B,out Result)"}.Result
     False=@n5.Global
     True.Rotation=@n5.Global.Rotation
     True.Scale3D=@n5.Global.Scale3D
     Condition=RigVMVariableNode() {variable=ProjectDown}.Value
    ) {template="DISPATCH_RigVMDispatch_If(in Condition,in True,in False,out Result)"}.Result
    ExecutePin=@RigVMFunction_ControlFlowBranch.False
    Item=RigVMVariableNode() {variable=Control}.Value
   )
   n7 = RigVMVariableNode() {variable=Control}
   n8 = &RigUnit_ResolveConnector(Connector=(Type=Connector,Name="PivotTarget"), SkipSocket=False)
   > RigVMVariableNode(Value=&RigVMFunction_MathTransformMakeRelative(
     Parent=&RigUnit_GetTransform(bInitial=False, Item=@n7.Value).Transform
     Global=&RigUnit_GetTransform(bInitial=False, Item=@n8.Result).Transform
    ).Local) {variable=Offset}
   n9 = /Script/ControlRig.RigUnit_InteractionExecution() {is=struct RigVMUnitNode}
   n10 = RigVMVariableNode() {variable=Control}
   &RigUnit_SetTransform(
    Weight=1
    bInitial=False
    ExecutePin=@n9.ExecutePin
    Value=&RigVMFunction_MathTransformMakeAbsolute(Parent=&RigUnit_GetTransform(bInitial=False, Item=@n10.Value).Transform, Local=RigVMVariableNode() {variable=Offset}.Value) {template="Make Absolute::Execute(in Local,in Parent,out Global)"}.Global
    Item=&RigUnit_ResolveConnector(Connector=(Type=Connector,Name="PivotTarget"), SkipSocket=False).Result
   )
  ]}
 RigVMModel Connector Event Graph {nodes [
   n1 = /Script/ControlRig.RigUnit_ConnectorExecution() {is=struct RigVMUnitNode}
   Connect to Module Metadata = RigVMFunctionReferenceNode(
    Connector=(Type=Connector,Name="Parent Control")
    Module Metadata Name=Root Control
    Module Metadata NameSpace=Root
    ExecuteContext=@n1.ExecuteContext
   )
  ]}
]
hierarchy [
 PivotTarget {type=Connector, loc=<0 0 0>, rot=<0 0 0 1>}
 Root {type=Socket, loc=<0 0 0>, rot=<0 0 0 1>}
 Parent Control {type=Connector, loc=<0 0 0>, rot=<0 0 0 1>}
 PivotSource {type=Connector, loc=<0 0 0>, rot=<0 0 0 1>}
]
variables [
 Control {type=FRigElementKey, default=(Type=Bone,Name="")}
 Control Shape {type=FName, default=Circle_Thick}
 ControlsToRunBackwards {type="TArray<FRigElementKey>"}
 Local {type=FTransform, default=(Rotation=(X=0,Y=0,Z=0,W=1),Translation=(X=0,Y=0,Z=0),Scale3D=(X=1,Y=1,Z=1))}
 MaxProjectionDistance {type=double, default=100}
 Module Settings {
  type=FUserDefinedStruct_#2
  default="(OverrideColor_48_#3=False,LeftColor_3_#4=(R=0,G=0.0231,B=0.4115,A=1),RightColor_5_#5=(R=1,G=0,B=0,A=1),CenterColor_9_#6=(R=1,G=1,B=0,A= ...(truncated)"
 }
 Offset {type=FTransform, default=(Rotation=(X=0,Y=0,Z=0,W=1),Translation=(X=0,Y=0,Z=0),Scale3D=(X=1,Y=1,Z=1))}
 ProjectDown {type=bool, default=False}
 Scale {type=float, default=4}
 UseInitial {type=bool, default=True}
]
