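# Control Rig module that spins a vehicle wheel bone procedurally from its motion, via a spawned control with Break, Gas and Simulate channels.
# Forward solve samples positions into Path, projects the latest step onto the DirectionGiver axis, adds Gas, and damps both by Break.
# That travel, divided by radius-based circumference, incrementally rotates the control's parent null about Axle; the bone copies the control.
# Backward solve resets simulation state and snaps the control to the bone.
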ControlRigBlueprint /ControlRigModules/Modules58/Wheel
graphs [
 ConstructionGraph {nodes [
   RigUnit_PrepareForExecution = &RigUnit_PrepareForExecution()
   RigVMFunction_Sequence = &RigVMFunction_Sequence(ExecuteContext=@RigUnit_PrepareForExecution.ExecutePin)
   n1 = /Script/ControlRig.RigUnit_ResolveConnector(Connector=(Type=Connector,Name="Wheel"), SkipSocket=true) {is=struct RigVMUnitNode}
   n2 = &RigUnit_HierarchyAddNull(
    Space=LocalSpace
    Name=/Script/RigVM.RigVMFunction_NameConcat(B=_wheel_null, A=@n1.Result.Name) {is=struct RigVMUnitNode, template="Concat::Execute(in A,in B,out Result)"}.Result
    Transform=&RigUnit_GetTransform(bInitial=true, Item=@n1.Result).Transform
    Parent=&RigUnit_ResolveConnector(Connector=(Type=Connector,Name="Parent Control"), SkipSocket=False).Result
    ExecutePin=@RigVMFunction_Sequence.A
   )
   > RigVMVariableNode(Value=@n2.Item) {variable=Null}
   Concat = &RigVMFunction_NameConcat(B=_wheel_ctrl, A=@n1.Result.Name) {template="Concat::Execute(in A,in B,out Result)"}
   DISPATCH_RigVMDispatch_CastEnumToInt = RigVMDispatchNode(Value=RigVMVariableNode() {variable=Axle}.Value) {template="DISPATCH_RigVMDispatch_CastEnumToInt(in Value,out Result)"}
   DISPATCH_RigVMDispatch_SelectInt32 = RigVMDispatchNode(Values=((X=0,Y=-0.7071,Z=0,W=0.7071),(X=-0.7071,Y=0,Z=0,W=0.7071),()), Index=@DISPATCH_RigVMDispatch_CastEnumToInt.Result) {template="DISPATCH_RigVMDispatch_SelectInt32(in Index,in Values,out Result)"}
   Scale = /Script/RigVM.RigVMFunction_MathVectorScale(Value=(X=1,Y=1,Z=1), Factor=RigVMVariableNode() {variable=Radius}.Value) {is=struct RigVMUnitNode, template="Scale::Execute(in Value,in Factor,out Result)"}
   > SpawnControl = /Script/ControlRig.RigUnit_HierarchyAddControlRotator(
    InitialValue=(Pitch=0,Yaw=0,Roll=0)
    OffsetSpace=LocalSpace
    OffsetTransform=(Rotation=(X=0,Y=0,Z=0,W=1),Translation=(X=0,Y=0,Z=0),Scale3D=(X=1,Y=1,Z=1))
    Settings="(InitialSpace=LocalSpace,Limits=(LimitPitch=(bMinimum=False,bMaximum=False),LimitYaw=(bMinimum=False,bMaximum=False),LimitRoll=(bMinimum=False,bMaximum=False),MinValue=(Pitch=-180,Yaw=-180,Roll=-180),MaxValue=(Pitch=180,Yaw=180,Roll=180),bDrawLimits=True), ...(truncated)"
    Name=@Concat.Result
    Settings.Shape.Transform.Rotation=@DISPATCH_RigVMDispatch_SelectInt32.Result
    Parent=@n2.Item
    Settings.Shape.Transform.Scale3D=@Scale.Result
    Settings.Shape.Color=RigVMVariableNode() {variable=Color}.Value
   ) {
    is=struct RigVMUnitNode
    template="SpawnControl::Execute(in InitialValue,in Settings,in OffsetTransform,in OffsetSpace,in Parent,in Name,out Item)"
   }
   > RigVMVariableNode(Value=@SpawnControl.Item) {variable=Control}
   > SpawnAnimationChannel = /Script/ControlRig.RigUnit_HierarchyAddAnimationChannelFloat(
    LimitsEnabled=(Enabled=(bMinimum=True,bMaximum=True))
    MaximumValue=1
    MinimumValue=0
    Name=Break
    Parent=@SpawnControl.Item
    InitialValue=RigVMVariableNode() {variable=Break}.Value
   ) {
    is=struct RigVMUnitNode
    template="SpawnAnimationChannel::Execute(in InitialValue,in MinimumValue,in MaximumValue,in LimitsEnabled,in Parent,in Name,out Item)"
   }
   > &RigUnit_HierarchyAddAnimationChannelFloat(
    LimitsEnabled=(Enabled=(bMinimum=True,bMaximum=false))
    MaximumValue=1
    MinimumValue=0
    Name=Gas
    Parent=@SpawnControl.Item
    InitialValue=RigVMVariableNode() {variable=Gas}.Value
   ) {template="SpawnAnimationChannel::Execute(in InitialValue,in MinimumValue,in MaximumValue,in LimitsEnabled,in Parent,in Name,out Item)"}
   > /Script/ControlRig.RigUnit_HierarchyAddAnimationChannelBool(
    LimitsEnabled=()
    MaximumValue=True
    MinimumValue=False
    Name=Simulate
    Parent=@SpawnControl.Item
    InitialValue=RigVMVariableNode() {variable=Simulate}.Value
   ) {
    is=struct RigVMUnitNode
    template="SpawnAnimationChannel::Execute(in InitialValue,in MinimumValue,in MaximumValue,in LimitsEnabled,in Parent,in Name,out Item)"
   }
   > /Script/ControlRig.RigUnit_SetMetadataTag(
    NameSpace=Root
    Tag=Wheel Bones
    Item=&RigUnit_ResolveConnector(Connector=(Type=Connector,Name="Wheel"), SkipSocket=true).Result
   ) {is=struct RigVMUnitNode}
   n3 = RigVMVariableNode() {variable=Module Settings}
   Scale Module Controls = RigVMFunctionReferenceNode(
    Controls=((Type=None,Name="None"))
    Global Scale=RigVMDispatchNode(Default=1, Name=Control Size, NameSpace=Root) {template="DISPATCH_RigDispatch_GetModuleMetadata(in Name,in NameSpace,in Default,out Value,out Found)"}.Value
    ExecuteContext=@RigVMFunction_Sequence.B
    Controls.0=@SpawnControl.Item
    Local Scale=@n3.Value.ControlSize_12_#1
   )
   > Color Module Controls = RigVMFunctionReferenceNode(
    Controls=((Type=None,Name="None"))
    Override=False
    Controls.0=@SpawnControl.Item
    CenterColor=@n3.Value.CenterColor_9_#2
    LeftColor=@n3.Value.LeftColor_3_#3
    RightColor=@n3.Value.RightColor_5_#4
   )
  ]}
 RigVMFunctionLibrary {links [], nodes []}
 RigVMFunction_Sequence_2_ContainedGraph {nodes [
   n1 = RigVMFunctionEntryNode()
   RigVMFunction_Sequence = &RigVMFunction_Sequence(ExecuteContext=@n1.ExecuteContext)
   n2 = &RigVMFunction_Sequence(ExecuteContext=@RigVMFunction_Sequence.B)
   RigVMFunctionReturnNode(A=@RigVMFunction_Sequence.A, B=@n2.A, C=@n2.B)
  ]}
 RigVMModel {nodes [
   n1 = /Script/ControlRig.RigUnit_InverseExecution() {is=struct RigVMUnitNode}
   n2 = &RigVMFunction_Sequence(ExecuteContext=@n1.ExecutePin)
   RigVMVariableNode(ExecuteContext=@n2.A) {variable=Simulate}
   > RigVMVariableNode() {variable=Travel}
   > RigVMVariableNode(Value=()) {variable=Path}
   > RigVMVariableNode() {variable=Gas}
   > RigVMVariableNode() {variable=Break}
   &RigUnit_SetTransform(
    Weight=1
    bInitial=False
    Value=&RigUnit_GetTransform(bInitial=true, Item=RigVMVariableNode() {variable=Null}.Value).Transform
    ExecutePin=@n2.B
    Item=RigVMVariableNode() {variable=Null}.Value
   )
   > &RigUnit_SetTransform(
    Weight=1
    bInitial=False
    Value=&RigUnit_GetTransform(bInitial=False, Item=&RigUnit_ResolveConnector(Connector=(Type=Connector,Name="Wheel"), SkipSocket=true).Result).Transform
    Item=RigVMVariableNode() {variable=Control}.Value
   )
   RigUnit_BeginExecution = &RigUnit_BeginExecution()
   RigVMFunction_ControlFlowBranch = &RigVMFunction_ControlFlowBranch(Condition=false, ExecuteContext=@RigUnit_BeginExecution.ExecutePin)
   Set Transform_3 = /Script/ControlRig.RigUnit_SetTranslation(
    Space=GlobalSpace
    Weight=1
    bInitial=False
    bPropagateToChildren=True
    Value=&RigUnit_GetTransform(bInitial=False, Item=&RigUnit_ResolveConnector(Connector=(Type=Connector,Name="Wheel"), SkipSocket=False).Result).Transform.Translation
    ExecutePin=@RigVMFunction_ControlFlowBranch.Completed
    Item=RigVMVariableNode() {variable=Null}.Value
   ) {is=struct RigVMUnitNode, template="Set Transform::Execute(in Item,in Space,in bInitial,in Value,in Weight,in bPropagateToChildren)"}
   n3 = &RigVMFunction_Sequence(ExecuteContext=@"Set Transform_3".ExecutePin)
   n4 = &RigUnit_GetTransform(bInitial=False, Item=&RigUnit_ResolveConnector(Connector=(Type=Connector,Name="Wheel"), SkipSocket=true).Result)
   n5 = RigVMDispatchNode(Value=RigVMVariableNode() {variable=Axle}.Value) {template="DISPATCH_RigVMDispatch_CastEnumToInt(in Value,out Result)"}
   RigVMFunction_MathVectorCross = /Script/RigVM.RigVMFunction_MathVectorCross(
    B=(X=0,Y=0,Z=1)
    A=/Script/RigVM.RigVMFunction_MathTransformRotateVector(
     Vector=RigVMDispatchNode(Values=((X=1,Y=0,Z=0),(X=0,Y=1,Z=0),(X=0,Y=0,Z=1)), Index=@n5.Result) {template="DISPATCH_RigVMDispatch_SelectInt32(in Index,in Values,out Result)"}.Result
     Transform=@n4.Transform
    ) {is=struct RigVMUnitNode, template="Rotate Vector::Execute(in Transform,in Vector,out Result)"}.Result
   ) {is=struct RigVMUnitNode, template="Cross::Execute(in A,in B,out Result)"}
   Scale = &RigVMFunction_MathVectorScale(
    Value=&RigVMFunction_MathVectorCross(
     A=@RigVMFunction_MathVectorCross.Result
     B=&RigVMFunction_MathVectorCross(B=(X=0,Y=0,Z=1), A=@RigVMFunction_MathVectorCross.Result) {template="Cross::Execute(in A,in B,out Result)"}.Result
    ) {template="Cross::Execute(in A,in B,out Result)"}.Result
    Factor=RigVMVariableNode() {variable=Radius}.Value
   ) {template="Scale::Execute(in Value,in Factor,out Result)"}
   n6 = RigVMVariableNode() {variable=Radius}
   DISPATCH_RigVMDispatch_ArrayInsert = RigVMDispatchNode(
    Index=0
    Element=/Script/RigVM.RigVMFunction_MathVectorAdd(
     A=@n4.Transform.Translation
     B=&RigVMFunction_MathVectorScale(Value=@Scale.Result, Factor=@n6.Value) {template="Scale::Execute(in Value,in Factor,out Result)"}.Result
    ) {is=struct RigVMUnitNode, template="Add::Execute(in A,in B,out Result)"}.Result
    ExecuteContext=@n3.A
    Array=RigVMVariableNode() {variable=Path}.Value
   ) {template="DISPATCH_RigVMDispatch_ArrayInsert(io Array,in Index,in Element)"}
   Minimum = /Script/RigVM.RigVMFunction_MathIntMin(
    A=RigVMDispatchNode(Array=@DISPATCH_RigVMDispatch_ArrayInsert.Array) {template="DISPATCH_RigVMDispatch_ArrayGetNum(in Array,out Num)"}.Num
    B=RigVMVariableNode() {variable=MaxSamples}.Value
   ) {is=struct RigVMUnitNode, template="Minimum::Execute(in A,in B,out Result)"}
   > DISPATCH_RigVMDispatch_ArraySetNum = RigVMDispatchNode(Array=@DISPATCH_RigVMDispatch_ArrayInsert.Array, Num=@Minimum.Result) {template="DISPATCH_RigVMDispatch_ArraySetNum(io Array,in Num)"}
   > RigVMVariableNode(Value=@DISPATCH_RigVMDispatch_ArraySetNum.Array) {variable=Path}
   RigVMFunction_Sequence = RigVMAggregateNode(ExecuteContext=@n3.B)
   GetAnimationChannel = /Script/ControlRig.RigUnit_GetFloatAnimationChannel(Channel=Break, bInitial=False, Control=RigVMVariableNode() {variable=Control}.Value.Name) {is=struct RigVMUnitNode, template="GetAnimationChannel::Execute(out Value,in Control,in Channel,in bInitial)"}
   Subtract = &RigVMFunction_MathDoubleSub(A=1, B=@GetAnimationChannel.Value)
   Clamp = &RigVMFunction_MathDoubleClamp(Value=@Subtract.Result)
   n7 = RigVMDispatchNode(Value=RigVMVariableNode() {variable=DirectionAxis}.Value) {template="DISPATCH_RigVMDispatch_CastEnumToInt(in Value,out Result)"}
   n8 = &RigUnit_ResolveConnector(Connector=(Type=Connector,Name="DirectionGiver"), SkipSocket=true)
   n9 = RigVMVariableNode() {variable=Path}
   DISPATCH_RigVMDispatch_ArrayGetAtIndex = RigVMDispatchNode(Index=0, Array=@n9.Value) {template="DISPATCH_RigVMDispatch_ArrayGetAtIndex(in Array,in Index,out Element)"}
   DISPATCH_RigVMDispatch_ArrayGetNum = RigVMDispatchNode(Array=@n9.Value) {template="DISPATCH_RigVMDispatch_ArrayGetNum(in Array,out Num)"}
   n10 = RigVMDispatchNode(
    False=0
    True=1
    Condition=/Script/RigVM.RigVMFunction_MathIntGreater(B=1, A=@DISPATCH_RigVMDispatch_ArrayGetNum.Num) {is=struct RigVMUnitNode, template="Greater::Execute(in A,in B,out Result)"}.Result
   ) {template="DISPATCH_RigVMDispatch_If(in Condition,in True,in False,out Result)"}
   Dot__ = /Script/RigVM.RigVMFunction_MathVectorDot(
    B=/Script/RigVM.RigVMFunction_MathQuaternionRotateVector(
     Vector=RigVMDispatchNode(Values=((X=1,Y=0,Z=0),(X=0,Y=1,Z=0),(X=0,Y=0,Z=1)), Index=@n7.Result) {template="DISPATCH_RigVMDispatch_SelectInt32(in Index,in Values,out Result)"}.Result
     Transform=&RigUnit_GetTransform(bInitial=False, Item=@n8.Result).Transform.Rotation
    ) {is=struct RigVMUnitNode, template="Rotate Vector::Execute(in Transform,in Vector,out Result)"}.Result
    A=/Script/RigVM.RigVMFunction_MathVectorSub(
     A=@DISPATCH_RigVMDispatch_ArrayGetAtIndex.Element
     B=RigVMDispatchNode(Index=@n10.Result, Array=@n9.Value) {template="DISPATCH_RigVMDispatch_ArrayGetAtIndex(in Array,in Index,out Element)"}.Element
    ) {is=struct RigVMUnitNode, template="Subtract::Execute(in A,in B,out Result)"}.Result
   ) {is=struct RigVMUnitNode, template="Dot,|::Execute(in A,in B,out Result)"}
   n11 = /Script/RigVM.RigVMFunction_MathDoubleNegate(Value=@Dot__.Result) {is=struct RigVMUnitNode, template="Negate::Execute(in Value,out Result)"}
   Greater = &RigVMFunction_MathIntGreater(B=1, A=@DISPATCH_RigVMDispatch_ArrayGetNum.Num) {template="Greater::Execute(in A,in B,out Result)"}
   RigVMFunction_MathBoolAnd = &RigVMFunction_MathBoolAnd(
    A=/Script/ControlRig.RigUnit_GetBoolAnimationChannel(Channel=Simulate, bInitial=False, Control=RigVMVariableNode() {variable=Control}.Value.Name) {is=struct RigVMUnitNode, template="GetAnimationChannel::Execute(out Value,in Control,in Channel,in bInitial)"}.Value
    B=@Greater.Result
   )
   n12 = &RigUnit_GetFloatAnimationChannel(Channel=Gas, bInitial=False, Control=RigVMVariableNode() {variable=Control}.Value.Name) {template="GetAnimationChannel::Execute(out Value,in Control,in Channel,in bInitial)"}
   Negate = &RigVMFunction_MathDoubleNegate(Value=@n12.Value) {template="Negate::Execute(in Value,out Result)"}
   DISPATCH_RigVMDispatch_If = RigVMDispatchNode(True=@n12.Value, False=@Negate.Result, Condition=RigVMVariableNode() {variable=Invert}.Value) {template="DISPATCH_RigVMDispatch_If(in Condition,in True,in False,out Result)"}
   Multiply = &RigVMFunction_MathDoubleMul(B=@DISPATCH_RigVMDispatch_If.Result, A=RigVMVariableNode() {variable=Radius}.Value)
   Add = &RigVMFunction_MathDoubleAdd(
    A=RigVMDispatchNode(False=0, True=&RigVMFunction_MathDoubleMul(A=@Clamp.Result, B=@n11.Result).Result, Condition=@RigVMFunction_MathBoolAnd.Result) {template="DISPATCH_RigVMDispatch_If(in Condition,in True,in False,out Result)"}.Result
    B=&RigVMFunction_MathDoubleMul(A=@Clamp.Result, B=@Multiply.Result).Result
   )
   RigVMVariableNode(Value=@Add.Result, ExecuteContext=@RigVMFunction_Sequence.A) {variable=NewTravel}
   DISPATCH_RigVMDispatch_CastEnumToInt = RigVMDispatchNode(Value=RigVMVariableNode() {variable=Axle}.Value) {template="DISPATCH_RigVMDispatch_CastEnumToInt(in Value,out Result)"}
   DISPATCH_RigVMDispatch_SelectInt32 = RigVMDispatchNode(Values=((X=1,Y=0,Z=0),(X=0,Y=1,Z=0),(X=0,Y=0,Z=1)), Index=@DISPATCH_RigVMDispatch_CastEnumToInt.Result) {template="DISPATCH_RigVMDispatch_SelectInt32(in Index,in Values,out Result)"}
   TwoPi = /Script/RigVM.RigVMFunction_MathDoubleConstTwoPi() {is=struct RigVMUnitNode, template="TwoPi::Execute(out Value)"}
   Divide = &RigVMFunction_MathDoubleDiv(
    B=&RigVMFunction_MathDoubleMul(B=@TwoPi.Value, A=RigVMVariableNode() {variable=Radius}.Value).Result
    A=RigVMVariableNode() {variable=NewTravel}.Value
   )
   RigVMFunction_MathQuaternionFromAxisAndAngle = /Script/RigVM.RigVMFunction_MathQuaternionFromAxisAndAngle(Axis=@DISPATCH_RigVMDispatch_SelectInt32.Result, Angle=@Divide.Result) {is=struct RigVMUnitNode, template="FromAxisAndAngle::Execute(in Axis,in Angle,out Result)"}
   > &RigUnit_OffsetTransformForItem(
    OffsetTransform=(Rotation=(X=0,Y=0,Z=0,W=1),Translation=(X=0,Y=0,Z=0),Scale3D=(X=1,Y=1,Z=1))
    Weight=1
    bPropagateToChildren=True
    OffsetTransform.Rotation=@RigVMFunction_MathQuaternionFromAxisAndAngle.Result
    Item=RigVMVariableNode() {variable=Null}.Value
   )
   RigVMVariableNode(
    Value=&RigVMFunction_MathDoubleAdd(B=RigVMVariableNode() {variable=NewTravel}.Value, A=RigVMVariableNode() {variable=Travel}.Value).Result
    ExecuteContext=@RigVMFunction_Sequence.B
   ) {variable=Travel}
   &RigUnit_SetTransform(
    Weight=1
    bInitial=False
    Value=&RigUnit_GetTransform(bInitial=False, Item=RigVMVariableNode() {variable=Control}.Value).Transform
    Item=&RigUnit_ResolveConnector(Connector=(Type=Connector,Name="Wheel"), SkipSocket=true).Result
    ExecutePin=@RigVMFunction_Sequence.C
   )
   RigVMFunction_Timeline = /Script/RigVM.RigVMFunction_Timeline(Speed=1) {is=struct RigVMUnitNode}
   Sin = /Script/RigVM.RigVMFunction_MathDoubleSin(Value=@RigVMFunction_Timeline.Time) {is=struct RigVMUnitNode, template="Sin::Execute(in Value,out Result)"}
   RigUnit_OffsetTransformForItem = &RigUnit_OffsetTransformForItem(
    Item=(Type=Bone,Name="Root")
    OffsetTransform=(Rotation=(X=0,Y=0,Z=0,W=1),Translation=(X=0,Y=0,Z=0),Scale3D=(X=1,Y=1,Z=1))
    Weight=1
    bPropagateToChildren=True
    OffsetTransform.Translation.Y=&RigVMFunction_MathDoubleMul(B=50, A=@Sin.Result).Result
    ExecutePin=@RigVMFunction_ControlFlowBranch.True
   )
   RigVMFunction_DebugLineStripNoSpace = /Script/RigVM.RigVMFunction_DebugLineStripNoSpace(
    DebugDrawSettings=(DepthPriority=SDPG_Foreground,Lifetime=-1)
    ExecutePin=()
    Thickness=0
    WorldOffset=(Rotation=(X=0,Y=0,Z=0,W=1),Translation=(X=0,Y=0,Z=0),Scale3D=(X=1,Y=1,Z=1))
    bEnabled=false
    Points=RigVMVariableNode() {variable=Path}.Value
    Color=RigVMVariableNode() {variable=Color}.Value
   ) {is=struct RigVMUnitNode}
   RigVMVariableNode() {variable=Simulate}
  ]}
 RigVMModel Connector Event Graph {nodes [
   n1 = /Script/ControlRig.RigUnit_ConnectorExecution() {is=struct RigVMUnitNode}
   Connect to Module Metadata = RigVMFunctionReferenceNode(
    Connector=(Type=Connector,Name="Parent Control")
    Module Metadata Name=Parent Control
    Module Metadata NameSpace=Parent
    ExecuteContext=@n1.ExecuteContext
   )
   n2 = &RigVMFunction_ControlFlowBranch(
    ExecuteContext=@"Connect to Module Metadata".ExecuteContext
    Condition=RigVMDispatchNode(
     B=/Script/ControlRig.RigUnit_GetCandidates() {is=struct RigVMUnitNode}.Connector
     A=/Script/ControlRig.RigUnit_Item(Item=(Type=Connector,Name="DirectionGiver")) {is=struct RigVMUnitNode}.Item
    ) {template="DISPATCH_RigVMDispatch_CoreEquals(in A,in B,out Result)"}.Result
   )
   n3 = RigVMDispatchNode(
    Index=0
    Array=/Script/ControlRig.RigUnit_HierarchyGetParentsItemArray(
     bDefaultParent=True
     bIncludeChild=False
     bReverse=true
     Child=&RigUnit_ResolveConnector(Connector=(Type=Connector,Name="Wheel"), SkipSocket=False).Result
    ) {is=struct RigVMUnitNode}.Parents
   ) {template="DISPATCH_RigVMDispatch_ArrayGetAtIndex(in Array,in Index,out Element)"}
   /Script/ControlRig.RigUnit_SetDefaultMatch(
    Default=RigVMDispatchNode(
     Index=0
     Array=/Script/ControlRig.RigUnit_CollectionChildrenArray(TypeToSearch=Bone, bDefaultChildren=True, bIncludeParent=False, bRecursive=False, Parent=@n3.Element) {is=struct RigVMUnitNode}.Items
    ) {template="DISPATCH_RigVMDispatch_ArrayGetAtIndex(in Array,in Index,out Element)"}.Element
    ExecutePin=@n2.True
   ) {is=struct RigVMUnitNode, template="Set Default Match::Execute(in Default)"}
  ]}
]
hierarchy [
 Wheel {type=Connector, loc=<0 0 0>, rot=<0 0 0 1>}
 pelvis_socket {type=Socket, loc=<0 0 0>, rot=<0 0 0 1>}
 Parent Control {type=Connector, loc=<0 0 0>, rot=<0 0 0 1>}
 DirectionGiver {type=Connector, loc=<0 0 0>, rot=<0 0 0 1>}
]
variables [
 Axle {type=ERigControlAxis, default=X}
 Break {type=double, default=0}
 Color {type=FLinearColor, default=(R=1,G=0,B=0,A=1)}
 Control {type=FRigElementKey, default=(Type=None,Name="")}
 DirectionAxis {type=ERigControlAxis, default=Y}
 Gas {type=double, default=0}
 Invert {type=bool, default=False}
 MaxSamples {type=int32, default=128}
 Module Settings {
  type=FUserDefinedStruct_#5
  default="(OverrideColor_48_#6=False,LeftColor_3_#3=(R=0,G=0.0231,B=0.4115,A=1),RightColor_5_#4=(R=1,G=0,B=0,A=1),CenterColor_9_#2=(R=1,G=1,B=0,A= ...(truncated)"
 }
 NewTravel {type=double, default=0}
 Null {type=FRigElementKey, default=(Type=None,Name="")}
 Path {type="TArray<FVector>"}
 Radius {type=float, default=3.9}
 Simulate {type=bool, default=True}
 Travel {type=double, default=0}
]
